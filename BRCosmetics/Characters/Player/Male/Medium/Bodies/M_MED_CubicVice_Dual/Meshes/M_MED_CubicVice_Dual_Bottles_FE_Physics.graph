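# Physics asset for the bottles piece of the male medium CubicVice Dual body cosmetic, defining bodies for dangling bottles and left-side ropes.
# Four capsule chains hang from a pelvis capsule: two two-link bottles and two three-link ropes, each starting at a tiny MSTR root body.
# Each MSTR joint rotates freely against the pelvis, while the links below it have narrow swing limits and zero twist limit.
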PhysicsAsset /BRCosmetics/Characters/Player/Male/Medium/Bodies/M_MED_CubicVice_Dual/Meshes/M_MED_CubicVice_Dual_Bottles_FE_Physics
bodies {
 dyn_bottleA_01 [{shapeType=Capsule, length=2.6153, radius=1.6776, shapeName=dyn_bottleA_01_capsule, center=<1.5163 -1.4 0>, rotation=<rot 90 0 0>}]
 dyn_bottleA_02 [{shapeType=Capsule, length=6.243, radius=4.7904, shapeName=dyn_bottleA_02_capsule, center=<5.0606 -4 -1>, rotation=<rot 90 0 0>}]
 dyn_bottleA_MSTR [{shapeType=Capsule, length=0.5, radius=0.5, shapeName=dyn_bottleA_MSTR_capsule}]
 dyn_bottleB_01 [{shapeType=Capsule, length=2.3201, radius=1.6338, shapeName=dyn_bottleB_01_capsule, center=<1.5 -1 0>, rotation=<rot 90 0 0>}]
 dyn_bottleB_02 [{
   shapeType=Capsule
   length=7.224
   radius=4.2348
   shapeName=dyn_bottleB_02_capsule
   center=<4.5871 -1 -3>
   rotation=<rot 90 -152.5946 -161.3438>
  }]
 dyn_bottleB_MSTR [{shapeType=Capsule, length=0.5, radius=0.5, shapeName=dyn_bottleB_MSTR_capsule}]
 dyn_ropeA_l_01 [{shapeType=Capsule, length=5, radius=2, shapeName=dyn_ropeA_l_01_capsule, center=<2 -1.3 0>, rotation=<rot 90 0 0>}]
 dyn_ropeA_l_02 [{shapeType=Capsule, length=5, radius=2, shapeName=dyn_ropeA_l_02_capsule, center=<2 -1.3 0>, rotation=<rot 90 0 0>}]
 dyn_ropeA_l_03 [{shapeType=Capsule, length=5, radius=2, shapeName=dyn_ropeA_l_03_capsule, center=<2 -1.3 0>, rotation=<rot 90 0 0>}]
 dyn_ropeA_l_MSTR [{shapeType=Capsule, length=0.5, radius=0.5, shapeName=dyn_ropeA_l_MSTR_capsule, rotation=<rot 90 0 0>}]
 dyn_ropeB_l_01 [{shapeType=Capsule, length=5, radius=2, shapeName=dyn_ropeB_l_01_capsule, center=<2 -3 0>, rotation=<rot 90 0 0>}]
 dyn_ropeB_l_02 [{shapeType=Capsule, length=5, radius=2, shapeName=dyn_ropeB_l_02_capsule, center=<2 -1.3 0>, rotation=<rot 90 0 0>}]
 dyn_ropeB_l_03 [{shapeType=Capsule, length=5, radius=2, shapeName=dyn_ropeB_l_03_capsule, center=<2 -1.3 0>, rotation=<rot 90 0 0>}]
 dyn_ropeB_l_MSTR [{shapeType=Capsule, length=0.5, radius=0.5, shapeName=dyn_ropeB_l_MSTR_capsule, rotation=<rot 90 0 0>}]
 pelvis [{shapeType=Capsule, length=9.9256, radius=12.1121, center=<0 -0.7167 -0.0346>}]
}
constraints [
 {
  bone1Name=dyn_bottleA_MSTR
  bone2Name=pelvis
  swing1LimitDegrees=45
  swing1Motion=Free
  swing2LimitDegrees=45
  swing2Motion=Free
  twistLimitDegrees=45
  twistMotion=Free
 }
 {bone1Name=dyn_bottleA_01, bone2Name=dyn_bottleA_MSTR, swing1LimitDegrees=20, swing2LimitDegrees=20, twistLimitDegrees=0}
 {bone1Name=dyn_bottleA_02, bone2Name=dyn_bottleA_01, swing1LimitDegrees=20, swing2LimitDegrees=20, twistLimitDegrees=0}
 {
  bone1Name=dyn_bottleB_MSTR
  bone2Name=pelvis
  swing1LimitDegrees=45
  swing1Motion=Free
  swing2LimitDegrees=45
  swing2Motion=Free
  twistLimitDegrees=45
  twistMotion=Free
 }
 {bone1Name=dyn_bottleB_01, bone2Name=dyn_bottleB_MSTR, swing1LimitDegrees=20, swing2LimitDegrees=25, twistLimitDegrees=0}
 {bone1Name=dyn_bottleB_02, bone2Name=dyn_bottleB_01, swing1LimitDegrees=20, swing2LimitDegrees=25, twistLimitDegrees=0}
 {
  bone1Name=dyn_ropeA_l_MSTR
  bone2Name=pelvis
  swing1LimitDegrees=45
  swing1Motion=Free
  swing2LimitDegrees=45
  swing2Motion=Free
  twistLimitDegrees=45
  twistMotion=Free
 }
 {bone1Name=dyn_ropeA_l_01, bone2Name=dyn_ropeA_l_MSTR, swing1LimitDegrees=13, swing2LimitDegrees=17, twistLimitDegrees=0}
 {bone1Name=dyn_ropeA_l_02, bone2Name=dyn_ropeA_l_01, swing1LimitDegrees=13, swing2LimitDegrees=17, twistLimitDegrees=0}
 {bone1Name=dyn_ropeA_l_03, bone2Name=dyn_ropeA_l_02, swing1LimitDegrees=13, swing2LimitDegrees=17, twistLimitDegrees=0}
 {
  bone1Name=dyn_ropeB_l_MSTR
  bone2Name=pelvis
  swing1LimitDegrees=45
  swing1Motion=Free
  swing2LimitDegrees=45
  swing2Motion=Free
  twistLimitDegrees=45
  twistMotion=Free
 }
 {bone1Name=dyn_ropeB_l_01, bone2Name=dyn_ropeB_l_MSTR, swing1LimitDegrees=20, swing2LimitDegrees=20, twistLimitDegrees=0}
 {bone1Name=dyn_ropeB_l_02, bone2Name=dyn_ropeB_l_01, swing1LimitDegrees=20, swing2LimitDegrees=20, twistLimitDegrees=0}
 {bone1Name=dyn_ropeB_l_03, bone2Name=dyn_ropeB_l_02, swing1LimitDegrees=20, swing2LimitDegrees=20, twistLimitDegrees=0}
]
