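# Control Rig for the M_MED_CubicVice_Dual male cosmetic body, driving secondary deformation of its skirt, collar, shoulder strap, bottle and rope bones.
# It runs the shared DeformRig function, then a sequence of weighted parent and aim constraints that follow spine, limbs and head.
# Spherical pose readers on thighs, head and cloth bones weight corrective transform offsets, with control-to-bone distance also pushing the left collar.
# Front-end state variables adjust skirt and bottle behaviour.
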ControlRigBlueprint /BRCosmetics/Characters/Player/Male/Medium/Bodies/M_MED_CubicVice_Dual/Meshes/M_MED_CubicVice_Dual_CtrlRig
graphs [
 RigVMFunctionLibrary {links [], nodes []}
 RigVMFunction_Sequence_1_ContainedGraph {nodes [
   n1 = RigVMFunctionEntryNode()
   RigVMFunction_Sequence = &RigVMFunction_Sequence(ExecuteContext=@n1.ExecuteContext)
   n2 = &RigVMFunction_Sequence(ExecuteContext=@RigVMFunction_Sequence.B)
   n3 = &RigVMFunction_Sequence(ExecuteContext=@n2.B)
   RigVMFunctionReturnNode(A=@RigVMFunction_Sequence.A, B=@n2.A, C=@n3.A, D=@n3.B)
  ]}
 RigVMModel {nodes [
   &RigUnit_ParentConstraint(
    AdvancedSettings=(RotationOrderForFilter=XZY)
    Child=(Type=Bone,Name="deform_skirt_bk_l")
    ExecutePin=()
    Parents="((Item=(Type=Bone,Name=\"thigh_twist_01_l\"),Weight=0.0048),(Item=(Type=Bone,Name=\"spine_03\"),Weight=0.0081),(Item=(Type=Bone,Name=\"spine_02\"),Weight=0.0124),(Item=(Type=Bone,Name=\"spine_01\"),Weight=0.0126),(Item=(Type=Bone,Name=\"deform_glute_l\"),Weight=0.02 ...(truncated)"
    Weight=1
   )
   > &RigUnit_ParentConstraint(
    AdvancedSettings=(RotationOrderForFilter=XZY)
    Child=(Type=Bone,Name="deform_skirt_bk_r")
    Parents="((Item=(Type=Bone,Name=\"thigh_twist_01_r\"),Weight=0.0048),(Item=(Type=Bone,Name=\"spine_03\"),Weight=0.0081),(Item=(Type=Bone,Name=\"spine_02\"),Weight=0.0124),(Item=(Type=Bone,Name=\"spine_01\"),Weight=0.0126),(Item=(Type=Bone,Name=\"deform_glute_r\"),Weight=0.02 ...(truncated)"
    Weight=1
   )
   > &RigUnit_ParentConstraint(
    AdvancedSettings=(RotationOrderForFilter=XZY)
    Child=(Type=Bone,Name="deform_skirt_bk_r1")
    Parents="((Item=(Type=Bone,Name=\"thigh_twist_01_r\"),Weight=0.0048),(Item=(Type=Bone,Name=\"spine_03\"),Weight=0.0081),(Item=(Type=Bone,Name=\"spine_02\"),Weight=0.0124),(Item=(Type=Bone,Name=\"spine_01\"),Weight=0.0126),(Item=(Type=Bone,Name=\"deform_glute_r\"),Weight=0.02 ...(truncated)"
    Weight=1
   )
   > &RigUnit_ParentConstraint(
    AdvancedSettings=(RotationOrderForFilter=XZY)
    Child=(Type=Bone,Name="deform_skirt_m_r")
    Parents="((Item=(Type=Bone,Name=\"deform_groin_r\"),Weight=0.0109),(Item=(Type=Bone,Name=\"thigh_twist_01_r\"),Weight=0.013),(Item=(Type=Bone,Name=\"spine_03\"),Weight=0.044),(Item=(Type=Bone,Name=\"spine_01\"),Weight=0.0674),(Item=(Type=Bone,Name=\"spine_02\"),Weight=0.0782 ...(truncated)"
    Weight=1
   )
   > &RigUnit_ParentConstraint(
    AdvancedSettings=(RotationOrderForFilter=XZY)
    Child=(Type=Bone,Name="deform_skirt_f_r")
    Parents="((Item=(Type=Bone,Name=\"thigh_twist_01_r\"),Weight=0.0071),(Item=(Type=Bone,Name=\"deform_groin_r\"),Weight=0.0095),(Item=(Type=Bone,Name=\"spine_03\"),Weight=0.0531),(Item=(Type=Bone,Name=\"spine_01\"),Weight=0.0787),(Item=(Type=Bone,Name=\"spine_02\"),Weight=0.09 ...(truncated)"
    Weight=1
   )
   > &RigUnit_ParentConstraint(
    AdvancedSettings=(RotationOrderForFilter=XZY)
    Child=(Type=Bone,Name="deform_skirt_f_l")
    Parents="((Item=(Type=Bone,Name=\"thigh_twist_01_l\"),Weight=0.0032),(Item=(Type=Bone,Name=\"deform_groin_l\"),Weight=0.0036),(Item=(Type=Bone,Name=\"spine_03\"),Weight=0.0541),(Item=(Type=Bone,Name=\"spine_01\"),Weight=0.0788),(Item=(Type=Bone,Name=\"spine_02\"),Weight=0.09 ...(truncated)"
    Weight=1
   )
   > &RigUnit_ParentConstraint(
    AdvancedSettings=(RotationOrderForFilter=XZY)
    Child=(Type=Bone,Name="deform_skirt_m_l")
    Parents="((Item=(Type=Bone,Name=\"deform_groin_l\"),Weight=0.0109),(Item=(Type=Bone,Name=\"thigh_twist_01_l\"),Weight=0.013),(Item=(Type=Bone,Name=\"spine_03\"),Weight=0.044),(Item=(Type=Bone,Name=\"spine_01\"),Weight=0.0674),(Item=(Type=Bone,Name=\"spine_02\"),Weight=0.0782 ...(truncated)"
    Weight=1
   )
   > n1 = &RigUnit_SphericalPoseReader(
    ActiveRegionScaleFactors=(PositiveWidth=3.5,NegativeWidth=0.5,PositiveHeight=3,NegativeHeight=3)
    ActiveRegionSize=0.1
    Debug=(bDrawDebug=false,DebugScale=25,DebugSegments=20,DebugThickness=0.25)
    DriverAxis=(X=-1,Y=0,Z=0)
    DriverItem=(Type=Bone,Name="thigh_l")
    FalloffRegionScaleFactors=(PositiveWidth=1,NegativeWidth=1,PositiveHeight=1,NegativeHeight=1)
    FalloffSize=0.45
    OptionalParentItem=(Type=Bone)
    RotationOffset=(X=90,Y=0,Z=-30)
   )
   > n2 = &RigUnit_SphericalPoseReader(
    ActiveRegionScaleFactors=(PositiveWidth=3.5,NegativeWidth=0.5,PositiveHeight=3,NegativeHeight=3)
    ActiveRegionSize=0.1
    Debug=(bDrawDebug=false,DebugScale=25,DebugSegments=20,DebugThickness=0.25)
    DriverAxis=(X=1,Y=0,Z=0)
    DriverItem=(Type=Bone,Name="thigh_r")
    FalloffRegionScaleFactors=(PositiveWidth=1,NegativeWidth=1,PositiveHeight=1,NegativeHeight=1)
    FalloffSize=0.45
    FlipHeightScaling=false
    FlipWidthScaling=true
    OptionalParentItem=(Type=Bone)
    RotationOffset=(X=-90,Y=0,Z=0)
   )
   > &RigUnit_AimConstraintLocalSpaceOffset(
    AdvancedSettings=(DebugSettings=(bEnabled=true,Scale=10,WorldOffset=(Rotation=(X=0,Y=0,Z=0,W=1),Translation=(X=0,Y=0,Z=0),Scale3D=(X=1,Y=1,Z=1))),RotationOrderForFilter=XZY)
    AimAxis=(X=-1,Y=0,Z=0)
    Child=(Type=Bone,Name="deform_skirt_f_l")
    Filter=(bX=True,bY=True,bZ=True)
    Parents=((Item=(Type=Bone,Name="calf_l"),Weight=1))
    UpAxis=(X=0,Y=1,Z=0)
    Weight=1
    WorldUp=(Target=(X=0,Y=0,Z=1),Kind=Location,Space=(Type=Bone,Name="calf_l"))
   )
   > &RigUnit_AimConstraintLocalSpaceOffset(
    AdvancedSettings=(DebugSettings=(bEnabled=true,Scale=10,WorldOffset=(Rotation=(X=0,Y=0,Z=0,W=1),Translation=(X=0,Y=0,Z=0),Scale3D=(X=1,Y=1,Z=1))),RotationOrderForFilter=XZY)
    Child=(Type=Bone,Name="deform_skirt_f_r")
    Filter=(bX=True,bY=True,bZ=True)
    Parents=((Item=(Type=Bone,Name="calf_r"),Weight=1))
    UpAxis=(X=0,Y=-1,Z=0)
    Weight=1
    WorldUp=(Target=(X=0,Y=0,Z=1),Kind=Location,Space=(Type=Bone,Name="calf_r"))
   )
   > &RigUnit_ModifyTransforms(
    ItemToModify=((Item=(Type=Bone,Name="deform_skirt_f_l"),Transform=(Rotation=(X=0,Y=0,Z=0.0436,W=0.999),Translation=(X=0,Y=5,Z=0),Scale3D=(X=1,Y=1,Z=1))))
    WeightMinimum=0
    Weight=@n1.OutputParam
   )
   > &RigUnit_ModifyTransforms(
    ItemToModify=((Item=(Type=Bone,Name="deform_skirt_f_r"),Transform=(Rotation=(X=0,Y=0,Z=0.0436,W=0.999),Translation=(X=0,Y=-5,Z=0),Scale3D=(X=1,Y=1,Z=1))))
    WeightMinimum=0
    Weight=@n2.OutputParam
   )
   > &RigUnit_AimConstraintLocalSpaceOffset(
    AdvancedSettings=(DebugSettings=(bEnabled=true,Scale=10,WorldOffset=(Rotation=(X=0,Y=0,Z=0,W=1),Translation=(X=0,Y=0,Z=0),Scale3D=(X=1,Y=1,Z=1))),RotationOrderForFilter=XZY)
    AimAxis=(X=-1,Y=0,Z=0)
    Child=(Type=Bone,Name="deform_skirt_m_l")
    Filter=(bX=True,bY=True,bZ=True)
    Parents=((Item=(Type=Bone,Name="calf_l"),Weight=1))
    UpAxis=(X=0,Y=1,Z=0)
    Weight=1
    WorldUp=(Target=(X=0,Y=0,Z=1),Kind=Location,Space=(Type=Bone,Name="calf_l"))
   )
   > &RigUnit_AimConstraintLocalSpaceOffset(
    AdvancedSettings=(DebugSettings=(bEnabled=true,Scale=10,WorldOffset=(Rotation=(X=0,Y=0,Z=0,W=1),Translation=(X=0,Y=0,Z=0),Scale3D=(X=1,Y=1,Z=1))),RotationOrderForFilter=XZY)
    Child=(Type=Bone,Name="deform_skirt_m_r")
    Filter=(bX=True,bY=True,bZ=True)
    Parents=((Item=(Type=Bone,Name="calf_r"),Weight=1))
    UpAxis=(X=0,Y=-1,Z=0)
    Weight=1
    WorldUp=(Target=(X=0,Y=0,Z=1),Kind=Location,Space=(Type=Bone,Name="pelvis"))
   )
   > &RigUnit_ModifyTransforms(
    ItemToModify=((Item=(Type=Bone,Name="deform_skirt_m_l"),Transform=(Rotation=(X=0,Y=0,Z=0,W=1),Translation=(X=-5,Y=5,Z=0),Scale3D=(X=1,Y=1,Z=1))))
    WeightMinimum=0
    Weight=@n1.OutputParam
   )
   > &RigUnit_ModifyTransforms(
    ItemToModify=((Item=(Type=Bone,Name="deform_skirt_m_r"),Transform=(Rotation=(X=0,Y=0,Z=0,W=1),Translation=(X=0,Y=-2,Z=0),Scale3D=(X=1,Y=1,Z=1))))
    WeightMinimum=0
    Weight=@n2.OutputParam
   )
   > &RigUnit_RotationConstraintLocalSpaceOffset(
    AdvancedSettings=(InterpolationType=Average,RotationOrderForFilter=XZY)
    Child=(Type=Bone,Name="deform_skirt_b_l")
    Filter=(bX=True,bY=True,bZ=True)
    Parents=((Item=(Type=Bone,Name="deform_skirt_m_l"),Weight=1))
    Weight=0.2
   )
   > &RigUnit_RotationConstraintLocalSpaceOffset(
    AdvancedSettings=(InterpolationType=Average,RotationOrderForFilter=XZY)
    Child=(Type=Bone,Name="deform_skirt_b_r")
    Filter=(bX=True,bY=True,bZ=True)
    Parents=((Item=(Type=Bone,Name="deform_skirt_m_r"),Weight=1))
    Weight=0.2
   )
   RigUnit_BeginExecution = &RigUnit_BeginExecution()
   FN_DeformRig_FNC_1 = RigVMFunctionReferenceNode(
    DeformRig_ Curve="(deformRig_l_upperarm_up_alpha_72_#1=False,deformRig_r_upperarm_up_alpha_73_#2=False,deformRig_l_upperarm_fr_alpha_74_#3=False,deformRig_r_upperarm_fr_alpha_75_CB51B8 ...(truncated)"
    DeformRig_Bool="(upperarm_1_#4=True,elbow_3_#5=True,wrist_7_#6=false,groin_8_#7=True,glute_17_#8=false,knee_9_40C408324CA2 ...(truncated)"
    DeformRig_Debug=(upperarm_debug_18_#4=False,wrist_debug_20_#6=False,lat_debug_26_#9=False,groinGlute_debug_27_#7=False)
    DeformRig_Value="(upperarm_up_18_#10=(Rotation=(X=0,Y=0,Z=0,W=1),Translation=(X=-1,Y=0,Z=2),Scale3D=(X=0.7,Y=1,Z=1)),upperarm_fwd_fr_19_#11=(Rotation=(X=0,Y=0,Z=0,W=1),Translation=(X=-5,Y=15,Z=0),Scale3D=(X=0.5,Y=1, ...(truncated)"
    ExecuteContext=@RigUnit_BeginExecution.ExecutePin
   )
   > RigVMFunction_Sequence = RigVMAggregateNode()
   &RigUnit_ParentConstraint(
    AdvancedSettings=(RotationOrderForFilter=XZY)
    Child=(Type=Bone,Name="dyn_strap_c_05")
    Filter=(TranslationFilter=(bX=True,bY=True,bZ=True),RotationFilter=(bX=false,bY=false,bZ=false),ScaleFilter=(bX=True,bY=True,bZ=True))
    Parents="((Item=(Type=Bone,Name=\"deform_upperarm_bk_r\"),Weight=0.1),(Item=(Type=Bone,Name=\"spine_02\"),Weight=0.0396),(Item=(Type=Bone,Name=\"spine_05\"),Weight=0.2455),(Item=(Type=Bone,Name=\"spine_04\"),Weight=0.3096),(Item=(Type=Bone,Name=\"spine_03\"),Weight=0.3952),( ...(truncated)"
    Weight=1
    ExecutePin=@RigVMFunction_Sequence.A
   )
   > &RigUnit_ParentConstraint(
    AdvancedSettings=(RotationOrderForFilter=XZY)
    Child=(Type=Bone,Name="dyn_strap_c_03")
    Filter=(TranslationFilter=(bX=True,bY=True,bZ=True),RotationFilter=(bX=false,bY=false,bZ=false),ScaleFilter=(bX=True,bY=True,bZ=True))
    Parents="((Item=(Type=Bone,Name=\"upperarm_twist_02_r\"),Weight=0),(Item=(Type=Bone,Name=\"spine_04\"),Weight=0.0157),(Item=(Type=Bone,Name=\"deform_upperarm_fr_r\"),Weight=0.1),(Item=(Type=Bone,Name=\"spine_03\"),Weight=0.036),(Item=(Type=Bone,Name=\"upperarm_twist_01_r\"), ...(truncated)"
    Weight=1
   )
   > /Script/ControlRig.RigUnit_PositionConstraintLocalSpaceOffset(
    Child=(Type=Bone,Name="dyn_strap_c_04")
    Filter=(bX=True,bY=True,bZ=True)
    Parents=((Item=(Type=Bone,Name="dyn_strap_c_03"),Weight=1),(Item=(Type=Bone,Name="dyn_strap_c_05"),Weight=0.8))
    Weight=1
    bMaintainOffset=True
   ) {is=struct RigVMUnitNode}
   > &RigUnit_AimConstraintLocalSpaceOffset(
    Child=(Type=Bone,Name="dyn_strap_c_05")
    Filter=(bX=True,bY=True,bZ=True)
    Parents=((Item=(Type=Bone,Name="None"),Weight=1))
    Weight=1
    WorldUp=(Target=(X=0,Y=0,Z=1),Kind=Direction,Space=(Type=Bone,Name="dyn_strap_c_03"))
   )
   > &RigUnit_AimConstraintLocalSpaceOffset(
    Child=(Type=Bone,Name="dyn_strap_c_04")
    Filter=(bX=True,bY=True,bZ=True)
    Parents=((Item=(Type=Bone,Name="None"),Weight=1))
    Weight=1
    WorldUp=(Target=(X=0,Y=0,Z=1),Kind=Direction,Space=(Type=Bone,Name="dyn_strap_c_05"))
   )
   > &RigUnit_ParentConstraint(
    AdvancedSettings=(RotationOrderForFilter=XZY)
    Child=(Type=Bone,Name="dyn_strap_c_02")
    Parents="((Item=(Type=Bone,Name=\"upperarm_twist_01_r\"),Weight=0.0257),(Item=(Type=Bone,Name=\"deform_upperarm_up_r\"),Weight=0.0277),(Item=(Type=Bone,Name=\"deform_upperarm_fr_r\"),Weight=0.2),(Item=(Type=Bone,Name=\"spine_04\"),Weight=0.0623),(Item=(Type=Bone,Name=\"dyn_ ...(truncated)"
    Weight=1
   )
   > &RigUnit_ParentConstraint(
    AdvancedSettings=(RotationOrderForFilter=XZY)
    Child=(Type=Bone,Name="dyn_strap_c_01")
    Parents="((Item=(Type=Bone,Name=\"upperarm_twist_01_r\"),Weight=0.0103),(Item=(Type=Bone,Name=\"deform_upperarm_fr_r\"),Weight=0.0129),(Item=(Type=Bone,Name=\"spine_02\"),Weight=0.0493),(Item=(Type=Bone,Name=\"spine_03\"),Weight=0.1224),(Item=(Type=Bone,Name=\"clavicle_r\"), ...(truncated)"
    Weight=1
   )
   > &RigUnit_ParentConstraint(
    AdvancedSettings=(RotationOrderForFilter=XZY)
    Child=(Type=Bone,Name="deform_collar_f_l")
    Parents="((Item=(Type=Bone,Name=\"dyn_strap_c_02\"),Weight=0),(Item=(Type=Bone,Name=\"neck_01\"),Weight=0.0204),(Item=(Type=Bone,Name=\"deform_pec_l\"),Weight=0.0316),(Item=(Type=Bone,Name=\"spine_03\"),Weight=0.0331),(Item=(Type=Bone,Name=\"neck_02\"),Weight=0.0503),(Item=( ...(truncated)"
    Weight=1
   )
   > &RigUnit_ParentConstraint(
    AdvancedSettings=(RotationOrderForFilter=XZY)
    Child=(Type=Bone,Name="deform_collar_l")
    Parents="((Item=(Type=Bone,Name=\"neck_01\"),Weight=0.0161),(Item=(Type=Bone,Name=\"neck_02\"),Weight=0.0402),(Item=(Type=Bone,Name=\"clavicle_l\"),Weight=0.3402),(Item=(Type=Bone,Name=\"spine_05\"),Weight=0.4),(Item=(Type=Bone,Name=\"dyn_strap_c_02\"),Weight=0),(Item=(Type= ...(truncated)"
    Weight=1
   )
   > &RigUnit_ParentConstraint(
    AdvancedSettings=(RotationOrderForFilter=XZY)
    Child=(Type=Bone,Name="deform_collar_b")
    Parents=((Item=(Type=Bone,Name="head"),Weight=0.0533),(Item=(Type=Bone,Name="neck_01"),Weight=0.1031),(Item=(Type=Bone,Name="neck_02"),Weight=0.2309),(Item=(Type=Bone,Name="spine_05"),Weight=0.6128))
    Weight=1
   )
   RigVMFunction_MathBoolToFloat = &RigVMFunction_MathBoolToFloat(Value=RigVMVariableNode() {variable=is_front_end_not_emoting}.Value)
   n3 = &RigVMFunction_AlphaInterp(
    ClampMin=0.35
    InRange=(Min=1,Max=0)
    InterpSpeedDecreasing=8
    InterpSpeedIncreasing=15
    OutRange=(Min=0.3,Max=1)
    Scale=1
    bClampResult=false
    bMapRange=true
    Value=@RigVMFunction_MathBoolToFloat.Result
   )
   > &RigUnit_ParentConstraint(
    AdvancedSettings=(InterpolationType=Average,RotationOrderForFilter=XZY)
    Child=(Type=Bone,Name="dyn_skirtA_l_MSTR")
    Parents=((Item=(Type=Bone,Name="thigh_l"),Weight=1))
    Weight=1
    bMaintainOffset=true
    Parents.0.Weight=@n3.Result
   )
   > &RigUnit_ModifyTransforms(
    ItemToModify=((Item=(Type=Bone,Name="dyn_skirtA_l_01"),Transform=(Rotation=(X=0.0905,Y=0.083,Z=-0.0509,W=0.9911),Translation=(X=0,Y=0,Z=0),Scale3D=(X=1,Y=1,Z=1))))
    WeightMinimum=0
    Weight=@n3.Result
   )
   > &RigUnit_ParentConstraint(
    AdvancedSettings=(InterpolationType=Average,RotationOrderForFilter=XZY)
    Child=(Type=Bone,Name="dyn_skirtB_r_MSTR")
    Filter=(TranslationFilter=(bX=false,bY=false,bZ=false),RotationFilter=(bX=true,bY=true,bZ=true),ScaleFilter=(bX=True,bY=True,bZ=True))
    Parents=((Item=(Type=Bone,Name="thigh_twist_01_r"),Weight=1),(Item=(Type=Bone,Name="pelvis"),Weight=1))
    bMaintainOffset=true
    Weight=@n3.Result
   )
   > &RigUnit_ModifyTransforms(
    ItemToModify=((Item=(Type=Bone,Name="dyn_skirtB_r_MSTR"),Transform=(Rotation=(X=0,Y=0,Z=0,W=1),Translation=(X=0,Y=-3,Z=0),Scale3D=(X=1,Y=1,Z=1))))
    WeightMinimum=0
    Weight=@n3.Result
   )
   > &RigUnit_ParentConstraint(
    AdvancedSettings=(RotationOrderForFilter=XZY)
    Child=(Type=Bone,Name="dyn_bottleA_MSTR")
    Parents=((Item=(Type=Bone,Name="spine_03"),Weight=0.0927),(Item=(Type=Bone,Name="spine_01"),Weight=0.1378),(Item=(Type=Bone,Name="spine_02"),Weight=0.1654),(Item=(Type=Bone,Name="pelvis"),Weight=0.6042))
    Weight=1
   )
   > n4 = &RigUnit_ParentConstraint(
    AdvancedSettings=(RotationOrderForFilter=XZY)
    Child=(Type=Bone,Name="dyn_bottleB_MSTR")
    Parents=((Item=(Type=Bone,Name="spine_04"),Weight=0.0005),(Item=(Type=Bone,Name="spine_03"),Weight=0.0976),(Item=(Type=Bone,Name="spine_01"),Weight=0.1424),(Item=(Type=Bone,Name="spine_02"),Weight=0.1703),(Item=(Type=Bone,Name="pelvis"),Weight=0.5891))
    Weight=1
   )
   RigVMFunction_ControlFlowBranch = &RigVMFunction_ControlFlowBranch(ExecuteContext=@n4.ExecutePin, Condition=RigVMVariableNode() {variable=is_front_end}.Value)
   &RigUnit_ModifyTransforms(
    ItemToModify="((Item=(Type=Bone,Name=\"dyn_bottleA_MSTR\"),Transform=(Rotation=(X=0.0151,Y=0.173,Z=-0.0858,W=0.9811),Translation=(X=0,Y=0,Z=0),Scale3D=(X=1,Y=1,Z=1))),(Item=(Type=Bone,Name=\"dyn_bottleB_MSTR\"),Transform=(Rotation=(X=0,Y=0.1736,Z=0,W=0.9848),Translation=(X= ...(truncated)"
    WeightMinimum=0
    ExecutePin=@RigVMFunction_ControlFlowBranch.False
    Weight=RigVMVariableNode() {variable=left_thigh_up}.Value
   )
   n5 = &RigUnit_SphericalPoseReader(
    ActiveRegionScaleFactors=(PositiveWidth=1,NegativeWidth=1,PositiveHeight=1,NegativeHeight=1)
    ActiveRegionSize=0.15
    Debug=(bDrawDebug=false,DebugScale=25,DebugSegments=20,DebugThickness=0.25)
    DriverAxis=(X=1,Y=0,Z=0)
    DriverItem=(Type=Bone,Name="deform_skirt_bk_r")
    FalloffRegionScaleFactors=(PositiveWidth=1,NegativeWidth=1,PositiveHeight=1,NegativeHeight=1)
    FalloffSize=0.45
    FlipHeightScaling=false
    FlipWidthScaling=true
    OptionalParentItem=(Type=Bone,Name="thigh_twist_01_r")
    RotationOffset=(X=0,Y=-100,Z=20)
    ExecutePin=@RigVMFunction_Sequence.C
   )
   > &RigUnit_AimConstraintLocalSpaceOffset(
    AdvancedSettings=(DebugSettings=(bEnabled=true,Scale=10,WorldOffset=(Rotation=(X=0,Y=0,Z=0,W=1),Translation=(X=0,Y=0,Z=0),Scale3D=(X=1,Y=1,Z=1))),RotationOrderForFilter=XZY)
    Child=(Type=Bone,Name="deform_skirt_bk_r")
    Filter=(bX=True,bY=True,bZ=True)
    Parents=((Item=(Type=Bone,Name="calf_r"),Weight=1))
    UpAxis=(X=0,Y=-1,Z=0)
    WorldUp=(Target=(X=0,Y=0,Z=1),Kind=Location,Space=(Type=Bone,Name="pelvis"))
    Weight=@n5.OutputParam
   )
   > &RigUnit_AimConstraintLocalSpaceOffset(
    AdvancedSettings=(DebugSettings=(bEnabled=true,Scale=10,WorldOffset=(Rotation=(X=0,Y=0,Z=0,W=1),Translation=(X=0,Y=0,Z=0),Scale3D=(X=1,Y=1,Z=1))),RotationOrderForFilter=XZY)
    Child=(Type=Bone,Name="deform_skirt_b_r")
    Filter=(bX=True,bY=True,bZ=True)
    Parents=((Item=(Type=Bone,Name="calf_r"),Weight=1))
    UpAxis=(X=0,Y=-1,Z=0)
    WorldUp=(Target=(X=0,Y=0,Z=1),Kind=Location,Space=(Type=Bone,Name="pelvis"))
    Weight=@n5.OutputParam
   )
   > n6 = &RigUnit_SphericalPoseReader(
    ActiveRegionScaleFactors=(PositiveWidth=1,NegativeWidth=1,PositiveHeight=1,NegativeHeight=1)
    ActiveRegionSize=0.15
    Debug=(bDrawDebug=false,DebugScale=25,DebugSegments=20,DebugThickness=0.25)
    DriverAxis=(X=-1,Y=0,Z=0)
    DriverItem=(Type=Bone,Name="deform_skirt_bk_l")
    FalloffRegionScaleFactors=(PositiveWidth=1,NegativeWidth=1,PositiveHeight=1,NegativeHeight=1)
    FalloffSize=0.45
    FlipHeightScaling=false
    FlipWidthScaling=true
    OptionalParentItem=(Type=Bone,Name="thigh_twist_01_l")
    RotationOffset=(X=0,Y=60,Z=30)
   )
   > &RigUnit_AimConstraintLocalSpaceOffset(
    AdvancedSettings=(DebugSettings=(bEnabled=true,Scale=10,WorldOffset=(Rotation=(X=0,Y=0,Z=0,W=1),Translation=(X=0,Y=0,Z=0),Scale3D=(X=1,Y=1,Z=1))),RotationOrderForFilter=XZY)
    AimAxis=(X=-1,Y=0,Z=0)
    Child=(Type=Bone,Name="deform_skirt_bk_l")
    Filter=(bX=True,bY=True,bZ=True)
    Parents=((Item=(Type=Bone,Name="calf_l"),Weight=1))
    UpAxis=(X=0,Y=1,Z=0)
    WorldUp=(Target=(X=0,Y=0,Z=-1),Kind=Location,Space=(Type=Bone,Name="pelvis"))
    bMaintainOffset=true
    Weight=@n6.OutputParam
   )
   > &RigUnit_AimConstraintLocalSpaceOffset(
    AdvancedSettings=(DebugSettings=(bEnabled=true,Scale=10,WorldOffset=(Rotation=(X=0,Y=0,Z=0,W=1),Translation=(X=0,Y=0,Z=0),Scale3D=(X=1,Y=1,Z=1))),RotationOrderForFilter=XZY)
    AimAxis=(X=-1,Y=0,Z=0)
    Child=(Type=Bone,Name="deform_skirt_b_l")
    Filter=(bX=True,bY=True,bZ=True)
    Parents=((Item=(Type=Bone,Name="calf_l"),Weight=1))
    UpAxis=(X=0,Y=1,Z=0)
    WorldUp=(Target=(X=0,Y=0,Z=-1),Kind=Location,Space=(Type=Bone,Name="deform_skirt_b_l"))
    Weight=@n6.OutputParam
   )
   > &RigUnit_AimConstraintLocalSpaceOffset(
    AdvancedSettings=(DebugSettings=(bEnabled=true,Scale=10,WorldOffset=(Rotation=(X=0,Y=0,Z=0,W=1),Translation=(X=0,Y=0,Z=0),Scale3D=(X=1,Y=1,Z=1))),RotationOrderForFilter=XZY)
    Child=(Type=Bone,Name="deform_skirt_bk_r1")
    Filter=(bX=True,bY=True,bZ=True)
    Parents=((Item=(Type=Bone,Name="dyn_skirt_bk_3_l"),Weight=1))
    UpAxis=(X=0,Y=-1,Z=0)
    Weight=1
    WorldUp=(Target=(X=0,Y=0,Z=1),Kind=Location,Space=(Type=Bone,Name="deform_skirt_bk_l"))
   )
   n7 = &RigUnit_SphericalPoseReader(
    ActiveRegionScaleFactors=(PositiveWidth=3.5,NegativeWidth=0.5,PositiveHeight=3,NegativeHeight=3)
    ActiveRegionSize=0.1
    Debug=(bDrawDebug=false,DebugScale=25,DebugSegments=20,DebugThickness=0.25)
    DriverAxis=(X=-1,Y=0,Z=0)
    DriverItem=(Type=Bone,Name="thigh_l")
    FalloffRegionScaleFactors=(PositiveWidth=1,NegativeWidth=1,PositiveHeight=1,NegativeHeight=1)
    FalloffSize=0.45
    OptionalParentItem=(Type=Bone,Name="pelvis")
    RotationOffset=(X=90,Y=0,Z=0)
    ExecutePin=@RigVMFunction_Sequence.D
   )
   > RigVMVariableNode(Value=@n7.OutputParam) {variable=left_thigh_up}
   n8 = RigVMDispatchNode(Condition=False, True=&RigVMFunction_MathDoubleMul(B=0.8, A=@n7.OutputParam).Result, False=@n7.OutputParam) {template="DISPATCH_RigVMDispatch_If(in Condition,in True,in False,out Result)"}
   > &RigUnit_ModifyTransforms(
    ItemToModify=((Item=(Type=Bone,Name="dyn_skirtA_l_MSTR"),Transform=(Rotation=(X=0,Y=0.2588,Z=0,W=0.9659),Translation=(X=10,Y=3,Z=0),Scale3D=(X=1,Y=1,Z=1))))
    WeightMinimum=0
    Weight=@n8.Result
   )
   > &RigUnit_ModifyTransforms(
    ItemToModify=((Item=(Type=Bone,Name="dyn_skirtA_l_02"),Transform=(Rotation=(X=0,Y=0,Z=0.1736,W=0.9848),Translation=(X=0,Y=0,Z=0),Scale3D=(X=1,Y=1,Z=1))))
    WeightMinimum=0
    Weight=@n8.Result
   )
   > &RigUnit_ModifyTransforms(
    ItemToModify=((Item=(Type=Bone,Name="dyn_skirtB_l_MSTR"),Transform=(Rotation=(X=0.2685,Y=0.0381,Z=-0.1893,W=0.9437),Translation=(X=0,Y=0,Z=0),Scale3D=(X=1,Y=1,Z=1))))
    WeightMinimum=0
    Weight=@n8.Result
   )
   > n9 = &RigUnit_SphericalPoseReader(
    ActiveRegionScaleFactors=(PositiveWidth=3.5,NegativeWidth=0.1,PositiveHeight=3,NegativeHeight=3)
    ActiveRegionSize=0.1
    Debug=(bDrawDebug=false,DebugScale=25,DebugSegments=20,DebugThickness=0.25)
    DriverAxis=(X=1,Y=0,Z=0)
    DriverItem=(Type=Bone,Name="thigh_r")
    FalloffRegionScaleFactors=(PositiveWidth=1,NegativeWidth=1.2,PositiveHeight=1,NegativeHeight=1)
    FalloffSize=0.45
    FlipHeightScaling=false
    FlipWidthScaling=true
    OptionalParentItem=(Type=Bone)
    RotationOffset=(X=-70,Y=-15,Z=10)
   )
   > RigVMVariableNode(Value=@n9.OutputParam) {variable=right_thigh_up}
   > &RigUnit_ModifyTransforms(
    ItemToModify=((Item=(Type=Bone,Name="dyn_skirtA_r_MSTR"),Transform=(Rotation=(X=0.1422,Y=-0.0996,Z=-0.5649,W=0.8067),Translation=(X=0,Y=0,Z=0),Scale3D=(X=0.7,Y=1,Z=1))))
    WeightMinimum=0.3
    Weight=@n9.OutputParam
   )
   > &RigUnit_ModifyTransforms(
    ItemToModify=((Item=(Type=Bone,Name="dyn_skirtB_r_MSTR"),Transform=(Rotation=(X=0,Y=0.0872,Z=0,W=0.9962),Translation=(X=-4,Y=2,Z=0),Scale3D=(X=0.8,Y=1,Z=1))))
    WeightMinimum=0
    Weight=@n9.OutputParam
   )
   > &RigUnit_ModifyTransforms(
    ItemToModify=((Item=(Type=Bone,Name="dyn_ropeA_l_MSTR"),Transform=(Rotation=(X=0,Y=0,Z=-0.766,W=0.6428),Translation=(X=3,Y=1,Z=0),Scale3D=(X=1,Y=1,Z=1))))
    WeightMinimum=0
    Weight=@n9.OutputParam
   )
   > &RigUnit_ModifyTransforms(
    ItemToModify=((Item=(Type=Bone,Name="dyn_ropeB_l_MSTR"),Transform=(Rotation=(X=0,Y=0,Z=-0.766,W=0.6428),Translation=(X=1,Y=0,Z=0),Scale3D=(X=1,Y=1,Z=1))))
    WeightMinimum=0
    Weight=@n9.OutputParam
   )
   > &RigUnit_ModifyTransforms(
    ItemToModify=((Item=(Type=Bone,Name="dyn_skirtTopA_l_MSTR"),Transform=(Rotation=(X=0.0302,Y=0.171,Z=-0.171,W=0.9698),Translation=(X=1,Y=0,Z=0),Scale3D=(X=1,Y=1,Z=1))))
    WeightMinimum=0
    Weight=@n9.OutputParam
   )
   > n10 = &RigUnit_SphericalPoseReader(
    ActiveRegionScaleFactors=(PositiveWidth=3.5,NegativeWidth=0.1,PositiveHeight=3,NegativeHeight=3)
    ActiveRegionSize=0.1
    Debug=(bDrawDebug=false,DebugScale=25,DebugSegments=20,DebugThickness=0.25)
    DriverAxis=(X=1,Y=0,Z=0)
    DriverItem=(Type=Bone,Name="thigh_r")
    FalloffRegionScaleFactors=(PositiveWidth=1,NegativeWidth=0.7,PositiveHeight=1,NegativeHeight=1)
    FalloffSize=0.45
    FlipHeightScaling=false
    FlipWidthScaling=true
    OptionalParentItem=(Type=Bone)
    RotationOffset=(X=0,Y=-10,Z=-70)
   )
   > &RigUnit_ModifyTransforms(
    ItemToModify="((Item=(Type=Bone,Name=\"dyn_skirtC_r_MSTR\"),Transform=(Rotation=(X=0,Y=0,Z=-0.2164,W=0.9763),Translation=(X=0,Y=0,Z=0),Scale3D=(X=1,Y=1,Z=1))),(Item=(Type=Bone,Name=\"dyn_skirtB_r_MSTR\"),Transform=(Rotation=(X=-0.0227,Y=-0.1722,Z=-0.1285,W=0.9764),Translati ...(truncated)"
    WeightMinimum=0
    Weight=@n10.OutputParam
   )
   n11 = &RigUnit_SphericalPoseReader(
    ActiveRegionScaleFactors=(PositiveWidth=4,NegativeWidth=4,PositiveHeight=2,NegativeHeight=0.1)
    ActiveRegionSize=0.05
    Debug=(bDrawDebug=false,bDraw2D=False,bDrawLocalAxes=False,DebugScale=25,DebugSegments=20,DebugThickness=0.25)
    DriverAxis=(X=1,Y=0,Z=0)
    DriverItem=(Type=Bone,Name="head")
    FalloffRegionScaleFactors=(PositiveWidth=1,NegativeWidth=1,PositiveHeight=1,NegativeHeight=1)
    FalloffSize=0.3
    FlipHeightScaling=False
    FlipWidthScaling=False
    OptionalParentItem=(Type=Bone,Name="neck_01")
    RotationOffset=(X=0,Y=-90,Z=40)
    ExecutePin=@RigVMFunction_Sequence.B
   )
   > RigVMVariableNode(Value=@n11.OutputParam) {variable=head_fwd}
   > n12 = &RigUnit_SphericalPoseReader(
    ActiveRegionScaleFactors=(PositiveWidth=4,NegativeWidth=4,PositiveHeight=0.1,NegativeHeight=2)
    ActiveRegionSize=0.05
    Debug=(bDrawDebug=false,bDraw2D=False,bDrawLocalAxes=False,DebugScale=25,DebugSegments=20,DebugThickness=0.25)
    DriverAxis=(X=1,Y=0,Z=0)
    DriverItem=(Type=Bone,Name="head")
    FalloffRegionScaleFactors=(PositiveWidth=1.5,NegativeWidth=1.5,PositiveHeight=1,NegativeHeight=0.2)
    FalloffSize=0.3
    FlipHeightScaling=False
    FlipWidthScaling=False
    OptionalParentItem=(Type=Bone,Name="neck_01")
    RotationOffset=(X=0,Y=-90,Z=-50)
   )
   > RigVMVariableNode(Value=@n12.OutputParam) {variable=head_bck}
   > n13 = &RigUnit_SphericalPoseReader(
    ActiveRegionScaleFactors=(PositiveWidth=0,NegativeWidth=2,PositiveHeight=4,NegativeHeight=4)
    ActiveRegionSize=0.05
    Debug=(bDrawDebug=false,bDraw2D=False,bDrawLocalAxes=false,DebugScale=25,DebugSegments=20,DebugThickness=0.25)
    DriverAxis=(X=1,Y=0,Z=0)
    DriverItem=(Type=Bone,Name="head")
    FalloffRegionScaleFactors=(PositiveWidth=1,NegativeWidth=0.2,PositiveHeight=1,NegativeHeight=1)
    FalloffSize=0.3
    FlipHeightScaling=false
    FlipWidthScaling=false
    OptionalParentItem=(Type=Bone,Name="spine_05")
    RotationOffset=(X=0,Y=-30,Z=-30)
   )
   > RigVMVariableNode(Value=@n13.OutputParam) {variable=head_rgt}
   > n14 = &RigUnit_SphericalPoseReader(
    ActiveRegionScaleFactors=(PositiveWidth=0,NegativeWidth=2,PositiveHeight=4,NegativeHeight=4)
    ActiveRegionSize=0.05
    Debug=(bDrawDebug=true,bDraw2D=False,bDrawLocalAxes=False,DebugScale=25,DebugSegments=20,DebugThickness=0.25)
    DriverAxis=(X=1,Y=0,Z=0)
    DriverItem=(Type=Bone,Name="head")
    FalloffRegionScaleFactors=(PositiveWidth=1,NegativeWidth=0.2,PositiveHeight=1,NegativeHeight=1)
    FalloffSize=0.3
    FlipHeightScaling=false
    FlipWidthScaling=false
    OptionalParentItem=(Type=Bone,Name="spine_05")
    RotationOffset=(X=0,Y=-210,Z=-210)
   )
   > RigVMVariableNode(Value=@n14.OutputParam) {variable=head_lft}
   > n15 = &RigUnit_SphericalPoseReader(
    ActiveRegionScaleFactors=(PositiveWidth=5,NegativeWidth=5,PositiveHeight=3,NegativeHeight=0.5)
    ActiveRegionSize=0.05
    Debug=(bDrawDebug=false,bDraw2D=False,bDrawLocalAxes=False,DebugScale=25,DebugSegments=20,DebugThickness=0.25)
    DriverAxis=(X=0,Y=1,Z=0)
    DriverItem=(Type=Bone,Name="head")
    FalloffRegionScaleFactors=(PositiveWidth=2,NegativeWidth=2,PositiveHeight=1,NegativeHeight=1.2)
    FalloffSize=0.3
    FlipHeightScaling=false
    FlipWidthScaling=false
    OptionalParentItem=(Type=Bone,Name="spine_05")
    RotationOffset=(X=170,Y=0,Z=10)
   )
   > RigVMVariableNode(Value=@n15.OutputParam) {variable=head_twist_lf}
   > n16 = &RigUnit_SphericalPoseReader(
    ActiveRegionScaleFactors=(PositiveWidth=5,NegativeWidth=5,PositiveHeight=0.5,NegativeHeight=3)
    ActiveRegionSize=0.05
    Debug=(bDrawDebug=false,bDraw2D=False,bDrawLocalAxes=False,DebugScale=25,DebugSegments=20,DebugThickness=0.25)
    DriverAxis=(X=0,Y=1,Z=0)
    DriverItem=(Type=Bone,Name="head")
    FalloffRegionScaleFactors=(PositiveWidth=2,NegativeWidth=2,PositiveHeight=1.2,NegativeHeight=1)
    FalloffSize=0.3
    FlipHeightScaling=false
    FlipWidthScaling=false
    OptionalParentItem=(Type=Bone,Name="spine_05")
    RotationOffset=(X=50,Y=0,Z=-10)
   )
   > RigVMVariableNode(Value=@n16.OutputParam) {variable=head_twist_rgt}
   Multiply = &RigVMFunction_MathDoubleMul(B=1, A=RigVMVariableNode() {variable=head_fwd}.Value)
   n17 = &RigVMFunction_MathDoubleAdd(A=@Multiply.Result, B=&RigVMFunction_MathDoubleMul(B=1, A=RigVMVariableNode() {variable=head_lft}.Value).Result)
   > &RigUnit_ModifyTransforms(
    ItemToModify=((Item=(Type=Bone,Name="deform_collar_f_l"),Transform=(Rotation=(X=0,Y=0.342,Z=0,W=0.9397),Translation=(X=0,Y=0,Z=0),Scale3D=(X=1,Y=1,Z=1))))
    WeightMinimum=0
    Weight=@n17.Result
   )
   > &RigUnit_ModifyTransforms(
    ItemToModify=((Item=(Type=Bone,Name="deform_collar_l"),Transform=(Rotation=(X=0,Y=0.342,Z=0,W=0.9397),Translation=(X=0,Y=0,Z=0),Scale3D=(X=1,Y=1,Z=1))))
    WeightMinimum=0
    Weight=RigVMVariableNode() {variable=head_twist_lf}.Value
   )
   > &RigUnit_ModifyTransforms(
    ItemToModify="((Item=(Type=Bone,Name=\"deform_collar_b\"),Transform=(Rotation=(X=0,Y=0,Z=0,W=1),Translation=(X=0,Y=0,Z=0),Scale3D=(X=1,Y=1,Z=1))),(Item=(Type=Bone,Name=\"deform_collar_r\"),Transform=(Rotation=(X=0,Y=0,Z=-0.4226,W=0.9063),Translation=(X=1.5,Y=0,Z=0),Scale3D= ...(truncated)"
    WeightMinimum=0.3
    Weight=RigVMVariableNode() {variable=head_bck}.Value
   )
   > &RigUnit_ModifyTransforms(
    ItemToModify="((Item=(Type=Bone,Name=\"deform_collar_b\"),Transform=(Rotation=(X=-0.0872,Y=0,Z=0,W=0.9962),Translation=(X=0,Y=0,Z=0),Scale3D=(X=1,Y=1,Z=1))),(Item=(Type=Bone,Name=\"deform_collar_r\"),Transform=(Rotation=(X=0,Y=0,Z=-0.1736,W=0.9848),Translation=(X=0,Y=0,Z=0) ...(truncated)"
    WeightMinimum=0
    Weight=RigVMVariableNode() {variable=head_rgt}.Value
   )
   > n18 = &RigUnit_SphericalPoseReader(
    ActiveRegionScaleFactors=(PositiveWidth=5,NegativeWidth=5,PositiveHeight=1,NegativeHeight=1)
    ActiveRegionSize=0.1
    Debug=(bDrawDebug=false,DebugScale=25,DebugSegments=20,DebugThickness=0.25)
    DriverAxis=(X=1,Y=0,Z=0)
    DriverItem=(Type=Bone,Name="deform_collar_b")
    FalloffRegionScaleFactors=(PositiveWidth=3,NegativeWidth=3,PositiveHeight=1,NegativeHeight=1)
    FalloffSize=0.6
    FlipHeightScaling=false
    FlipWidthScaling=true
    OptionalParentItem=(Type=Bone,Name="head")
    RotationOffset=(X=180,Y=90,Z=100)
   )
   > &RigUnit_AimConstraintLocalSpaceOffset(
    Child=(Type=Bone,Name="deform_collar_b")
    Filter=(bX=True,bY=True,bZ=True)
    Parents=((Item=(Type=Bone,Name="head"),Weight=1))
    UpAxis=(X=0,Y=-1,Z=0)
    Weight=0.3
    WorldUp=(Target=(X=0,Y=0,Z=1),Kind=Location,Space=(Type=Bone,Name="head"))
   )
   > &RigUnit_AimConstraintLocalSpaceOffset(
    Child=(Type=Bone,Name="deform_collar_l")
    Filter=(bX=True,bY=True,bZ=True)
    Parents=((Item=(Type=Bone,Name="head"),Weight=1))
    UpAxis=(X=0,Y=0,Z=-1)
    Weight=0.6
    WorldUp=(Target=(X=0,Y=0,Z=1),Kind=Location,Space=(Type=Bone,Name="head"))
   )
   > &RigUnit_ModifyTransforms(
    ItemToModify=((Item=(Type=Bone,Name="deform_collar_b"),Transform=(Rotation=(X=0,Y=0,Z=-0.7071,W=0.7071),Translation=(X=-1,Y=-1,Z=0),Scale3D=(X=1,Y=1,Z=1.3))))
    WeightMinimum=0
    Weight=@n18.OutputParam
   )
   RigUnit_GetTransform = &RigUnit_GetTransform(Item=(Type=Control,Name="head_ctrl"), bInitial=False)
   n19 = &RigUnit_GetTransform(Item=(Type=Bone,Name="deform_collar_l"), bInitial=False)
   n20 = &RigUnit_GetTransform(Item=(Type=Bone,Name="deform_collar_l"), bInitial=False)
   n21 = &RigUnit_GetTransform(Item=(Type=Control,Name="head_ctrl_2"), bInitial=False)
   Add = &RigVMFunction_MathDoubleAdd(
    A=&RigVMFunction_MathDoubleRemap(
     SourceMaximum=20
     SourceMinimum=8
     TargetMaximum=0
     TargetMinimum=1
     Value=&RigVMFunction_MathVectorDistance(A=@RigUnit_GetTransform.Transform.Translation, B=@n19.Transform.Translation).Result
    ).Result
    B=&RigVMFunction_MathDoubleRemap(
     SourceMaximum=15
     SourceMinimum=0
     TargetMaximum=0
     TargetMinimum=1
     Value=&RigVMFunction_MathVectorDistance(B=@n20.Transform.Translation, A=@n21.Transform.Translation).Result
    ).Result
   )
   > &RigUnit_ModifyTransforms(
    ItemToModify=((Item=(Type=Bone,Name="deform_collar_l"),Transform=(Rotation=(X=0.0842,Y=0.2578,Z=-0.0226,W=0.9623),Translation=(X=0,Y=0,Z=-1),Scale3D=(X=1,Y=1,Z=1))))
    WeightMinimum=0
    Weight=@Add.Result
   )
   &RigUnit_AimConstraintLocalSpaceOffset(
    AimAxis=(X=-1,Y=0,Z=0)
    Child=(Type=Bone,Name="deform_collar_r")
    ExecutePin=()
    Filter=(bX=True,bY=True,bZ=True)
    Parents=((Item=(Type=Bone,Name="head"),Weight=1))
    Weight=0.6
    WorldUp=(Target=(X=0,Y=0,Z=1),Kind=Location,Space=(Type=Bone,Name="head"))
   )
   AlphaInterp = &RigVMFunction_AlphaInterp(InterpSpeedDecreasing=10, InterpSpeedIncreasing=10, Scale=1, Value=0, bClampResult=False, bInterpResult=False)
   RigVMDispatchNode(Condition=False, True=&RigVMFunction_MathDoubleMul(B=0, A=@n9.OutputParam).Result, False=@n9.OutputParam) {template="DISPATCH_RigVMDispatch_If(in Condition,in True,in False,out Result)"}
   RigVMFunction_MathVectorDistance = &RigVMFunction_MathVectorDistance(
    A=@RigUnit_GetTransform.Transform.Translation
    B=&RigUnit_GetTransform(Item=(Type=Bone,Name="deform_collar_r"), bInitial=False).Transform.Translation
   )
   n22 = &RigVMFunction_MathVectorDistance(
    B=&RigUnit_GetTransform(Item=(Type=Bone,Name="deform_collar_r"), bInitial=False).Transform.Translation
    A=@n21.Transform.Translation
   )
   &RigUnit_ModifyTransforms(
    ExecutePin=()
    ItemToModify=((Item=(Type=Bone,Name="deform_collar_r"),Transform=(Rotation=(X=0.1677,Y=0.2549,Z=-0.0449,W=0.9513),Translation=(X=0,Y=0,Z=1),Scale3D=(X=1,Y=1,Z=1))))
    WeightMinimum=0
    Weight=&RigVMFunction_MathDoubleAdd(
     A=&RigVMFunction_MathDoubleRemap(SourceMaximum=20, SourceMinimum=8, TargetMaximum=0, TargetMinimum=1, Value=@RigVMFunction_MathVectorDistance.Result).Result
     B=&RigVMFunction_MathDoubleRemap(SourceMaximum=15, SourceMinimum=0, TargetMaximum=0, TargetMinimum=1, Value=@n22.Result).Result
    ).Result
   )
   &RigUnit_ParentConstraint(
    AdvancedSettings=(RotationOrderForFilter=XZY)
    Child=(Type=Bone,Name="dyn_strap_c_04")
    ExecutePin=()
    Filter=(TranslationFilter=(bX=True,bY=True,bZ=True),RotationFilter=(bX=true,bY=true,bZ=true),ScaleFilter=(bX=True,bY=True,bZ=True))
    Parents="((Item=(Type=Bone,Name=\"deform_upperarm_bk_r\"),Weight=0.3),(Item=(Type=Bone,Name=\"spine_04\"),Weight=0.053),(Item=(Type=Bone,Name=\"deform_upperarm_up_r\"),Weight=0.0577),(Item=(Type=Bone,Name=\"upperarm_twist_01_r\"),Weight=0.0807),(Item=(Type=Bone,Name=\"spine ...(truncated)"
    Weight=1
   )
   RigVMDispatchNode(Enabled=True, ExecuteContext=(), ScreenColor=(R=1,G=1,B=1,A=1), ScreenDuration=0.05, Value=@Add.Result) {template="DISPATCH_RigVMDispatch_Print(in Prefix,in Value,in Enabled,in ScreenDuration,in ScreenColor)"}
   Remap = &RigVMFunction_MathDoubleRemap(SourceMaximum=1, SourceMinimum=0, TargetMaximum=1, TargetMinimum=0, bClamp=False, Value=@n17.Result)
   RigVMVariableNode() {variable=head_lft}
  ]}
]
hierarchy [
 root {loc=<0 0 0>, rot=<0 0 0 1>}
 pelvis {parent=root, loc=<0 -0.7071 93.1191>, rot=<0.0127 0.707 -0.0127 -0.707>}
 spine_01 {parent=pelvis, loc=<6.7494 -0.6388 0>, rot=<0 0 0.034 -0.9994>}
 spine_02 {parent=spine_01, loc=<7.252 -0.035 0>, rot=<0 0 0.0238 -0.9997>}
 spine_03 {parent=spine_02, loc=<6.0106 0 0>, rot=<0 0 0.0149 -0.9999>}
 spine_04 {parent=spine_03, loc=<8.1023 0 0>, rot=<0 0 -0.0132 -0.9999>}
 spine_05 {parent=spine_04, loc=<10.0142 0 0>, rot=<0 0 -0.0015 -1>}
 neck_01 {parent=spine_05, loc=<11.3138 0.1529 0>, rot=<0 0 -0.2003 -0.9797>}
 neck_02 {parent=neck_01, loc=<4.5627 0 0>, rot=<0 0 0.0504 -0.9987>}
 head {parent=neck_02, loc=<5.3914 0 0>, rot=<0 0 0.115 -0.9934>}
 head_ctrl {controlType=EULER_TRANSFORM, parent=head, type=Control, loc=<0 0 0>, rot=<0 0 0 1>}
 head_ctrl_2 {controlType=EULER_TRANSFORM, parent=head, type=Control, loc=<0 0 0>, rot=<0 0 0 1>}
 clavicle_l {parent=spine_05, loc=<4.3575 -0.1466 -2.3223>, rot=<0.0465 -0.7289 -0.009 -0.6829>}
 upperarm_l {parent=clavicle_l, loc=<15.5111 0 0>, rot=<0.0226 -0.3234 -0.0088 -0.9459>}
 lowerarm_l {parent=upperarm_l, loc=<26.6514 0 0>, rot=<0.0391 -0.0099 -0.2453 -0.9686>}
 hand_l {parent=lowerarm_l, loc=<22.3491 0 0>, rot=<-0.7081 -0.0507 0.0975 -0.6975>}
 thumb_01_l {parent=hand_l, loc=<1.3118 -0.3191 -2.459>, rot=<0.6546 -0.4042 0.0664 -0.6354>}
 thumb_02_l {parent=thumb_01_l, loc=<5.9336 -0.0448 0.0401>, rot=<-0.0037 0.0213 0.298 -0.9543>}
 thumb_03_l {parent=thumb_02_l, loc=<2.2223 0.3682 -0.0907>, rot=<0.0118 -0.013 0.0066 -0.9998>}
 index_metacarpal_l {parent=hand_l, loc=<1.6535 0.0183 -2.96>, rot=<0.0158 -0.0124 -0.0173 -0.9996>}
 index_01_l {parent=index_metacarpal_l, loc=<8.3008 0.0357 -0.0524>, rot=<-0.0044 0.0485 0.0924 -0.9945>}
 index_02_l {parent=index_01_l, loc=<3.3587 -0.3004 0.0452>, rot=<-0.002 0.0027 0.3225 -0.9466>}
 index_03_l {parent=index_02_l, loc=<2.5578 -0.123 -0.0252>, rot=<-0.002 0.0067 0.2558 -0.9667>}
 middle_metacarpal_l {parent=hand_l, loc=<2.1856 -0.1725 -0.5984>, rot=<0.006 0.0168 -0.0329 -0.9993>}
 middle_01_l {parent=middle_metacarpal_l, loc=<7.3275 -0.2447 0.0129>, rot=<-0.0208 0.0042 0.0858 -0.9961>}
 middle_02_l {parent=middle_01_l, loc=<4.0123 0.1401 0.1175>, rot=<-0.0218 0.0258 0.3512 -0.9357>}
 middle_03_l {parent=middle_02_l, loc=<2.6077 0.0534 0.1373>, rot=<0.0082 0.0188 0.1708 -0.9851>}
 ring_metacarpal_l {parent=hand_l, loc=<2.2074 -0.3625 1.3638>, rot=<-0.0284 0.0722 -0.0458 -0.9959>}
 ring_01_l {parent=ring_metacarpal_l, loc=<7.0295 -0.6197 -0.1446>, rot=<-0.0489 -0.0252 0.1892 -0.9804>}
 ring_02_l {parent=ring_01_l, loc=<3.6174 0.4647 -0.1027>, rot=<0 0 0.2806 -0.9598>}
 ring_03_l {parent=ring_02_l, loc=<2.4152 0.114 0.0199>, rot=<0 0 0.1189 -0.9929>}
 pinky_metacarpal_l {parent=hand_l, loc=<1.7616 -0.9753 3.3814>, rot=<-0.0397 0.0814 -0.069 -0.9935>}
 pinky_01_l {parent=pinky_metacarpal_l, loc=<7.1709 -0.4831 -0.1094>, rot=<-0.0654 -0.0035 0.2242 -0.9723>}
 pinky_02_l {parent=pinky_01_l, loc=<2.8274 0.0893 -0.1043>, rot=<0.0037 0.0011 0.2854 -0.9584>}
 pinky_03_l {parent=pinky_02_l, loc=<2.2046 0.0307 -0.0078>, rot=<-0.0002 0.0098 0.1445 -0.9895>}
 deform_wrist_dn_base_l {parent=hand_l, loc=<0 -2.2342 0>, rot=<0 0 0 -1>}
 deform_wrist_dn_l {parent=deform_wrist_dn_base_l, loc=<0 0 0>, rot=<0 0 0 1>}
 deform_wrist_up_base_l {parent=hand_l, loc=<0 2.2682 0>, rot=<0 0 0 -1>}
 deform_wrist_up_l {parent=deform_wrist_up_base_l, loc=<0 0 0>, rot=<0 0 0 1>}
 lowerarm_twist_02_l {parent=lowerarm_l, loc=<9.6199 0 0>, rot=<0 0 0 -1>}
 lowerarm_twist_01_l {parent=lowerarm_l, loc=<17.8549 0 0>, rot=<0 0 0 -1>}
 deform_elbow_in_base_l {parent=lowerarm_l, loc=<0.2112 5.259 0.8274>, rot=<0 0 0 -1>}
 deform_elbow_in_l {parent=deform_elbow_in_base_l, loc=<0 0 0>, rot=<0 0 0 1>}
 deform_elbow_out_base_l {parent=lowerarm_l, loc=<0 -4 0>, rot=<0 0 0 -1>}
 deform_elbow_out_l {parent=deform_elbow_out_base_l, loc=<0 0 0>, rot=<0 0 0 1>}
 upperarm_twist_01_l {parent=upperarm_l, loc=<0.0041 0 0>, rot=<0 0 0 -1>}
 deform_upperarm_bk_base_l {parent=upperarm_twist_01_l, loc=<1.0813 -7.2193 -1.9453>, rot=<0 0 0 1>}
 deform_upperarm_bk_l {parent=deform_upperarm_bk_base_l, loc=<0 0 0>, rot=<0 0 0 1>}
 deform_upperarm_fr_base_l {parent=upperarm_twist_01_l, loc=<-0.3691 7.0506 -2.765>, rot=<0 0 0 1>}
 deform_upperarm_fr_l {parent=deform_upperarm_fr_base_l, loc=<0 0 0>, rot=<0 0 0 1>}
 deform_upperarm_up_base_l {parent=upperarm_twist_01_l, loc=<-3.2216 -1.1391 3.3311>, rot=<0 0 0 1>}
 deform_upperarm_up_l {parent=deform_upperarm_up_base_l, loc=<0 0 0>, rot=<0 0 0 1>}
 upperarm_twist_02_l {parent=upperarm_l, loc=<13.7941 0 0>, rot=<0 0 0 -1>}
 clavicle_r {parent=spine_05, loc=<4.3575 -0.1466 2.3223>, rot=<0.7289 0.0465 -0.6829 0.009>}
 upperarm_r {parent=clavicle_r, loc=<-15.5111 0 0>, rot=<0.0226 -0.3234 -0.0088 -0.9459>}
 lowerarm_r {parent=upperarm_r, loc=<-26.6514 0 0>, rot=<0.0391 -0.0099 -0.2453 -0.9686>}
 hand_r {parent=lowerarm_r, loc=<-22.3491 0 0>, rot=<-0.7081 -0.0507 0.0975 -0.6975>}
 thumb_01_r {parent=hand_r, loc=<-1.3118 0.3191 2.459>, rot=<0.6546 -0.4042 0.0664 -0.6354>}
 thumb_02_r {parent=thumb_01_r, loc=<-5.9336 0.0448 -0.0401>, rot=<-0.0037 0.0213 0.298 -0.9543>}
 thumb_03_r {parent=thumb_02_r, loc=<-2.2223 -0.3682 0.0907>, rot=<0.0118 -0.013 0.0066 -0.9998>}
 index_metacarpal_r {parent=hand_r, loc=<-1.6535 -0.0183 2.96>, rot=<0.0158 -0.0124 -0.0173 -0.9996>}
 index_01_r {parent=index_metacarpal_r, loc=<-8.3008 -0.0357 0.0524>, rot=<-0.0044 0.0485 0.0924 -0.9945>}
 index_02_r {parent=index_01_r, loc=<-3.3587 0.3004 -0.0452>, rot=<-0.002 0.0027 0.3225 -0.9466>}
 index_03_r {parent=index_02_r, loc=<-2.5578 0.123 0.0252>, rot=<-0.002 0.0067 0.2558 -0.9667>}
 middle_metacarpal_r {parent=hand_r, loc=<-2.1856 0.1725 0.5984>, rot=<0.006 0.0168 -0.0329 -0.9993>}
 middle_01_r {parent=middle_metacarpal_r, loc=<-7.3275 0.2447 -0.0129>, rot=<-0.0208 0.0042 0.0858 -0.9961>}
 middle_02_r {parent=middle_01_r, loc=<-4.0123 -0.1401 -0.1175>, rot=<-0.0218 0.0258 0.3512 -0.9357>}
 middle_03_r {parent=middle_02_r, loc=<-2.6077 -0.0534 -0.1373>, rot=<0.0082 0.0188 0.1708 -0.9851>}
 ring_metacarpal_r {parent=hand_r, loc=<-2.2074 0.3625 -1.3638>, rot=<-0.0284 0.0722 -0.0458 -0.9959>}
 ring_01_r {parent=ring_metacarpal_r, loc=<-7.0295 0.6197 0.1446>, rot=<-0.0489 -0.0252 0.1892 -0.9804>}
 ring_02_r {parent=ring_01_r, loc=<-3.6174 -0.4647 0.1027>, rot=<0 0 0.2806 -0.9598>}
 ring_03_r {parent=ring_02_r, loc=<-2.4152 -0.114 -0.0199>, rot=<0 0 0.1189 -0.9929>}
 pinky_metacarpal_r {parent=hand_r, loc=<-1.7616 0.9753 -3.3814>, rot=<-0.0397 0.0814 -0.069 -0.9935>}
 pinky_01_r {parent=pinky_metacarpal_r, loc=<-7.1709 0.4831 0.1094>, rot=<-0.0654 -0.0035 0.2242 -0.9723>}
 pinky_02_r {parent=pinky_01_r, loc=<-2.8274 -0.0893 0.1043>, rot=<0.0037 0.0011 0.2854 -0.9584>}
 pinky_03_r {parent=pinky_02_r, loc=<-2.2046 -0.0307 0.0078>, rot=<-0.0002 0.0098 0.1445 -0.9895>}
 deform_wrist_dn_base_r {parent=hand_r, loc=<-0.0004 2.2345 -0.0001>, rot=<0 0 0 -1>}
 deform_wrist_dn_r {parent=deform_wrist_dn_base_r, loc=<0 0 0>, rot=<0 0 0 1>}
 deform_wrist_up_base_r {parent=hand_r, loc=<0.0003 -2.2684 0.0001>, rot=<0 0 0 -1>}
 deform_wrist_up_r {parent=deform_wrist_up_base_r, loc=<0 0 0>, rot=<0 0 0 1>}
 lowerarm_twist_02_r {parent=lowerarm_r, loc=<-9.6199 0 0>, rot=<0 0 0 -1>}
 lowerarm_twist_01_r {parent=lowerarm_r, loc=<-17.8549 0 0>, rot=<0 0 0 -1>}
 deform_elbow_in_base_r {parent=lowerarm_r, loc=<-0.2111 -5.2591 -0.8275>, rot=<0 0 0 -1>}
 deform_elbow_in_r {parent=deform_elbow_in_base_r, loc=<-0.0001 0 0.0002>, rot=<0 0 0 1>}
 deform_elbow_out_base_r {parent=lowerarm_r, loc=<0 4 -0.0001>, rot=<0 0 0 -1>}
 deform_elbow_out_r {parent=deform_elbow_out_base_r, loc=<0 0 0>, rot=<0 0 0 1>}
 upperarm_twist_01_r {parent=upperarm_r, loc=<-0.0041 0 0>, rot=<0 0 0 -1>}
 deform_upperarm_bk_base_r {parent=upperarm_twist_01_r, loc=<-1.0814 7.2192 1.9455>, rot=<0 0 0 1>}
 deform_upperarm_bk_r {parent=deform_upperarm_bk_base_r, loc=<0 0 0>, rot=<0 0 0 1>}
 deform_upperarm_fr_base_r {parent=upperarm_twist_01_r, loc=<0.3689 -7.0506 2.7652>, rot=<0 0 0 1>}
 deform_upperarm_fr_r {parent=deform_upperarm_fr_base_r, loc=<0 0 0>, rot=<0 0 0 1>}
 deform_upperarm_up_base_r {parent=upperarm_twist_01_r, loc=<3.2214 1.1391 -3.3308>, rot=<0 0 0 1>}
 deform_upperarm_up_r {parent=deform_upperarm_up_base_r, loc=<0 0 0>, rot=<0 0 0 -1>}
 upperarm_twist_02_r {parent=upperarm_r, loc=<-13.7941 0 0>, rot=<0 0 0 -1>}
 dyn_splineStrap_MSTR {parent=clavicle_r, loc=<-11.9951 0.4227 -9.6787>, rot=<-0.2221 -0.0156 -0.637 -0.738>}
 dyn_splineStrap_01 {parent=dyn_splineStrap_MSTR, loc=<0 0 0>, rot=<0 0 0 1>}
 dyn_splineStrap_02 {parent=dyn_splineStrap_01, loc=<6.072 0 0>, rot=<-0.087 0.318 0.0392 -0.9433>}
 dyn_splineStrap_03 {parent=dyn_splineStrap_02, loc=<6.3737 0 0>, rot=<-0.0238 0.1931 0.0435 -0.9799>}
 dyn_splineStrap_04 {parent=dyn_splineStrap_03, loc=<6.2894 0 0>, rot=<-0.0556 0.0676 -0.02 -0.996>}
 dyn_splineStrap_05 {parent=dyn_splineStrap_04, loc=<4.5577 0 0>, rot=<-0.0149 0.0603 0.0045 -0.9981>}
 dyn_splineStrap_06 {parent=dyn_splineStrap_05, loc=<5.6332 0 0>, rot=<-0.0176 0.0674 0.0116 -0.9975>}
 dyn_splineStrap_07 {parent=dyn_splineStrap_06, loc=<5.3457 0 0>, rot=<-0.0115 0.0302 -0.0075 -0.9994>}
 dyn_splineStrap_08 {parent=dyn_splineStrap_07, loc=<5.0665 0 0>, rot=<-0.0085 0.0914 -0.0059 -0.9958>}
 dyn_splineStrap_09 {parent=dyn_splineStrap_08, loc=<5.3393 0 0>, rot=<0.0832 0.2051 0.0578 -0.9735>}
 dyn_splineStrap_10 {parent=dyn_splineStrap_09, loc=<4.6513 0 0>, rot=<0.097 0.2612 -0.0029 -0.9604>}
 dyn_splineStrap_11 {parent=dyn_splineStrap_10, loc=<4.5238 0 0>, rot=<0 0 0 1>}
 dyn_splineStrapFr_01 {parent=dyn_splineStrap_MSTR, loc=<0 0 0>, rot=<0.0985 -0.9247 0.0137 0.3674>}
 dyn_splineStrapFr_02 {parent=dyn_splineStrapFr_01, loc=<6.6834 0 0>, rot=<0.005 -0.1964 0.0538 -0.979>}
 dyn_splineStrapFr_03 {parent=dyn_splineStrapFr_02, loc=<5.5909 0 0>, rot=<0.0279 -0.122 0.0487 -0.9909>}
 dyn_splineStrapFr_04 {parent=dyn_splineStrapFr_03, loc=<6.5431 0 0>, rot=<0.0699 -0.0007 0.0319 -0.997>}
 dyn_splineStrapFr_05 {parent=dyn_splineStrapFr_04, loc=<6.2566 0 0>, rot=<0.0079 -0.1759 0.0094 -0.9843>}
 dyn_splineStrapFr_06 {parent=dyn_splineStrapFr_05, loc=<5.6953 0 0>, rot=<-0.0042 -0.0873 0.0427 -0.9953>}
 dyn_splineStrapFr_07 {parent=dyn_splineStrapFr_06, loc=<6.4843 0 0>, rot=<-0.0384 -0.1346 -0.0066 -0.9901>}
 dyn_splineStrapFr_08 {parent=dyn_splineStrapFr_07, loc=<5.5953 0 0>, rot=<-0.0612 -0.1568 0.0647 -0.9836>}
 dyn_splineStrapFr_09 {parent=dyn_splineStrapFr_08, loc=<5.0256 0 0>, rot=<-0.0204 -0.149 0.0432 -0.9877>}
 dyn_splineStrapFr_10 {parent=dyn_splineStrapFr_09, loc=<4.2025 0 0>, rot=<-0.0914 -0.2185 0.0539 -0.97>}
 dyn_splineStrapFr_11 {parent=dyn_splineStrapFr_10, loc=<4.3567 0 0>, rot=<0.1305 0 0 -0.9914>}
 attach_backpack {parent=spine_05, loc=<-0.123 -13.473 0>, rot=<-0.0093 -0.707 0.0093 -0.707>}
 attach_cape {parent=spine_05, loc=<9.017 -11.714 0>, rot=<0.0616 -0.7044 -0.0616 -0.7044>}
 dyn_strap_c_04 {parent=spine_05, loc=<5.5847 -10.8983 9.2415>, rot=<-0.2702 0.6564 0.6707 0.2151>}
 dyn_strap_c_03 {parent=spine_05, loc=<13.3415 -1.333 14.8746>, rot=<0.4847 -0.8235 -0.2652 -0.1287>}
 dyn_strap_c_02 {parent=spine_05, loc=<7.4544 9.2914 11.4839>, rot=<0.8479 -0.5091 0.0862 0.1205>}
 dyn_strap_c_05 {parent=spine_05, loc=<-3.9738 -11.7306 1.1466>, rot=<-0.2702 0.6564 0.6707 0.2151>}
 dyn_strap_c_01 {parent=spine_05, loc=<1.003 12.1572 6.6994>, rot=<0.8401 -0.2929 0.1843 0.4176>}
 deform_collar_l {parent=spine_05, loc=<11.9654 0.4309 -9.507>, rot=<0 0 -0.2003 -0.9797>}
 dyn_collar_l {parent=deform_collar_l, loc=<0 0 0>, rot=<0 0 0 1>}
 deform_collar_r {parent=spine_05, loc=<11.965 0.4308 9.507>, rot=<0 0 0.9797 -0.2003>}
 dyn_collar_r {parent=deform_collar_r, loc=<0 0 0>, rot=<0 0 0 1>}
 deform_collar_b {parent=spine_05, loc=<15.2886 -5.6882 0>, rot=<0 0 -0.2003 -0.9797>}
 dyn_collar_b {parent=deform_collar_b, loc=<0 0 0>, rot=<0 0 0 1>}
 deform_collar_f_l {parent=spine_05, loc=<6.7139 8.1548 -5.2701>, rot=<-0.3434 -0.0444 -0.2545 -0.9029>}
 dyn_collar_f_l {parent=deform_collar_f_l, loc=<0 0 0>, rot=<0 0 0 1>}
 deform_collar_f_r {parent=spine_05, loc=<6.7141 8.1548 5.2701>, rot=<-0.0444 0.3434 0.9029 -0.2545>}
 dyn_collar_f_r {parent=deform_collar_f_r, loc=<0 0 0>, rot=<0 0 0 1>}
 deform_pec_aim_base_l {parent=spine_04, loc=<7.8728 12.6388 -2.4123>, rot=<0.0102 -0.4602 0.3979 -0.7936>}
 deform_pec_aim_l {parent=deform_pec_aim_base_l, loc=<0 0 0>, rot=<0 0 0 1>}
 deform_pec_base_l {parent=deform_pec_aim_l, loc=<10.766 0 0>, rot=<0 0 0 1>}
 deform_pec_l {parent=deform_pec_base_l, loc=<0.055 0 0>, rot=<0 0 0 1>}
 deform_pec_aim_base_r {parent=spine_04, loc=<7.8723 12.6388 2.4123>, rot=<-0.4602 -0.0102 0.7936 0.3979>}
 deform_pec_aim_r {parent=deform_pec_aim_base_r, loc=<0 0 0>, rot=<0 0 0 1>}
 deform_pec_base_r {parent=deform_pec_aim_r, loc=<-10.7662 -0.0006 -0.0007>, rot=<0 0 0 1>}
 deform_pec_r {parent=deform_pec_base_r, loc=<0 0 0>, rot=<0 0 0 1>}
 dyn_skirtTopA_l_MSTR {parent=spine_02, loc=<-2.2546 4.936 13.426>, rot=<0.233 -0.5051 0.8012 0.2206>}
 dyn_skirtTopA_l_01 {parent=dyn_skirtTopA_l_MSTR, loc=<0 0 0>, rot=<0 0 0 1>}
 dyn_skirtTopB_l_MSTR {parent=spine_02, loc=<-1.3921 -8.5262 -4.6265>, rot=<-0.1252 -0.0323 0.1343 -0.9825>}
 dyn_skirtTopB_l_01 {parent=dyn_skirtTopB_l_MSTR, loc=<0 0 0>, rot=<0 0 0 1>}
 dyn_skirtTopC_l_MSTR {parent=spine_02, loc=<-1.0091 -2.768 -12.4915>, rot=<-0.5546 -0.1614 0.1672 -0.799>}
 dyn_skirtTopC_l_01 {parent=dyn_skirtTopC_l_MSTR, loc=<0 0 0>, rot=<0 0 0 1>}
 dyn_bottleB_MSTR {parent=spine_01, loc=<5.0349 -5.3616 -12.344>, rot=<-0.2029 0.6511 0.7201 0.1282>}
 dyn_bottleB_01 {parent=dyn_bottleB_MSTR, loc=<0 0 0>, rot=<0 0 0 1>}
 dyn_bottleB_02 {parent=dyn_bottleB_01, loc=<4.9886 0 0>, rot=<0 0 0 1>}
 dyn_bottleA_MSTR {parent=spine_01, loc=<4.2565 -0.1746 -14.4124>, rot=<-0.1878 0.5623 0.7996 0.0961>}
 dyn_bottleA_01 {parent=dyn_bottleA_MSTR, loc=<0 0 0>, rot=<0 0 0 1>}
 dyn_bottleA_02 {parent=dyn_bottleA_01, loc=<4.4928 0 0>, rot=<0 0 0 1>}
 thigh_l {parent=pelvis, loc=<-2.5693 -0.0109 -7.7501>, rot=<0.0012 0.0572 0.0155 -0.9982>}
 calf_l {parent=thigh_l, loc=<-39.0484 0 0>, rot=<0.0006 -0.0161 -0.0391 -0.9991>}
 foot_l {parent=calf_l, loc=<-42.561 0.1077 0.4292>, rot=<0.0418 -0.0455 0.0343 -0.9975>}
 ball_l {parent=foot_l, loc=<-6.1599 13.8264 0.6806>, rot=<0.0001 -0.0001 -0.7063 -0.7079>}
 calf_twist_02_l {parent=calf_l, loc=<-18.8064 -0.0967 0.0397>, rot=<0.0237 0.0095 -0.0035 -0.9997>}
 calf_twist_01_l {parent=calf_l, loc=<-35.813 -0.0967 0.0397>, rot=<0.0237 0.0095 -0.0035 -0.9997>}
 deform_knee_in_base_l {parent=calf_l, loc=<1.298 -5.2957 -0.0496>, rot=<0 0 0 -1>}
 deform_knee_in_l {parent=deform_knee_in_base_l, loc=<0 0 0>, rot=<0 0 0 1>}
 deform_knee_out_base_l {parent=calf_l, loc=<0 4 0>, rot=<0 0 0 -1>}
 deform_knee_out_l {parent=deform_knee_out_base_l, loc=<0 0 0>, rot=<0 0 0 1>}
 thigh_twist_01_l {parent=thigh_l, loc=<-21.6303 0 0>, rot=<0.0013 0 0 -1>}
 deform_groin_base_l {parent=thigh_l, loc=<-2.1908 8.1142 -1.7442>, rot=<0 0 0 -1>}
 deform_groin_l {parent=deform_groin_base_l, loc=<0 0 0>, rot=<0 0 0 1>}
 thigh_r {parent=pelvis, loc=<-2.5693 -0.0109 7.7501>, rot=<0.0572 -0.0012 0.9982 0.0155>}
 calf_r {parent=thigh_r, loc=<39.0484 0 0>, rot=<0.0006 -0.0161 -0.0391 -0.9991>}
 foot_r {parent=calf_r, loc=<42.561 -0.1077 -0.4292>, rot=<0.0418 -0.0455 0.0343 -0.9975>}
 ball_r {parent=foot_r, loc=<6.1599 -13.8264 -0.6806>, rot=<0.0001 -0.0001 -0.7063 -0.7079>}
 calf_twist_02_r {parent=calf_r, loc=<18.8064 0.0967 -0.0397>, rot=<0.0237 0.0095 -0.0035 -0.9997>}
 calf_twist_01_r {parent=calf_r, loc=<35.813 0.0967 -0.0397>, rot=<0.0237 0.0095 -0.0035 -0.9997>}
 deform_knee_in_base_r {parent=calf_r, loc=<-1.298 5.2957 0.0496>, rot=<0 0 0 -1>}
 deform_knee_in_r {parent=deform_knee_in_base_r, loc=<0 0 0>, rot=<0 0 0 1>}
 deform_knee_out_base_r {parent=calf_r, loc=<0 -4 0>, rot=<0 0 0 -1>}
 deform_knee_out_r {parent=deform_knee_out_base_r, loc=<0 0 0>, rot=<0 0 0 1>}
 thigh_twist_01_r {parent=thigh_r, loc=<21.6303 0 0>, rot=<0.0013 0 0 -1>}
 deform_groin_base_r {parent=thigh_r, loc=<2.1908 -8.1142 1.7442>, rot=<0 0 0 -1>}
 deform_groin_r {parent=deform_groin_base_r, loc=<0 0 0>, rot=<0 0 0 1>}
 deform_glute_base_l {parent=pelvis, loc=<-8.8064 -8.8222 -8.4287>, rot=<0.0275 0.0501 -0.4472 -0.8926>}
 deform_glute_l {parent=deform_glute_base_l, loc=<0 0 0>, rot=<0 0 0 1>}
 deform_glute_base_r {parent=pelvis, loc=<-8.8062 -8.8222 8.4287>, rot=<0.0501 -0.0275 0.8926 -0.4472>}
 deform_glute_r {parent=deform_glute_base_r, loc=<0 0 0>, rot=<0 0 0 1>}
 dyn_skirtA_l_MSTR {parent=pelvis, loc=<5.1456 8.2609 -5.1258>, rot=<-0.1449 0.1169 0.9763 0.1105>}
 dyn_skirtA_l_01 {parent=dyn_skirtA_l_MSTR, loc=<0 0 0>, rot=<0 0 0 1>}
 dyn_skirtA_l_02 {parent=dyn_skirtA_l_01, loc=<7.4441 0 0>, rot=<0 0 -0.0087 -1>}
 dyn_skirtA_l_03 {parent=dyn_skirtA_l_02, loc=<8.3991 0 0>, rot=<0 0 -0.0269 -0.9996>}
 dyn_skirtA_l_04 {parent=dyn_skirtA_l_03, loc=<14.1276 0.4542 0>, rot=<0 0 -0.0466 -0.9989>}
 dyn_skirtA_l_05 {parent=dyn_skirtA_l_04, loc=<11.0826 0 0>, rot=<0 0 0.0175 -0.9998>}
 dyn_skirtB_l_MSTR {parent=pelvis, loc=<7.5896 -1.4735 -13.6831>, rot=<0.0996 -0.7604 -0.6343 -0.0976>}
 dyn_skirtB_l_01 {parent=dyn_skirtB_l_MSTR, loc=<0 0 0>, rot=<0 0 0 1>}
 dyn_skirtB_l_02 {parent=dyn_skirtB_l_01, loc=<8.2649 0 0>, rot=<-0.0285 0.0186 0.0168 -0.9993>}
 dyn_skirtB_l_03 {parent=dyn_skirtB_l_02, loc=<9.5201 0 0>, rot=<0 0 0 1>}
 dyn_skirtB_l_04 {parent=dyn_skirtB_l_03, loc=<11.4527 0 0>, rot=<0 0 0 1>}
 dyn_ropeA_l_MSTR {parent=pelvis, loc=<8.846 8.7916 3.5661>, rot=<0.2664 -0.252 0.8465 0.3859>}
 dyn_ropeA_l_01 {parent=dyn_ropeA_l_MSTR, loc=<0 0 0>, rot=<0 0 0 1>}
 dyn_ropeA_l_02 {parent=dyn_ropeA_l_01, loc=<5.5567 0 0>, rot=<-0.0045 -0.0164 -0.2919 -0.9563>}
 dyn_ropeA_l_03 {parent=dyn_ropeA_l_02, loc=<6.8628 0 0>, rot=<0 0 0 1>}
 dyn_ropeB_l_MSTR {parent=pelvis, loc=<6.7632 8.6289 5.6481>, rot=<-0.0176 -0.0024 0.9744 0.2243>}
 dyn_ropeB_l_01 {parent=dyn_ropeB_l_MSTR, loc=<0 0 0>, rot=<0 0 0 1>}
 dyn_ropeB_l_02 {parent=dyn_ropeB_l_01, loc=<6.7386 0 0>, rot=<-0.0097 0.0157 -0.154 -0.9879>}
 dyn_ropeB_l_03 {parent=dyn_ropeB_l_02, loc=<7.0916 0 0>, rot=<0 0 0 1>}
 dyn_skirtB_r_MSTR {parent=pelvis, loc=<7.4323 0.9 13.3576>, rot=<0.0672 -0.6891 0.7153 0.0944>}
 dyn_skirtB_r_01 {parent=dyn_skirtB_r_MSTR, loc=<0 0 0>, rot=<0 0 0 1>}
 dyn_skirtB_r_02 {parent=dyn_skirtB_r_01, loc=<6.257 -0.0125 0.076>, rot=<0.005 -0.0189 -0.0032 -0.9998>}
 dyn_skirtB_r_03 {parent=dyn_skirtB_r_02, loc=<8.2197 -0.3611 0.1697>, rot=<0 0 0 1>}
 dyn_skirtA_r_MSTR {parent=pelvis, loc=<5.3241 7.5916 6.1206>, rot=<0.0574 -0.0924 0.9894 0.096>}
 dyn_skirtA_r_01 {parent=dyn_skirtA_r_MSTR, loc=<0 0 0>, rot=<0 0 0 1>}
 dyn_skirtA_r_02 {parent=dyn_skirtA_r_01, loc=<5.5952 0 0>, rot=<0 0 0 1>}
 dyn_skirtC_r_MSTR {parent=pelvis, loc=<7.4671 -9.0098 8.909>, rot=<0.1322 -0.9912 0 0>}
 dyn_skirtC_r_01 {parent=dyn_skirtC_r_MSTR, loc=<0 0 0>, rot=<0 0 0 1>}
 dyn_skirtC_r_02 {parent=dyn_skirtC_r_01, loc=<8.9346 0 0>, rot=<0 0 -0.0552 -0.9985>}
 dyn_skirtC_r_03 {parent=dyn_skirtC_r_02, loc=<10.7076 0 0>, rot=<0 0 0 1>}
 dyn_skirtC_l_MSTR {parent=pelvis, loc=<7.4646 -8.507 -9.1278>, rot=<0.1354 -0.9908 0 0>}
 dyn_skirtC_l_01 {parent=dyn_skirtC_l_MSTR, loc=<0 0 0>, rot=<0 0 0 1>}
 dyn_skirtC_l_02 {parent=dyn_skirtC_l_01, loc=<8.9343 0 0>, rot=<0 0 -0.0412 -0.9992>}
 dyn_skirtC_l_03 {parent=dyn_skirtC_l_02, loc=<10.7077 0 0>, rot=<0 0 0 1>}
 dyn_skirtC_l_04 {parent=dyn_skirtC_l_03, loc=<13.7494 0 0>, rot=<0 0 0 1>}
]
variables [
 head_bck {type=float, default=0}
 head_fwd {type=float, default=0}
 head_lft {type=float, default=0}
 head_rgt {type=float, default=0}
 head_twist_lf {type=float, default=0}
 head_twist_rgt {type=float, default=0}
 is_front_end {type=bool, default=False}
 is_front_end_not_emoting {type=bool, default=False}
 is_targeting {type=bool, default=False}
 left_thigh_up {type=float, default=0}
 right_thigh_up {type=float, default=0}
]
