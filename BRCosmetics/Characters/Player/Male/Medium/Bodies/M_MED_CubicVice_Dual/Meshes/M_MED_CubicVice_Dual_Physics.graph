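# Physics asset for the CubicVice_Dual medium male player body cosmetic, defining capsule bodies and joints for five asymmetric dynamic skirt chains plus the body.
# Each chain hangs from a tiny MSTR body on the pelvis, with tight per-chain swing limits and little or no twist.
# Core skeleton joints and MSTR links are free.
# Legs, neck and head have fitted capsules, while torso and arms use identical generic ones, with no forearms or hands.
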PhysicsAsset /BRCosmetics/Characters/Player/Male/Medium/Bodies/M_MED_CubicVice_Dual/Meshes/M_MED_CubicVice_Dual_Physics
bodies {
 calf_l [{shapeType=Capsule, length=23.5369, radius=13.5145, center=<-21.8562 0 0>, rotation=<rot 90 0 0>}]
 calf_r [{shapeType=Capsule, length=23.5369, radius=13.5145, center=<22.0749 0 0>, rotation=<rot 90 0 0>}]
 clavicle_l [{shapeType=Capsule, length=5, radius=10}]
 clavicle_r [{shapeType=Capsule, length=5, radius=10}]
 dyn_skirtA_l_01 [{shapeType=Capsule, length=10, radius=6, shapeName=dyn_skirtA_l_01_capsule, center=<3 -7 0>}]
 dyn_skirtA_l_02 [{shapeType=Capsule, length=10, radius=8, shapeName=dyn_skirtA_l_02_capsule, center=<3 -7 0>}]
 dyn_skirtA_l_03 [{shapeType=Capsule, length=9, radius=9, shapeName=dyn_skirtA_l_03_capsule, center=<3 -7 0>}]
 dyn_skirtA_l_04 [{shapeType=Capsule, length=8, radius=10, shapeName=dyn_skirtA_l_04_capsule, center=<3.6671 -7 0>}]
 dyn_skirtA_l_05 [{shapeType=Capsule, length=8, radius=10, shapeName=dyn_skirtA_l_05_capsule, center=<6.408 -7 0>}]
 dyn_skirtA_l_MSTR [{shapeType=Capsule, length=0.5, radius=0.5, shapeName=dyn_skirtA_l_MSTR_capsule}]
 dyn_skirtB_l_01 [{shapeType=Capsule, length=10.5, radius=5.5, shapeName=dyn_skirtB_l_01_capsule, center=<3 -7 0>}]
 dyn_skirtB_l_02 [{shapeType=Capsule, length=11.5, radius=6.5, shapeName=dyn_skirtB_l_02_capsule, center=<3 -7 0>}]
 dyn_skirtB_l_03 [{shapeType=Capsule, length=10.5, radius=7.5, shapeName=dyn_skirtB_l_03_capsule, center=<3 -7 0>}]
 dyn_skirtB_l_04 [{shapeType=Capsule, length=16.5, radius=9, shapeName=dyn_skirtB_l_04_capsule, center=<6.1727 -7 0>, rotation=<rot 0 0 9.7717>}]
 dyn_skirtB_l_MSTR [{shapeType=Capsule, length=0.5, radius=0.5, shapeName=dyn_skirtB_l_MSTR_capsule}]
 dyn_skirtB_r_01 [{shapeType=Capsule, length=10, radius=6, shapeName=dyn_skirtB_r_01_capsule, center=<3 -7 0>}]
 dyn_skirtB_r_02 [{shapeType=Capsule, length=10, radius=8, shapeName=dyn_skirtB_r_02_capsule, center=<3 -7 0>}]
 dyn_skirtB_r_03 [{shapeType=Capsule, length=11, radius=7, shapeName=dyn_skirtB_r_03_capsule, center=<3 -7 0>}]
 dyn_skirtB_r_MSTR [{shapeType=Capsule, length=0.5, radius=0.5, shapeName=dyn_skirtB_r_MSTR_capsule}]
 dyn_skirtC_l_01 [{shapeType=Capsule, length=10, radius=8, shapeName=dyn_skirtC_l_01_capsule, center=<3 -7 0>}]
 dyn_skirtC_l_02 [{shapeType=Capsule, length=10, radius=8, shapeName=dyn_skirtC_l_02_capsule, center=<3 -7 0>}]
 dyn_skirtC_l_03 [{shapeType=Capsule, length=9, radius=9, shapeName=dyn_skirtC_l_03_capsule, center=<3 -7 0>}]
 dyn_skirtC_l_04 [{shapeType=Capsule, length=15, radius=9, shapeName=dyn_skirtC_l_04_capsule, center=<3.7098 -7 0>, rotation=<rot 0 0.0383 6.2863>}]
 dyn_skirtC_l_MSTR [{shapeType=Capsule, length=0.5, radius=0.5, shapeName=dyn_skirtC_l_MSTR_capsule}]
 dyn_skirtC_r_01 [{shapeType=Capsule, length=11, radius=6.5, shapeName=dyn_skirtC_r_01_capsule, center=<3 -7 0>, rotation=<rot 0 0 -11.9719>}]
 dyn_skirtC_r_02 [{shapeType=Capsule, length=11, radius=8.5, shapeName=dyn_skirtC_r_02_capsule, center=<3 -7 0>, rotation=<rot 0 0 -11.9719>}]
 dyn_skirtC_r_03 [{shapeType=Capsule, length=9, radius=10.5, shapeName=dyn_skirtC_r_03_capsule, center=<6.3513 -7 0>, rotation=<rot 0 0 -11.9719>}]
 dyn_skirtC_r_MSTR [{shapeType=Capsule, length=0.5, radius=0.5, shapeName=dyn_skirtC_r_MSTR_capsule}]
 foot_l [{shapeType=Capsule, length=15.89, radius=9.7017, center=<-8.3313 6.8359 0>, rotation=<rot 0 15.0296 90>}]
 foot_r [{shapeType=Capsule, length=15.89, radius=9.7017, center=<8.3313 -6.8359 0>, rotation=<rot 0 15.0296 -90>}]
 head [{shapeType=Capsule, length=10.6737, radius=7.3797, center=<4.68 1.9501 0>, rotation=<rot 90 0 0>}]
 neck_01 [{shapeType=Capsule, length=10.1454, radius=4.8546, center=<0 1.2878 0>, rotation=<rot 90 0 0>}]
 neck_02 [{shapeType=Capsule, length=9.2432, radius=5.7568, center=<0 2.2107 0>, rotation=<rot 90 0 0>}]
 pelvis [{shapeType=Capsule, length=5, radius=10}]
 spine_01 [{shapeType=Capsule, length=5, radius=10}]
 spine_02 [{shapeType=Capsule, length=5, radius=10}]
 spine_03 [{shapeType=Capsule, length=5, radius=10}]
 spine_04 [{shapeType=Capsule, length=5, radius=10}]
 spine_05 [{shapeType=Capsule, length=5, radius=10}]
 thigh_l [{shapeType=Capsule, length=37.598, radius=11.0411, center=<-17.6896 0 2.305>, rotation=<rot 90 0 0>}]
 thigh_r [{shapeType=Capsule, length=39.098, radius=9.5411, center=<17.896 0 0>, rotation=<rot 90 0 0>}]
 upperarm_l [{shapeType=Capsule, length=5, radius=10}]
 upperarm_r [{shapeType=Capsule, length=5, radius=10}]
}
constraints [
 {
  bone1Name=spine_01
  bone2Name=pelvis
  swing1LimitDegrees=45
  swing1Motion=Free
  swing2LimitDegrees=45
  swing2Motion=Free
  twistLimitDegrees=45
  twistMotion=Free
 }
 {
  bone1Name=spine_02
  bone2Name=spine_01
  swing1LimitDegrees=45
  swing1Motion=Free
  swing2LimitDegrees=45
  swing2Motion=Free
  twistLimitDegrees=45
  twistMotion=Free
 }
 {
  bone1Name=spine_03
  bone2Name=spine_02
  swing1LimitDegrees=45
  swing1Motion=Free
  swing2LimitDegrees=45
  swing2Motion=Free
  twistLimitDegrees=45
  twistMotion=Free
 }
 {
  bone1Name=spine_04
  bone2Name=spine_03
  swing1LimitDegrees=45
  swing1Motion=Free
  swing2LimitDegrees=45
  swing2Motion=Free
  twistLimitDegrees=45
  twistMotion=Free
 }
 {
  bone1Name=spine_05
  bone2Name=spine_04
  swing1LimitDegrees=45
  swing1Motion=Free
  swing2LimitDegrees=45
  swing2Motion=Free
  twistLimitDegrees=45
  twistMotion=Free
 }
 {
  bone1Name=clavicle_l
  bone2Name=spine_05
  swing1LimitDegrees=45
  swing1Motion=Free
  swing2LimitDegrees=45
  swing2Motion=Free
  twistLimitDegrees=45
  twistMotion=Free
 }
 {
  bone1Name=upperarm_l
  bone2Name=clavicle_l
  swing1LimitDegrees=45
  swing1Motion=Free
  swing2LimitDegrees=45
  swing2Motion=Free
  twistLimitDegrees=45
  twistMotion=Free
 }
 {
  bone1Name=clavicle_r
  bone2Name=spine_05
  swing1LimitDegrees=45
  swing1Motion=Free
  swing2LimitDegrees=45
  swing2Motion=Free
  twistLimitDegrees=45
  twistMotion=Free
 }
 {
  bone1Name=upperarm_r
  bone2Name=clavicle_r
  swing1LimitDegrees=45
  swing1Motion=Free
  swing2LimitDegrees=45
  swing2Motion=Free
  twistLimitDegrees=45
  twistMotion=Free
 }
 {
  bone1Name=neck_01
  bone2Name=spine_05
  swing1LimitDegrees=45
  swing1Motion=Free
  swing2LimitDegrees=45
  swing2Motion=Free
  twistLimitDegrees=45
  twistMotion=Free
 }
 {
  bone1Name=neck_02
  bone2Name=neck_01
  swing1LimitDegrees=45
  swing1Motion=Free
  swing2LimitDegrees=45
  swing2Motion=Free
  twistLimitDegrees=45
  twistMotion=Free
 }
 {
  bone1Name=head
  bone2Name=neck_02
  swing1LimitDegrees=45
  swing1Motion=Free
  swing2LimitDegrees=45
  swing2Motion=Free
  twistLimitDegrees=45
  twistMotion=Free
 }
 {
  bone1Name=thigh_l
  bone2Name=pelvis
  swing1LimitDegrees=45
  swing1Motion=Free
  swing2LimitDegrees=45
  swing2Motion=Free
  twistLimitDegrees=45
  twistMotion=Free
 }
 {
  bone1Name=calf_l
  bone2Name=thigh_l
  swing1LimitDegrees=45
  swing1Motion=Free
  swing2LimitDegrees=45
  swing2Motion=Free
  twistLimitDegrees=45
  twistMotion=Free
 }
 {
  bone1Name=thigh_r
  bone2Name=pelvis
  swing1LimitDegrees=45
  swing1Motion=Free
  swing2LimitDegrees=45
  swing2Motion=Free
  twistLimitDegrees=45
  twistMotion=Free
 }
 {
  bone1Name=calf_r
  bone2Name=thigh_r
  swing1LimitDegrees=45
  swing1Motion=Free
  swing2LimitDegrees=45
  swing2Motion=Free
  twistLimitDegrees=45
  twistMotion=Free
 }
 {
  bone1Name=foot_l
  bone2Name=calf_l
  swing1LimitDegrees=45
  swing1Motion=Free
  swing2LimitDegrees=45
  swing2Motion=Free
  twistLimitDegrees=45
  twistMotion=Free
 }
 {
  bone1Name=dyn_skirtA_l_MSTR
  bone2Name=pelvis
  swing1LimitDegrees=45
  swing1Motion=Free
  swing2LimitDegrees=45
  swing2Motion=Free
  twistLimitDegrees=45
  twistMotion=Free
 }
 {bone1Name=dyn_skirtA_l_01, bone2Name=dyn_skirtA_l_MSTR, swing1LimitDegrees=10, swing2LimitDegrees=10, twistLimitDegrees=0}
 {bone1Name=dyn_skirtA_l_02, bone2Name=dyn_skirtA_l_01, swing1LimitDegrees=15, swing2LimitDegrees=10, twistLimitDegrees=0}
 {bone1Name=dyn_skirtA_l_03, bone2Name=dyn_skirtA_l_02, swing1LimitDegrees=15, swing2LimitDegrees=10, twistLimitDegrees=0}
 {bone1Name=dyn_skirtA_l_04, bone2Name=dyn_skirtA_l_03, swing1LimitDegrees=15, swing2LimitDegrees=10, twistLimitDegrees=0}
 {bone1Name=dyn_skirtA_l_05, bone2Name=dyn_skirtA_l_04, swing1LimitDegrees=15, swing2LimitDegrees=10, twistLimitDegrees=0}
 {
  bone1Name=dyn_skirtB_l_MSTR
  bone2Name=pelvis
  swing1LimitDegrees=45
  swing1Motion=Free
  swing2LimitDegrees=45
  swing2Motion=Free
  twistLimitDegrees=45
  twistMotion=Free
 }
 {bone1Name=dyn_skirtB_l_01, bone2Name=dyn_skirtB_l_MSTR, swing1LimitDegrees=20, swing2LimitDegrees=15, twistLimitDegrees=0}
 {bone1Name=dyn_skirtB_l_02, bone2Name=dyn_skirtB_l_01, swing1LimitDegrees=20, swing2LimitDegrees=15, twistLimitDegrees=0}
 {bone1Name=dyn_skirtB_l_03, bone2Name=dyn_skirtB_l_02, swing1LimitDegrees=20, swing2LimitDegrees=15, twistLimitDegrees=0}
 {bone1Name=dyn_skirtB_l_04, bone2Name=dyn_skirtB_l_03, swing1LimitDegrees=20, swing2LimitDegrees=15, twistLimitDegrees=0}
 {
  bone1Name=dyn_skirtB_r_MSTR
  bone2Name=pelvis
  swing1LimitDegrees=45
  swing1Motion=Free
  swing2LimitDegrees=45
  swing2Motion=Free
  twistLimitDegrees=45
  twistMotion=Free
 }
 {bone1Name=dyn_skirtB_r_01, bone2Name=dyn_skirtB_r_MSTR, swing1LimitDegrees=10, swing2LimitDegrees=10, twistLimitDegrees=0}
 {bone1Name=dyn_skirtB_r_02, bone2Name=dyn_skirtB_r_01, swing1LimitDegrees=10, swing2LimitDegrees=10, twistLimitDegrees=0}
 {bone1Name=dyn_skirtB_r_03, bone2Name=dyn_skirtB_r_02, swing1LimitDegrees=10, swing2LimitDegrees=10, twistLimitDegrees=0}
 {
  bone1Name=dyn_skirtC_l_MSTR
  bone2Name=pelvis
  swing1LimitDegrees=45
  swing1Motion=Free
  swing2LimitDegrees=45
  swing2Motion=Free
  twistLimitDegrees=45
  twistMotion=Free
 }
 {bone1Name=dyn_skirtC_l_01, bone2Name=dyn_skirtC_l_MSTR, swing1LimitDegrees=20, swing2LimitDegrees=5, twistLimitDegrees=10}
 {bone1Name=dyn_skirtC_l_02, bone2Name=dyn_skirtC_l_01, swing1LimitDegrees=20, swing2LimitDegrees=5, twistLimitDegrees=10}
 {bone1Name=dyn_skirtC_l_03, bone2Name=dyn_skirtC_l_02, swing1LimitDegrees=27, swing2LimitDegrees=5, twistLimitDegrees=10}
 {bone1Name=dyn_skirtC_l_04, bone2Name=dyn_skirtC_l_03, swing1LimitDegrees=30, swing2LimitDegrees=5, twistLimitDegrees=10}
 {
  bone1Name=dyn_skirtC_r_MSTR
  bone2Name=pelvis
  swing1LimitDegrees=45
  swing1Motion=Free
  swing2LimitDegrees=45
  swing2Motion=Free
  twistLimitDegrees=45
  twistMotion=Free
 }
 {bone1Name=dyn_skirtC_r_01, bone2Name=dyn_skirtC_r_MSTR, swing1LimitDegrees=20, swing2LimitDegrees=0, twistLimitDegrees=0}
 {bone1Name=dyn_skirtC_r_02, bone2Name=dyn_skirtC_r_01, swing1LimitDegrees=20, swing2LimitDegrees=0, twistLimitDegrees=0}
 {bone1Name=dyn_skirtC_r_03, bone2Name=dyn_skirtC_r_02, swing1LimitDegrees=20, swing2LimitDegrees=0, twistLimitDegrees=0}
 {
  bone1Name=foot_r
  bone2Name=calf_r
  swing1LimitDegrees=45
  swing1Motion=Free
  swing2LimitDegrees=45
  swing2Motion=Free
  twistLimitDegrees=45
  twistMotion=Free
 }
]
