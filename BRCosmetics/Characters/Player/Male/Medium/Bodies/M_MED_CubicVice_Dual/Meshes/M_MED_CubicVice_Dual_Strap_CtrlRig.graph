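# Control Rig that poses the front and rear strap bone chains on the CubicVice Dual medium male cosmetic body.
# Weighted parent constraints make strap nulls and bones follow the spine, clavicles, upper arms and nearby deform bones.
# Arm and clavicle spherical pose readers, plus pistol ADS, melee, floating and glider flags, offset those nulls.
# Two B-spline IK solves shape the straps, then spatial clamps reposition right collar bones outside melee.
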ControlRigBlueprint /BRCosmetics/Characters/Player/Male/Medium/Bodies/M_MED_CubicVice_Dual/Meshes/M_MED_CubicVice_Dual_Strap_CtrlRig
graphs [
 New Function_ContainedGraph {nodes [
   RigVMFunctionEntryNode()
   > &RigUnit_ParentConstraint(
    AdvancedSettings=(RotationOrderForFilter=XZY)
    Child=(Type=Null,Name="dyn_splineStrap_01_null")
    Parents="((Item=(Type=Bone,Name=\"deform_upperarm_fr_r\"),Weight=0.0091),(Item=(Type=Bone,Name=\"spine_04\"),Weight=0.013),(Item=(Type=Bone,Name=\"dyn_collar_r\"),Weight=0.0161),(Item=(Type=Bone,Name=\"spine_03\"),Weight=0.0307),(Item=(Type=Bone,Name=\"upperarm_twist_01_r\") ...(truncated)"
    Weight=1
   )
   > &RigUnit_ParentConstraint(
    AdvancedSettings=(RotationOrderForFilter=XZY)
    Child=(Type=Null,Name="dyn_splineStrap_02_null")
    Parents="((Item=(Type=Bone,Name=\"upperarm_twist_02_r\"),Weight=0.0039),(Item=(Type=Bone,Name=\"deform_upperarm_fr_r\"),Weight=0.0063),(Item=(Type=Bone,Name=\"spine_04\"),Weight=0.0109),(Item=(Type=Bone,Name=\"spine_03\"),Weight=0.0489),(Item=(Type=Bone,Name=\"upperarm_twis ...(truncated)"
    Weight=1
   )
   > &RigUnit_ParentConstraint(
    AdvancedSettings=(RotationOrderForFilter=XZY)
    Child=(Type=Null,Name="dyn_splineStrap_03_null")
    Parents="((Item=(Type=Bone,Name=\"dyn_collar_r\"),Weight=0.0002),(Item=(Type=Bone,Name=\"deform_upperarm_bk_r\"),Weight=0.0076),(Item=(Type=Bone,Name=\"spine_04\"),Weight=0.0354),(Item=(Type=Bone,Name=\"upperarm_twist_01_r\"),Weight=0.071),(Item=(Type=Bone,Name=\"deform_upp ...(truncated)"
    Weight=1
   )
   > &RigUnit_ParentConstraint(
    AdvancedSettings=(RotationOrderForFilter=XZY)
    Child=(Type=Null,Name="dyn_splineStrap_04_null")
    Parents="((Item=(Type=Bone,Name=\"deform_upperarm_up_r\"),Weight=0.0071),(Item=(Type=Bone,Name=\"spine_02\"),Weight=0.0084),(Item=(Type=Bone,Name=\"deform_upperarm_bk_r\"),Weight=0.0154),(Item=(Type=Bone,Name=\"upperarm_twist_01_r\"),Weight=0.0195),(Item=(Type=Bone,Name=\"s ...(truncated)"
    Weight=1
   )
   > &RigUnit_ParentConstraint(
    AdvancedSettings=(RotationOrderForFilter=XZY)
    Child=(Type=Null,Name="dyn_splineStrap_05_null")
    Parents="((Item=(Type=Bone,Name=\"upperarm_twist_01_r\"),Weight=0.0034),(Item=(Type=Bone,Name=\"deform_upperarm_bk_r\"),Weight=0.0035),(Item=(Type=Bone,Name=\"clavicle_l\"),Weight=0.0046),(Item=(Type=Bone,Name=\"spine_02\"),Weight=0.0258),(Item=(Type=Bone,Name=\"clavicle_r\" ...(truncated)"
    Weight=1
   )
   > &RigUnit_ParentConstraint(
    AdvancedSettings=(RotationOrderForFilter=XZY)
    Child=(Type=Null,Name="dyn_splineStrap_06_null")
    Parents=((Item=(Type=Bone,Name="spine_01"),Weight=0.0101),(Item=(Type=Bone,Name="spine_02"),Weight=0.0396),(Item=(Type=Bone,Name="spine_05"),Weight=0.2455),(Item=(Type=Bone,Name="spine_04"),Weight=0.3096),(Item=(Type=Bone,Name="spine_03"),Weight=0.3952))
    Weight=1
   )
   > &RigUnit_ParentConstraint(
    AdvancedSettings=(RotationOrderForFilter=XZY)
    Child=(Type=Null,Name="dyn_splineStrap_07_null")
    Parents="((Item=(Type=Bone,Name=\"pelvis\"),Weight=0.0082),(Item=(Type=Bone,Name=\"spine_01\"),Weight=0.0464),(Item=(Type=Bone,Name=\"clavicle_l\"),Weight=0.079),(Item=(Type=Bone,Name=\"spine_05\"),Weight=0.088),(Item=(Type=Bone,Name=\"spine_02\"),Weight=0.1082),(Item=(Type= ...(truncated)"
    Weight=1
   )
   > &RigUnit_ParentConstraint(
    AdvancedSettings=(RotationOrderForFilter=XZY)
    Child=(Type=Null,Name="dyn_splineStrap_08_null")
    Parents="((Item=(Type=Bone,Name=\"spine_05\"),Weight=0.0048),(Item=(Type=Bone,Name=\"deform_upperarm_bk_l\"),Weight=0.0077),(Item=(Type=Bone,Name=\"pelvis\"),Weight=0.0453),(Item=(Type=Bone,Name=\"spine_01\"),Weight=0.0909),(Item=(Type=Bone,Name=\"clavicle_l\"),Weight=0.1075 ...(truncated)"
    Weight=1
   )
   > &RigUnit_ParentConstraint(
    AdvancedSettings=(RotationOrderForFilter=XZY)
    Child=(Type=Null,Name="dyn_splineStrap_09_null")
    Parents="((Item=(Type=Bone,Name=\"upperarm_twist_01_l\"),Weight=0.0019),(Item=(Type=Bone,Name=\"deform_upperarm_bk_l\"),Weight=0.0052),(Item=(Type=Bone,Name=\"clavicle_l\"),Weight=0.0531),(Item=(Type=Bone,Name=\"spine_04\"),Weight=0.0951),(Item=(Type=Bone,Name=\"pelvis\"),We ...(truncated)"
    Weight=1
   )
   > &RigUnit_ParentConstraint(
    AdvancedSettings=(RotationOrderForFilter=XZY)
    Child=(Type=Null,Name="dyn_splineStrap_10_null")
    Parents="((Item=(Type=Bone,Name=\"clavicle_l\"),Weight=0.0398),(Item=(Type=Bone,Name=\"spine_04\"),Weight=0.07),(Item=(Type=Bone,Name=\"pelvis\"),Weight=0.1174),(Item=(Type=Bone,Name=\"spine_01\"),Weight=0.1675),(Item=(Type=Bone,Name=\"spine_03\"),Weight=0.2824),(Item=(Type= ...(truncated)"
    Weight=1
   )
   > &RigUnit_ParentConstraint(
    AdvancedSettings=(RotationOrderForFilter=XZY)
    Child=(Type=Null,Name="dyn_splineStrap_11_null")
    Parents="((Item=(Type=Bone,Name=\"clavicle_l\"),Weight=0.0362),(Item=(Type=Bone,Name=\"spine_04\"),Weight=0.0642),(Item=(Type=Bone,Name=\"pelvis\"),Weight=0.12),(Item=(Type=Bone,Name=\"spine_01\"),Weight=0.1743),(Item=(Type=Bone,Name=\"spine_03\"),Weight=0.2799),(Item=(Type= ...(truncated)"
    Weight=1
   )
   > &RigUnit_ParentConstraint(
    AdvancedSettings=(RotationOrderForFilter=XZY)
    Child=(Type=Null,Name="dyn_splineStrapFr_01_ctrl_null")
    Parents="((Item=(Type=Bone,Name=\"deform_upperarm_fr_r\"),Weight=0.0091),(Item=(Type=Bone,Name=\"spine_04\"),Weight=0.013),(Item=(Type=Bone,Name=\"dyn_collar_r\"),Weight=0.0161),(Item=(Type=Bone,Name=\"spine_03\"),Weight=0.0307),(Item=(Type=Bone,Name=\"upperarm_twist_01_r\") ...(truncated)"
    Weight=1
   )
   > &RigUnit_ParentConstraint(
    AdvancedSettings=(RotationOrderForFilter=XZY)
    Child=(Type=Null,Name="dyn_splineStrapFr_02_ctrl_null")
    Parents="((Item=(Type=Bone,Name=\"spine_03\"),Weight=0.0177),(Item=(Type=Bone,Name=\"spine_04\"),Weight=0.044),(Item=(Type=Bone,Name=\"deform_pec_r\"),Weight=0.0469),(Item=(Type=Bone,Name=\"deform_upperarm_fr_r\"),Weight=0.0677),(Item=(Type=Bone,Name=\"upperarm_twist_01_r\") ...(truncated)"
    Weight=1
   )
   > &RigUnit_ParentConstraint(
    AdvancedSettings=(RotationOrderForFilter=XZY)
    Child=(Type=Null,Name="dyn_splineStrapFr_03_ctrl_null")
    Parents="((Item=(Type=Bone,Name=\"deform_upperarm_up_r\"),Weight=0.0255),(Item=(Type=Bone,Name=\"upperarm_twist_01_r\"),Weight=0.0426),(Item=(Type=Bone,Name=\"deform_upperarm_fr_r\"),Weight=0.0428),(Item=(Type=Bone,Name=\"dyn_collar_f_r\"),Weight=0.0534),(Item=(Type=Bone,N ...(truncated)"
    Weight=1
   )
   > &RigUnit_ParentConstraint(
    AdvancedSettings=(RotationOrderForFilter=XZY)
    Child=(Type=Null,Name="dyn_splineStrapFr_04_ctrl_null")
    Parents="((Item=(Type=Bone,Name=\"upperarm_twist_01_r\"),Weight=0.0103),(Item=(Type=Bone,Name=\"deform_upperarm_fr_r\"),Weight=0.0129),(Item=(Type=Bone,Name=\"spine_02\"),Weight=0.0493),(Item=(Type=Bone,Name=\"spine_03\"),Weight=0.1224),(Item=(Type=Bone,Name=\"clavicle_r\"), ...(truncated)"
    Weight=1
   )
   > &RigUnit_ParentConstraint(
    AdvancedSettings=(RotationOrderForFilter=XZY)
    Child=(Type=Null,Name="dyn_splineStrapFr_05_ctrl_null")
    Parents="((Item=(Type=Bone,Name=\"dyn_collar_f_l\"),Weight=0.0002),(Item=(Type=Bone,Name=\"dyn_collar_f_r\"),Weight=0.0012),(Item=(Type=Bone,Name=\"spine_02\"),Weight=0.0692),(Item=(Type=Bone,Name=\"clavicle_r\"),Weight=0.0854),(Item=(Type=Bone,Name=\"deform_pec_r\"),Weight= ...(truncated)"
    Weight=1
   )
   > &RigUnit_ParentConstraint(
    AdvancedSettings=(RotationOrderForFilter=XZY)
    Child=(Type=Null,Name="dyn_splineStrapFr_06_ctrl_null")
    Parents="((Item=(Type=Bone,Name=\"dyn_collar_f_l\"),Weight=0.0001),(Item=(Type=Bone,Name=\"clavicle_l\"),Weight=0.0078),(Item=(Type=Bone,Name=\"deform_pec_l\"),Weight=0.0109),(Item=(Type=Bone,Name=\"spine_01\"),Weight=0.0286),(Item=(Type=Bone,Name=\"spine_05\"),Weight=0.0871 ...(truncated)"
    Weight=1
   )
   > &RigUnit_ParentConstraint(
    AdvancedSettings=(RotationOrderForFilter=XZY)
    Child=(Type=Null,Name="dyn_splineStrapFr_07_ctrl_null")
    Parents="((Item=(Type=Bone,Name=\"pelvis\"),Weight=0.0043),(Item=(Type=Bone,Name=\"spine_05\"),Weight=0.028),(Item=(Type=Bone,Name=\"clavicle_l\"),Weight=0.0522),(Item=(Type=Bone,Name=\"deform_pec_l\"),Weight=0.0528),(Item=(Type=Bone,Name=\"spine_01\"),Weight=0.0578),(Item=( ...(truncated)"
    Weight=1
   )
   > &RigUnit_ParentConstraint(
    AdvancedSettings=(RotationOrderForFilter=XZY)
    Child=(Type=Null,Name="dyn_splineStrapFr_08_ctrl_null")
    Parents="((Item=(Type=Bone,Name=\"spine_05\"),Weight=0.0018),(Item=(Type=Bone,Name=\"pelvis\"),Weight=0.0211),(Item=(Type=Bone,Name=\"deform_pec_l\"),Weight=0.0581),(Item=(Type=Bone,Name=\"clavicle_l\"),Weight=0.0717),(Item=(Type=Bone,Name=\"spine_01\"),Weight=0.0794),(Item= ...(truncated)"
    Weight=1
   )
   > &RigUnit_ParentConstraint(
    AdvancedSettings=(RotationOrderForFilter=XZY)
    Child=(Type=Null,Name="dyn_splineStrapFr_09_ctrl_null")
    Parents="((Item=(Type=Bone,Name=\"deform_pec_l\"),Weight=0.0128),(Item=(Type=Bone,Name=\"clavicle_l\"),Weight=0.0443),(Item=(Type=Bone,Name=\"pelvis\"),Weight=0.0705),(Item=(Type=Bone,Name=\"spine_04\"),Weight=0.0893),(Item=(Type=Bone,Name=\"spine_01\"),Weight=0.1435),(Item= ...(truncated)"
    Weight=1
   )
   > &RigUnit_ParentConstraint(
    AdvancedSettings=(RotationOrderForFilter=XZY)
    Child=(Type=Null,Name="dyn_splineStrapFr_10_ctrl_null")
    Parents="((Item=(Type=Bone,Name=\"upperarm_twist_01_l\"),Weight=0.001),(Item=(Type=Bone,Name=\"deform_pec_l\"),Weight=0.0015),(Item=(Type=Bone,Name=\"clavicle_l\"),Weight=0.035),(Item=(Type=Bone,Name=\"spine_04\"),Weight=0.064),(Item=(Type=Bone,Name=\"pelvis\"),Weight=0.1082 ...(truncated)"
    Weight=1
   )
   > &RigUnit_ParentConstraint(
    AdvancedSettings=(RotationOrderForFilter=XZY)
    Child=(Type=Null,Name="dyn_splineStrapFr_11_ctrl_null")
    Parents="((Item=(Type=Bone,Name=\"clavicle_l\"),Weight=0.0362),(Item=(Type=Bone,Name=\"spine_04\"),Weight=0.0642),(Item=(Type=Bone,Name=\"pelvis\"),Weight=0.12),(Item=(Type=Bone,Name=\"spine_01\"),Weight=0.1743),(Item=(Type=Bone,Name=\"spine_03\"),Weight=0.2799),(Item=(Type= ...(truncated)"
    Weight=1
   )
   > &RigUnit_ParentConstraint(
    AdvancedSettings=(RotationOrderForFilter=XZY)
    Child=(Type=Bone,Name="dyn_splineStrap_MSTR")
    Parents="((Item=(Type=Bone,Name=\"deform_upperarm_fr_r\"),Weight=0.0091),(Item=(Type=Bone,Name=\"spine_04\"),Weight=0.013),(Item=(Type=Bone,Name=\"dyn_collar_r\"),Weight=0.0161),(Item=(Type=Bone,Name=\"spine_03\"),Weight=0.0307),(Item=(Type=Bone,Name=\"upperarm_twist_01_r\") ...(truncated)"
    Weight=1
   )
   > RigVMFunctionReturnNode()
   &RigUnit_ParentConstraint(
    AdvancedSettings=(RotationOrderForFilter=XZY)
    Child=(Type=Bone,Name="dyn_splineStrap_MSTR")
    ExecutePin=()
    Parents="((Item=(Type=Bone,Name=\"upperarm_twist_02_r\"),Weight=0.005),(Item=(Type=Bone,Name=\"spine_03\"),Weight=0.0171),(Item=(Type=Bone,Name=\"deform_upperarm_fr_r\"),Weight=0.0187),(Item=(Type=Bone,Name=\"upperarm_twist_01_r\"),Weight=0.0572),(Item=(Type=Bone,Name=\"dyn ...(truncated)"
    Weight=1
   )
   > &RigUnit_ParentConstraint(
    AdvancedSettings=(RotationOrderForFilter=XZY)
    Child=(Type=Bone,Name="dyn_splineStrap_01")
    Parents="((Item=(Type=Bone,Name=\"upperarm_twist_02_r\"),Weight=0.005),(Item=(Type=Bone,Name=\"spine_03\"),Weight=0.0171),(Item=(Type=Bone,Name=\"deform_upperarm_fr_r\"),Weight=0.0187),(Item=(Type=Bone,Name=\"upperarm_twist_01_r\"),Weight=0.0572),(Item=(Type=Bone,Name=\"dyn ...(truncated)"
    Weight=1
   )
   > &RigUnit_ParentConstraint(
    AdvancedSettings=(RotationOrderForFilter=XZY)
    Child=(Type=Bone,Name="dyn_splineStrap_02")
    Parents="((Item=(Type=Bone,Name=\"deform_upperarm_bk_r\"),Weight=0.0081),(Item=(Type=Bone,Name=\"deform_upperarm_fr_r\"),Weight=0.0099),(Item=(Type=Bone,Name=\"spine_04\"),Weight=0.0107),(Item=(Type=Bone,Name=\"spine_03\"),Weight=0.0451),(Item=(Type=Bone,Name=\"upperarm_twi ...(truncated)"
    Weight=1
   )
   > &RigUnit_ParentConstraint(
    AdvancedSettings=(RotationOrderForFilter=XZY)
    Child=(Type=Bone,Name="dyn_splineStrap_03")
    Parents="((Item=(Type=Bone,Name=\"upperarm_twist_02_r\"),Weight=0.0063),(Item=(Type=Bone,Name=\"spine_04\"),Weight=0.0397),(Item=(Type=Bone,Name=\"deform_upperarm_bk_r\"),Weight=0.0505),(Item=(Type=Bone,Name=\"deform_upperarm_up_r\"),Weight=0.0765),(Item=(Type=Bone,Name=\"u ...(truncated)"
    Weight=1
   )
   > &RigUnit_ParentConstraint(
    AdvancedSettings=(RotationOrderForFilter=XZY)
    Child=(Type=Bone,Name="dyn_splineStrap_04")
    Parents="((Item=(Type=Bone,Name=\"deform_upperarm_up_r\"),Weight=0.0103),(Item=(Type=Bone,Name=\"spine_02\"),Weight=0.0113),(Item=(Type=Bone,Name=\"upperarm_twist_01_r\"),Weight=0.0124),(Item=(Type=Bone,Name=\"deform_upperarm_bk_r\"),Weight=0.0213),(Item=(Type=Bone,Name=\"s ...(truncated)"
    Weight=1
   )
   > &RigUnit_ParentConstraint(
    AdvancedSettings=(RotationOrderForFilter=XZY)
    Child=(Type=Bone,Name="dyn_splineStrap_05")
    Parents="((Item=(Type=Bone,Name=\"upperarm_twist_01_r\"),Weight=0.0034),(Item=(Type=Bone,Name=\"deform_upperarm_bk_r\"),Weight=0.0035),(Item=(Type=Bone,Name=\"clavicle_l\"),Weight=0.0046),(Item=(Type=Bone,Name=\"spine_02\"),Weight=0.0258),(Item=(Type=Bone,Name=\"clavicle_r\" ...(truncated)"
    Weight=1
   )
   > &RigUnit_ParentConstraint(
    AdvancedSettings=(RotationOrderForFilter=XZY)
    Child=(Type=Bone,Name="dyn_splineStrap_06")
    Parents=((Item=(Type=Bone,Name="spine_01"),Weight=0.0101),(Item=(Type=Bone,Name="spine_02"),Weight=0.0396),(Item=(Type=Bone,Name="spine_05"),Weight=0.2455),(Item=(Type=Bone,Name="spine_04"),Weight=0.3096),(Item=(Type=Bone,Name="spine_03"),Weight=0.3952))
    Weight=1
   )
   > &RigUnit_ParentConstraint(
    AdvancedSettings=(RotationOrderForFilter=XZY)
    Child=(Type=Bone,Name="dyn_splineStrap_07")
    Parents="((Item=(Type=Bone,Name=\"pelvis\"),Weight=0.0082),(Item=(Type=Bone,Name=\"spine_01\"),Weight=0.0464),(Item=(Type=Bone,Name=\"clavicle_l\"),Weight=0.079),(Item=(Type=Bone,Name=\"spine_05\"),Weight=0.088),(Item=(Type=Bone,Name=\"spine_02\"),Weight=0.1082),(Item=(Type= ...(truncated)"
    Weight=1
   )
   > &RigUnit_ParentConstraint(
    AdvancedSettings=(RotationOrderForFilter=XZY)
    Child=(Type=Bone,Name="dyn_splineStrap_08")
    Parents="((Item=(Type=Bone,Name=\"spine_05\"),Weight=0.0048),(Item=(Type=Bone,Name=\"deform_upperarm_bk_l\"),Weight=0.0077),(Item=(Type=Bone,Name=\"pelvis\"),Weight=0.0453),(Item=(Type=Bone,Name=\"spine_01\"),Weight=0.0909),(Item=(Type=Bone,Name=\"clavicle_l\"),Weight=0.1075 ...(truncated)"
    Weight=1
   )
   > &RigUnit_ParentConstraint(
    AdvancedSettings=(RotationOrderForFilter=XZY)
    Child=(Type=Bone,Name="dyn_splineStrap_09")
    Parents="((Item=(Type=Bone,Name=\"deform_upperarm_bk_l\"),Weight=0.0052),(Item=(Type=Bone,Name=\"clavicle_l\"),Weight=0.0532),(Item=(Type=Bone,Name=\"spine_04\"),Weight=0.0952),(Item=(Type=Bone,Name=\"pelvis\"),Weight=0.0989),(Item=(Type=Bone,Name=\"spine_01\"),Weight=0.1431 ...(truncated)"
    Weight=1
   )
   > &RigUnit_ParentConstraint(
    AdvancedSettings=(RotationOrderForFilter=XZY)
    Child=(Type=Bone,Name="dyn_splineStrap_10")
    Parents="((Item=(Type=Bone,Name=\"clavicle_l\"),Weight=0.0398),(Item=(Type=Bone,Name=\"spine_04\"),Weight=0.07),(Item=(Type=Bone,Name=\"pelvis\"),Weight=0.1174),(Item=(Type=Bone,Name=\"spine_01\"),Weight=0.1675),(Item=(Type=Bone,Name=\"spine_03\"),Weight=0.2824),(Item=(Type= ...(truncated)"
    Weight=1
   )
   > &RigUnit_ParentConstraint(
    AdvancedSettings=(RotationOrderForFilter=XZY)
    Child=(Type=Bone,Name="dyn_splineStrap_11")
    Parents="((Item=(Type=Bone,Name=\"clavicle_l\"),Weight=0.0362),(Item=(Type=Bone,Name=\"spine_04\"),Weight=0.0642),(Item=(Type=Bone,Name=\"pelvis\"),Weight=0.12),(Item=(Type=Bone,Name=\"spine_01\"),Weight=0.1743),(Item=(Type=Bone,Name=\"spine_03\"),Weight=0.2799),(Item=(Type= ...(truncated)"
    Weight=1
   )
   > &RigUnit_ParentConstraint(
    AdvancedSettings=(RotationOrderForFilter=XZY)
    Child=(Type=Bone,Name="dyn_splineStrapFr_01")
    Parents="((Item=(Type=Bone,Name=\"upperarm_twist_02_r\"),Weight=0.005),(Item=(Type=Bone,Name=\"spine_03\"),Weight=0.0171),(Item=(Type=Bone,Name=\"deform_upperarm_fr_r\"),Weight=0.0187),(Item=(Type=Bone,Name=\"upperarm_twist_01_r\"),Weight=0.0572),(Item=(Type=Bone,Name=\"dyn ...(truncated)"
    Weight=1
   )
   > &RigUnit_ParentConstraint(
    AdvancedSettings=(RotationOrderForFilter=XZY)
    Child=(Type=Bone,Name="dyn_splineStrapFr_02")
    Parents="((Item=(Type=Bone,Name=\"spine_04\"),Weight=0.0246),(Item=(Type=Bone,Name=\"dyn_collar_r\"),Weight=0.0424),(Item=(Type=Bone,Name=\"deform_pec_r\"),Weight=0.0464),(Item=(Type=Bone,Name=\"deform_upperarm_fr_r\"),Weight=0.0691),(Item=(Type=Bone,Name=\"upperarm_twist_0 ...(truncated)"
    Weight=1
   )
   > &RigUnit_ParentConstraint(
    AdvancedSettings=(RotationOrderForFilter=XZY)
    Child=(Type=Bone,Name="dyn_splineStrapFr_03")
    Parents="((Item=(Type=Bone,Name=\"spine_02\"),Weight=0.0148),(Item=(Type=Bone,Name=\"upperarm_twist_01_r\"),Weight=0.04),(Item=(Type=Bone,Name=\"deform_upperarm_fr_r\"),Weight=0.0434),(Item=(Type=Bone,Name=\"spine_03\"),Weight=0.0621),(Item=(Type=Bone,Name=\"spine_04\"),Weig ...(truncated)"
    Weight=1
   )
   > &RigUnit_ParentConstraint(
    AdvancedSettings=(RotationOrderForFilter=XZY)
    Child=(Type=Bone,Name="dyn_splineStrapFr_04")
    Parents="((Item=(Type=Bone,Name=\"upperarm_twist_01_r\"),Weight=0.0103),(Item=(Type=Bone,Name=\"deform_upperarm_fr_r\"),Weight=0.0129),(Item=(Type=Bone,Name=\"spine_02\"),Weight=0.0493),(Item=(Type=Bone,Name=\"spine_03\"),Weight=0.1224),(Item=(Type=Bone,Name=\"clavicle_r\"), ...(truncated)"
    Weight=1
   )
   > &RigUnit_ParentConstraint(
    AdvancedSettings=(RotationOrderForFilter=XZY)
    Child=(Type=Bone,Name="dyn_splineStrapFr_05")
    Parents="((Item=(Type=Bone,Name=\"dyn_collar_f_l\"),Weight=0.0002),(Item=(Type=Bone,Name=\"dyn_collar_f_r\"),Weight=0.0012),(Item=(Type=Bone,Name=\"spine_02\"),Weight=0.0692),(Item=(Type=Bone,Name=\"clavicle_r\"),Weight=0.0854),(Item=(Type=Bone,Name=\"deform_pec_r\"),Weight= ...(truncated)"
    Weight=1
   )
   > &RigUnit_ParentConstraint(
    AdvancedSettings=(RotationOrderForFilter=XZY)
    Child=(Type=Bone,Name="dyn_splineStrapFr_06")
    Parents="((Item=(Type=Bone,Name=\"dyn_collar_f_l\"),Weight=0.0001),(Item=(Type=Bone,Name=\"clavicle_l\"),Weight=0.0078),(Item=(Type=Bone,Name=\"deform_pec_l\"),Weight=0.0109),(Item=(Type=Bone,Name=\"spine_01\"),Weight=0.0286),(Item=(Type=Bone,Name=\"spine_05\"),Weight=0.0871 ...(truncated)"
    Weight=1
   )
   > &RigUnit_ParentConstraint(
    AdvancedSettings=(RotationOrderForFilter=XZY)
    Child=(Type=Bone,Name="dyn_splineStrapFr_07")
    Parents="((Item=(Type=Bone,Name=\"pelvis\"),Weight=0.0043),(Item=(Type=Bone,Name=\"spine_05\"),Weight=0.028),(Item=(Type=Bone,Name=\"clavicle_l\"),Weight=0.0522),(Item=(Type=Bone,Name=\"deform_pec_l\"),Weight=0.0528),(Item=(Type=Bone,Name=\"spine_01\"),Weight=0.0578),(Item=( ...(truncated)"
    Weight=1
   )
   > &RigUnit_ParentConstraint(
    AdvancedSettings=(RotationOrderForFilter=XZY)
    Child=(Type=Bone,Name="dyn_splineStrapFr_08")
    Parents="((Item=(Type=Bone,Name=\"spine_05\"),Weight=0.0018),(Item=(Type=Bone,Name=\"pelvis\"),Weight=0.0211),(Item=(Type=Bone,Name=\"deform_pec_l\"),Weight=0.0581),(Item=(Type=Bone,Name=\"clavicle_l\"),Weight=0.0717),(Item=(Type=Bone,Name=\"spine_01\"),Weight=0.0794),(Item= ...(truncated)"
    Weight=1
   )
   > &RigUnit_ParentConstraint(
    AdvancedSettings=(RotationOrderForFilter=XZY)
    Child=(Type=Bone,Name="dyn_splineStrapFr_09")
    Parents="((Item=(Type=Bone,Name=\"deform_pec_l\"),Weight=0.0128),(Item=(Type=Bone,Name=\"clavicle_l\"),Weight=0.0443),(Item=(Type=Bone,Name=\"pelvis\"),Weight=0.0705),(Item=(Type=Bone,Name=\"spine_04\"),Weight=0.0893),(Item=(Type=Bone,Name=\"spine_01\"),Weight=0.1435),(Item= ...(truncated)"
    Weight=1
   )
   > &RigUnit_ParentConstraint(
    AdvancedSettings=(RotationOrderForFilter=XZY)
    Child=(Type=Bone,Name="dyn_splineStrapFr_10")
    Parents="((Item=(Type=Bone,Name=\"deform_pec_l\"),Weight=0.0015),(Item=(Type=Bone,Name=\"clavicle_l\"),Weight=0.0351),(Item=(Type=Bone,Name=\"spine_04\"),Weight=0.0641),(Item=(Type=Bone,Name=\"pelvis\"),Weight=0.1083),(Item=(Type=Bone,Name=\"spine_01\"),Weight=0.1749),(Item= ...(truncated)"
    Weight=1
   )
   > &RigUnit_ParentConstraint(
    AdvancedSettings=(RotationOrderForFilter=XZY)
    Child=(Type=Bone,Name="dyn_splineStrapFr_11")
    Parents="((Item=(Type=Bone,Name=\"clavicle_l\"),Weight=0.0362),(Item=(Type=Bone,Name=\"spine_04\"),Weight=0.0642),(Item=(Type=Bone,Name=\"pelvis\"),Weight=0.12),(Item=(Type=Bone,Name=\"spine_01\"),Weight=0.1743),(Item=(Type=Bone,Name=\"spine_03\"),Weight=0.2799),(Item=(Type= ...(truncated)"
    Weight=1
   )
  ]}
 New Function_ContainedGraph {nodes [
   RigVMFunctionEntryNode()
   > RigVMFunction_Sequence = RigVMAggregateNode()
   n1 = &RigVMFunction_ControlFlowBranch(Condition=RigVMVariableNode() {variable=is_not_playing_melee}.Value, ExecuteContext=@RigVMFunction_Sequence.C)
   ClampSpatially = /Script/RigVM.RigVMFunction_MathTransformClampSpatially(
    Axis=Y
    DebugColor=(R=1,G=0,B=0,A=1)
    DebugThickness=1
    Maximum=3
    Minimum=3
    Space=(Rotation=(X=0,Y=-0.866,Z=0,W=0.5),Translation=(X=0,Y=0,Z=0),Scale3D=(X=2,Y=1.4,Z=2))
    Type=Capsule
    bDrawDebug=false
    Space.Translation=&RigUnit_GetTransform(Item=(Type=Bone,Name="dyn_splineStrapFr_03"), bInitial=False).Transform.Translation
    Value=&RigUnit_GetTransform(Item=(Type=Bone,Name="deform_collar_f_r"), bInitial=False).Transform
   ) {
    is=struct RigVMUnitNode
    template="ClampSpatially::Execute(in Value,in Axis,in Type,in Minimum,in Maximum,in Space,in bDrawDebug,in DebugColor,in DebugThickness,out Result)"
   }
   Set Transform = /Script/ControlRig.RigUnit_SetTranslation(
    Item=(Type=Bone,Name="dyn_collar_f_r")
    Space=GlobalSpace
    Weight=1
    bInitial=False
    bPropagateToChildren=True
    Value=@ClampSpatially.Result.Translation
    ExecutePin=@n1.True
   ) {is=struct RigVMUnitNode, template="Set Transform::Execute(in Item,in Space,in bInitial,in Value,in Weight,in bPropagateToChildren)"}
   n2 = &RigVMFunction_MathTransformClampSpatially(
    Axis=Y
    DebugColor=(R=1,G=0,B=0,A=1)
    DebugThickness=1
    Maximum=0
    Minimum=3
    Space=(Rotation=(X=-0.3687,Y=-0.6275,Z=-0.5265,W=0.4394),Translation=(X=0,Y=0,Z=0),Scale3D=(X=2,Y=1.5,Z=2))
    Type=Plane
    bDrawDebug=false
    Value=&RigUnit_GetTransform(Item=(Type=Bone,Name="dyn_collar_r"), bInitial=False).Transform
    Space.Translation=&RigUnit_GetTransform(Item=(Type=Bone,Name="dyn_splineStrapFr_01"), bInitial=False).Transform.Translation
   ) {template="ClampSpatially::Execute(in Value,in Axis,in Type,in Minimum,in Maximum,in Space,in bDrawDebug,in DebugColor,in DebugThickness,out Result)"}
   > Set Transform_1 = &RigUnit_SetTranslation(
    Item=(Type=Bone,Name="deform_collar_r")
    Space=GlobalSpace
    Weight=1
    bInitial=False
    bPropagateToChildren=True
    Value=@n2.Result.Translation
   ) {template="Set Transform::Execute(in Item,in Space,in bInitial,in Value,in Weight,in bPropagateToChildren)"}
   RigVMFunctionReturnNode(ExecuteContext=@RigVMFunction_Sequence.D)
   SplineIK = RigVMFunctionReferenceNode(
    Primary Axis=(X=1,Y=0,Z=0)
    Samples Per Segment=10
    Secondary Spline Direction=(X=0,Y=10,Z=0)
    Spline Mode=BSpline
    Stretch=false
    Up Axis=(X=0,Y=1,Z=0)
    debug=false
    Bones=RigVMVariableNode() {variable=rear_bones}.Value
    Controls=RigVMVariableNode() {variable=rear_ctrls}.Value
    ExecuteContext=@RigVMFunction_Sequence.B
   )
   > SplineIK_1 = RigVMFunctionReferenceNode(
    Primary Axis=(X=1,Y=0,Z=0)
    Samples Per Segment=6
    Secondary Spline Direction=(X=0,Y=5,Z=0)
    Spline Mode=BSpline
    Up Axis=(X=0,Y=1,Z=0)
    debug=false
    Controls=RigVMVariableNode() {variable=front_ctrls}.Value
    Bones=RigVMVariableNode() {variable=front_bones}.Value
   )
   n3 = &RigUnit_SphericalPoseReader(
    ActiveRegionScaleFactors=(PositiveWidth=0.5,NegativeWidth=3,PositiveHeight=1.5,NegativeHeight=1.5)
    ActiveRegionSize=0.1
    Debug=(bDrawDebug=false,DebugScale=25,DebugSegments=20,DebugThickness=0.25)
    DriverAxis=(X=-1,Y=0,Z=0)
    DriverItem=(Type=Bone,Name="upperarm_r")
    FalloffRegionScaleFactors=(PositiveWidth=1,NegativeWidth=1,PositiveHeight=1,NegativeHeight=1)
    FalloffSize=0.4
    FlipWidthScaling=true
    OptionalParentItem=(Type=Bone)
    RotationOffset=(X=90,Y=180,Z=180)
    ExecutePin=@RigVMFunction_Sequence.A
   )
   n4 = RigVMVariableNode() {variable=isPistolADS}
   > &RigUnit_ModifyTransforms(
    ItemToModify="((Item=(Type=Null,Name=\"dyn_splineStrap_01_null\"),Transform=(Rotation=(X=0,Y=0,Z=0,W=1),Translation=(X=0,Y=0,Z=-1),Scale3D=(X=1,Y=1,Z=1))),(Item=(Type=Null,Name=\"dyn_splineStrap_02_null\"),Transform=(Rotation=(X=0,Y=0,Z=0,W=1),Translation=(X=0,Y=0,Z=-3),Sca ...(truncated)"
    WeightMinimum=0
    ItemToModify.1.Transform.Translation.Z=RigVMDispatchNode(False=-2, True=-1.5, Condition=@n4.Value) {template="DISPATCH_RigVMDispatch_If(in Condition,in True,in False,out Result)"}.Result
    ItemToModify.0.Transform.Translation.Z=RigVMDispatchNode(False=-0.5, True=-1, Condition=@n4.Value) {template="DISPATCH_RigVMDispatch_If(in Condition,in True,in False,out Result)"}.Result
    ItemToModify.2.Transform.Translation.Z=RigVMDispatchNode(False=-0.5, True=-1, Condition=@n4.Value) {template="DISPATCH_RigVMDispatch_If(in Condition,in True,in False,out Result)"}.Result
    Weight=@n3.OutputParam
   )
   > &RigUnit_ModifyTransforms(
    ItemToModify="((Item=(Type=Null,Name=\"dyn_splineStrapFr_01_ctrl_null\"),Transform=(Rotation=(X=0,Y=0,Z=0.2588,W=0.9659),Translation=(X=0,Y=0,Z=-2),Scale3D=(X=1,Y=1,Z=1))),(Item=(Type=Null,Name=\"dyn_splineStrapFr_02_ctrl_null\"),Transform=(Rotation=(X=0,Y=0,Z=0,W=1),Transl ...(truncated)"
    WeightMinimum=0
    ItemToModify.0.Transform.Translation.Z=RigVMDispatchNode(False=0.5, True=1, Condition=@n4.Value) {template="DISPATCH_RigVMDispatch_If(in Condition,in True,in False,out Result)"}.Result
    ItemToModify.1.Transform.Translation.Z=RigVMDispatchNode(False=0, True=3, Condition=@n4.Value) {template="DISPATCH_RigVMDispatch_If(in Condition,in True,in False,out Result)"}.Result
    Weight=@n3.OutputParam
   )
   > &RigUnit_ModifyTransforms(
    ItemToModify="((Item=(Type=Null,Name=\"dyn_splineStrap_10_null\"),Transform=(Rotation=(X=0,Y=0,Z=0,W=1),Translation=(X=0,Y=-1,Z=-2),Scale3D=(X=1,Y=1,Z=1))),(Item=(Type=Null,Name=\"dyn_splineStrap_05_null\"),Transform=(Rotation=(X=0,Y=0,Z=0,W=1),Translation=(X=0,Y=0,Z=0),Sca ...(truncated)"
    Weight=1
    WeightMinimum=0
   )
   > n5 = &RigUnit_SphericalPoseReader(
    ActiveRegionScaleFactors=(PositiveWidth=0.1,NegativeWidth=1.9,PositiveHeight=1.9,NegativeHeight=1.9)
    ActiveRegionSize=0.1
    Debug=(bDrawDebug=true,bDraw2D=false,bDrawLocalAxes=False,DebugScale=25,DebugSegments=20,DebugThickness=0.25)
    DriverAxis=(X=-1,Y=0,Z=0)
    DriverItem=(Type=Bone,Name="clavicle_r")
    FalloffRegionScaleFactors=(PositiveWidth=1,NegativeWidth=1,PositiveHeight=1,NegativeHeight=1)
    FalloffSize=0.2
    FlipHeightScaling=False
    FlipWidthScaling=False
    OptionalParentItem=(Type=Bone,Name="spine_05")
    RotationOffset=(X=20,Y=120,Z=30)
   )
   n6 = RigVMVariableNode() {variable=is_floating}
   > &RigUnit_ModifyTransforms(
    ItemToModify="((Item=(Type=Null,Name=\"dyn_splineStrap_01_null\"),Transform=(Rotation=(X=0,Y=0,Z=0,W=1),Translation=(X=0,Y=0,Z=0),Scale3D=(X=1,Y=1,Z=1))),(Item=(Type=Null,Name=\"dyn_splineStrap_02_null\"),Transform=(Rotation=(X=0,Y=0,Z=0,W=1),Translation=(X=0,Y=0,Z=-3),Scal ...(truncated)"
    WeightMinimum=0
    Weight=RigVMDispatchNode(False=0, Condition=&RigVMFunction_MathBoolNot(Value=@n6.Value).Result, True=@n5.OutputParam) {template="DISPATCH_RigVMDispatch_If(in Condition,in True,in False,out Result)"}.Result
   )
   > n7 = &RigUnit_SphericalPoseReader(
    ActiveRegionScaleFactors=(PositiveWidth=0.5,NegativeWidth=3,PositiveHeight=1.5,NegativeHeight=1.5)
    ActiveRegionSize=0.1
    Debug=(bDrawDebug=false,DebugScale=25,DebugSegments=20,DebugThickness=0.25)
    DriverAxis=(X=-1,Y=0,Z=0)
    DriverItem=(Type=Bone,Name="upperarm_r")
    FalloffRegionScaleFactors=(PositiveWidth=1,NegativeWidth=1,PositiveHeight=1,NegativeHeight=1)
    FalloffSize=0.4
    FlipWidthScaling=true
    OptionalParentItem=(Type=Bone,Name="clavicle_r")
    RotationOffset=(X=0,Y=50,Z=30)
   )
   > &RigUnit_ModifyTransforms(
    ItemToModify="((Item=(Type=Null,Name=\"dyn_splineStrap_01_null\"),Transform=(Rotation=(X=0,Y=0,Z=0,W=1),Translation=(X=0,Y=0,Z=-0.5),Scale3D=(X=1,Y=1,Z=1))),(Item=(Type=Null,Name=\"dyn_splineStrap_02_null\"),Transform=(Rotation=(X=0,Y=0,Z=0,W=1),Translation=(X=0,Y=0,Z=-2.5) ...(truncated)"
    WeightMinimum=0
    Weight=@n7.OutputParam
   )
   > &RigUnit_ModifyTransforms(
    ItemToModify="((Item=(Type=Null,Name=\"dyn_splineStrapFr_01_ctrl_null\"),Transform=(Rotation=(X=0,Y=0,Z=0,W=1),Translation=(X=0,Y=0,Z=1),Scale3D=(X=1,Y=1,Z=1))),(Item=(Type=Null,Name=\"dyn_splineStrapFr_02_ctrl_null\"),Transform=(Rotation=(X=0,Y=0,Z=0,W=1),Translation=(X=0, ...(truncated)"
    WeightMinimum=0
    Weight=@n7.OutputParam
   )
   RigVMFunction_MathBoolNot = &RigVMFunction_MathBoolNot(Value=RigVMVariableNode() {variable=is_not_playing_melee}.Value)
   RigVMFunction_MathBoolToFloat = &RigVMFunction_MathBoolToFloat(Value=@RigVMFunction_MathBoolNot.Result)
   TimeOffset = /Script/RigVM.RigVMFunction_TimeOffsetFloat(BufferSize=16, SecondsAgo=-0.5, TimeRange=2, Value=@RigVMFunction_MathBoolToFloat.Result) {is=struct RigVMUnitNode, template="TimeOffset::Execute(in Value,in SecondsAgo,in BufferSize,in TimeRange,out Result)"}
   > &RigUnit_ModifyTransforms(
    ItemToModify="((Item=(Type=Null,Name=\"dyn_splineStrap_08_null\"),Transform=(Rotation=(X=0,Y=0,Z=0,W=1),Translation=(X=0,Y=0,Z=-0.5),Scale3D=(X=1,Y=1,Z=1))),(Item=(Type=Null,Name=\"dyn_splineStrap_09_null\"),Transform=(Rotation=(X=0,Y=0,Z=0,W=1),Translation=(X=0,Y=0,Z=-0.5) ...(truncated)"
    WeightMinimum=0
    Weight=@TimeOffset.Result
   )
   > n8 = &RigUnit_SphericalPoseReader(
    ActiveRegionScaleFactors=(PositiveWidth=0.5,NegativeWidth=3,PositiveHeight=3,NegativeHeight=3)
    ActiveRegionSize=0.1
    Debug=(bDrawDebug=false,DebugScale=25,DebugSegments=20,DebugThickness=0.25)
    DriverAxis=(X=1,Y=0,Z=0)
    DriverItem=(Type=Bone,Name="upperarm_l")
    FalloffRegionScaleFactors=(PositiveWidth=1,NegativeWidth=1,PositiveHeight=1,NegativeHeight=1)
    FalloffSize=0.5
    OptionalParentItem=(Type=Bone)
    RotationOffset=(X=0,Y=0,Z=0)
   )
   > n9 = &RigUnit_ModifyTransforms(
    ItemToModify="((Item=(Type=Null,Name=\"dyn_splineStrapFr_07_ctrl_null\"),Transform=(Rotation=(X=0,Y=0,Z=0,W=1),Translation=(X=0,Y=1,Z=1),Scale3D=(X=1,Y=1,Z=1))),(Item=(Type=Null,Name=\"dyn_splineStrapFr_08_ctrl_null\"),Transform=(Rotation=(X=0,Y=0,Z=0,W=1),Translation=(X=0, ...(truncated)"
    WeightMinimum=0
    Weight=@n8.OutputParam
   )
   n10 = &RigVMFunction_ControlFlowBranch(ExecuteContext=@n9.ExecutePin, Condition=RigVMVariableNode() {variable=is_floating}.Value)
   RigVMFunction_ControlFlowBranch = &RigVMFunction_ControlFlowBranch(ExecuteContext=@n10.Completed, Condition=RigVMVariableNode() {variable=has_hang_glider}.Value)
   &RigUnit_ModifyTransforms(
    ItemToModify="((Item=(Type=Null,Name=\"dyn_splineStrap_01_null\"),Transform=(Rotation=(X=0,Y=0,Z=0,W=1),Translation=(X=0,Y=2,Z=-3),Scale3D=(X=1,Y=1,Z=1))),(Item=(Type=Null,Name=\"dyn_splineStrap_02_null\"),Transform=(Rotation=(X=0,Y=0,Z=0,W=1),Translation=(X=0,Y=1,Z=-3),Sca ...(truncated)"
    Weight=1
    WeightMinimum=0
    ExecutePin=@RigVMFunction_ControlFlowBranch.True
   )
   n11 = &RigUnit_SphericalPoseReader(
    ActiveRegionScaleFactors=(PositiveWidth=3,NegativeWidth=3,PositiveHeight=3,NegativeHeight=0.5)
    ActiveRegionSize=0.1
    Debug=(bDrawDebug=true,DebugScale=25,DebugSegments=20,DebugThickness=0.25)
    DriverAxis=(X=-1,Y=0,Z=0)
    DriverItem=(Type=Bone,Name="upperarm_r")
    FalloffRegionScaleFactors=(PositiveWidth=1,NegativeWidth=1,PositiveHeight=1,NegativeHeight=1)
    FalloffSize=0.4
    FlipHeightScaling=true
    FlipWidthScaling=false
    OptionalParentItem=(Type=Bone,Name="spine_05")
    RotationOffset=(X=150,Y=-70,Z=0)
    ExecutePin=@n10.True
   )
   > &RigUnit_ModifyTransforms(
    ItemToModify="((Item=(Type=Null,Name=\"dyn_splineStrap_01_null\"),Transform=(Rotation=(X=0,Y=0,Z=0,W=1),Translation=(X=1,Y=0,Z=-1),Scale3D=(X=1,Y=1,Z=1))),(Item=(Type=Null,Name=\"dyn_splineStrap_02_null\"),Transform=(Rotation=(X=0,Y=0,Z=0,W=1),Translation=(X=1,Y=0,Z=-3),Sca ...(truncated)"
    WeightMinimum=0
    Weight=@n11.OutputParam
   )
   &RigUnit_OffsetTransformForItem(
    ExecutePin=()
    Item=(Type=Null,Name="dyn_splineStrapFr_03_null")
    OffsetTransform=(Rotation=(X=0,Y=0,Z=0,W=1),Translation=(X=0,Y=0,Z=0),Scale3D=(X=1,Y=1,Z=1))
    Weight=1
    bPropagateToChildren=True
   )
  ]}
 RigVMFunctionLibrary {nodes [
   Pin default = RigVMCollapseNode(ExecuteContext=())
   Solve Strap Spline IK = RigVMCollapseNode(ExecuteContext=())
  ]}
 RigVMFunction_Sequence_1_ContainedGraph {nodes [
   n1 = RigVMFunctionEntryNode()
   RigVMFunction_Sequence = &RigVMFunction_Sequence(ExecuteContext=@n1.ExecuteContext)
   n2 = &RigVMFunction_Sequence(ExecuteContext=@RigVMFunction_Sequence.B)
   n3 = &RigVMFunction_Sequence(ExecuteContext=@n2.B)
   RigVMFunctionReturnNode(A=@RigVMFunction_Sequence.A, B=@n2.A, C=@n3.A, D=@n3.B)
  ]}
 RigVMModel {nodes [
   &RigUnit_ParentConstraint(
    AdvancedSettings=(RotationOrderForFilter=XZY)
    Child=(Type=Bone,Name="dyn_splineStrap_MSTR")
    ExecutePin=()
    Parents="((Item=(Type=Bone,Name=\"upperarm_twist_02_r\"),Weight=0.005),(Item=(Type=Bone,Name=\"spine_03\"),Weight=0.0171),(Item=(Type=Bone,Name=\"deform_upperarm_fr_r\"),Weight=0.0187),(Item=(Type=Bone,Name=\"upperarm_twist_01_r\"),Weight=0.0572),(Item=(Type=Bone,Name=\"dyn ...(truncated)"
    Weight=1
   )
   > &RigUnit_ParentConstraint(
    AdvancedSettings=(RotationOrderForFilter=XZY)
    Child=(Type=Bone,Name="dyn_splineStrap_01")
    Parents="((Item=(Type=Bone,Name=\"upperarm_twist_02_r\"),Weight=0.005),(Item=(Type=Bone,Name=\"spine_03\"),Weight=0.0171),(Item=(Type=Bone,Name=\"deform_upperarm_fr_r\"),Weight=0.0187),(Item=(Type=Bone,Name=\"upperarm_twist_01_r\"),Weight=0.0572),(Item=(Type=Bone,Name=\"dyn ...(truncated)"
    Weight=1
   )
   > &RigUnit_ParentConstraint(
    AdvancedSettings=(RotationOrderForFilter=XZY)
    Child=(Type=Bone,Name="dyn_splineStrap_02")
    Parents="((Item=(Type=Bone,Name=\"deform_upperarm_bk_r\"),Weight=0.0081),(Item=(Type=Bone,Name=\"deform_upperarm_fr_r\"),Weight=0.0099),(Item=(Type=Bone,Name=\"spine_04\"),Weight=0.0107),(Item=(Type=Bone,Name=\"spine_03\"),Weight=0.0451),(Item=(Type=Bone,Name=\"upperarm_twi ...(truncated)"
    Weight=1
   )
   > &RigUnit_ParentConstraint(
    AdvancedSettings=(RotationOrderForFilter=XZY)
    Child=(Type=Bone,Name="dyn_splineStrap_03")
    Parents="((Item=(Type=Bone,Name=\"upperarm_twist_02_r\"),Weight=0.0063),(Item=(Type=Bone,Name=\"spine_04\"),Weight=0.0397),(Item=(Type=Bone,Name=\"deform_upperarm_bk_r\"),Weight=0.0505),(Item=(Type=Bone,Name=\"deform_upperarm_up_r\"),Weight=0.0765),(Item=(Type=Bone,Name=\"u ...(truncated)"
    Weight=1
   )
   > &RigUnit_ParentConstraint(
    AdvancedSettings=(RotationOrderForFilter=XZY)
    Child=(Type=Bone,Name="dyn_splineStrap_04")
    Parents="((Item=(Type=Bone,Name=\"deform_upperarm_up_r\"),Weight=0.0103),(Item=(Type=Bone,Name=\"spine_02\"),Weight=0.0113),(Item=(Type=Bone,Name=\"upperarm_twist_01_r\"),Weight=0.0124),(Item=(Type=Bone,Name=\"deform_upperarm_bk_r\"),Weight=0.0213),(Item=(Type=Bone,Name=\"s ...(truncated)"
    Weight=1
   )
   > &RigUnit_ParentConstraint(
    AdvancedSettings=(RotationOrderForFilter=XZY)
    Child=(Type=Bone,Name="dyn_splineStrap_05")
    Parents="((Item=(Type=Bone,Name=\"upperarm_twist_01_r\"),Weight=0.0034),(Item=(Type=Bone,Name=\"deform_upperarm_bk_r\"),Weight=0.0035),(Item=(Type=Bone,Name=\"clavicle_l\"),Weight=0.0046),(Item=(Type=Bone,Name=\"spine_02\"),Weight=0.0258),(Item=(Type=Bone,Name=\"clavicle_r\" ...(truncated)"
    Weight=1
   )
   > &RigUnit_ParentConstraint(
    AdvancedSettings=(RotationOrderForFilter=XZY)
    Child=(Type=Bone,Name="dyn_splineStrap_06")
    Parents=((Item=(Type=Bone,Name="spine_01"),Weight=0.0101),(Item=(Type=Bone,Name="spine_02"),Weight=0.0396),(Item=(Type=Bone,Name="spine_05"),Weight=0.2455),(Item=(Type=Bone,Name="spine_04"),Weight=0.3096),(Item=(Type=Bone,Name="spine_03"),Weight=0.3952))
    Weight=1
   )
   > &RigUnit_ParentConstraint(
    AdvancedSettings=(RotationOrderForFilter=XZY)
    Child=(Type=Bone,Name="dyn_splineStrap_07")
    Parents="((Item=(Type=Bone,Name=\"pelvis\"),Weight=0.0082),(Item=(Type=Bone,Name=\"spine_01\"),Weight=0.0464),(Item=(Type=Bone,Name=\"clavicle_l\"),Weight=0.079),(Item=(Type=Bone,Name=\"spine_05\"),Weight=0.088),(Item=(Type=Bone,Name=\"spine_02\"),Weight=0.1082),(Item=(Type= ...(truncated)"
    Weight=1
   )
   > &RigUnit_ParentConstraint(
    AdvancedSettings=(RotationOrderForFilter=XZY)
    Child=(Type=Bone,Name="dyn_splineStrap_08")
    Parents="((Item=(Type=Bone,Name=\"spine_05\"),Weight=0.0048),(Item=(Type=Bone,Name=\"deform_upperarm_bk_l\"),Weight=0.0077),(Item=(Type=Bone,Name=\"pelvis\"),Weight=0.0453),(Item=(Type=Bone,Name=\"spine_01\"),Weight=0.0909),(Item=(Type=Bone,Name=\"clavicle_l\"),Weight=0.1075 ...(truncated)"
    Weight=1
   )
   > &RigUnit_ParentConstraint(
    AdvancedSettings=(RotationOrderForFilter=XZY)
    Child=(Type=Bone,Name="dyn_splineStrap_09")
    Parents="((Item=(Type=Bone,Name=\"deform_upperarm_bk_l\"),Weight=0.0052),(Item=(Type=Bone,Name=\"clavicle_l\"),Weight=0.0532),(Item=(Type=Bone,Name=\"spine_04\"),Weight=0.0952),(Item=(Type=Bone,Name=\"pelvis\"),Weight=0.0989),(Item=(Type=Bone,Name=\"spine_01\"),Weight=0.1431 ...(truncated)"
    Weight=1
   )
   > &RigUnit_ParentConstraint(
    AdvancedSettings=(RotationOrderForFilter=XZY)
    Child=(Type=Bone,Name="dyn_splineStrap_10")
    Parents="((Item=(Type=Bone,Name=\"clavicle_l\"),Weight=0.0398),(Item=(Type=Bone,Name=\"spine_04\"),Weight=0.07),(Item=(Type=Bone,Name=\"pelvis\"),Weight=0.1174),(Item=(Type=Bone,Name=\"spine_01\"),Weight=0.1675),(Item=(Type=Bone,Name=\"spine_03\"),Weight=0.2824),(Item=(Type= ...(truncated)"
    Weight=1
   )
   > &RigUnit_ParentConstraint(
    AdvancedSettings=(RotationOrderForFilter=XZY)
    Child=(Type=Bone,Name="dyn_splineStrap_11")
    Parents="((Item=(Type=Bone,Name=\"clavicle_l\"),Weight=0.0362),(Item=(Type=Bone,Name=\"spine_04\"),Weight=0.0642),(Item=(Type=Bone,Name=\"pelvis\"),Weight=0.12),(Item=(Type=Bone,Name=\"spine_01\"),Weight=0.1743),(Item=(Type=Bone,Name=\"spine_03\"),Weight=0.2799),(Item=(Type= ...(truncated)"
    Weight=1
   )
   > &RigUnit_ParentConstraint(
    AdvancedSettings=(RotationOrderForFilter=XZY)
    Child=(Type=Bone,Name="dyn_splineStrapFr_01")
    Parents="((Item=(Type=Bone,Name=\"upperarm_twist_02_r\"),Weight=0.005),(Item=(Type=Bone,Name=\"spine_03\"),Weight=0.0171),(Item=(Type=Bone,Name=\"deform_upperarm_fr_r\"),Weight=0.0187),(Item=(Type=Bone,Name=\"upperarm_twist_01_r\"),Weight=0.0572),(Item=(Type=Bone,Name=\"dyn ...(truncated)"
    Weight=1
   )
   > &RigUnit_ParentConstraint(
    AdvancedSettings=(RotationOrderForFilter=XZY)
    Child=(Type=Bone,Name="dyn_splineStrapFr_02")
    Parents="((Item=(Type=Bone,Name=\"spine_04\"),Weight=0.0246),(Item=(Type=Bone,Name=\"dyn_collar_r\"),Weight=0.0424),(Item=(Type=Bone,Name=\"deform_pec_r\"),Weight=0.0464),(Item=(Type=Bone,Name=\"deform_upperarm_fr_r\"),Weight=0.0691),(Item=(Type=Bone,Name=\"upperarm_twist_0 ...(truncated)"
    Weight=1
   )
   > &RigUnit_ParentConstraint(
    AdvancedSettings=(RotationOrderForFilter=XZY)
    Child=(Type=Bone,Name="dyn_splineStrapFr_03")
    Parents="((Item=(Type=Bone,Name=\"spine_02\"),Weight=0.0148),(Item=(Type=Bone,Name=\"upperarm_twist_01_r\"),Weight=0.04),(Item=(Type=Bone,Name=\"deform_upperarm_fr_r\"),Weight=0.0434),(Item=(Type=Bone,Name=\"spine_03\"),Weight=0.0621),(Item=(Type=Bone,Name=\"spine_04\"),Weig ...(truncated)"
    Weight=1
   )
   > &RigUnit_ParentConstraint(
    AdvancedSettings=(RotationOrderForFilter=XZY)
    Child=(Type=Bone,Name="dyn_splineStrapFr_04")
    Parents="((Item=(Type=Bone,Name=\"upperarm_twist_01_r\"),Weight=0.0103),(Item=(Type=Bone,Name=\"deform_upperarm_fr_r\"),Weight=0.0129),(Item=(Type=Bone,Name=\"spine_02\"),Weight=0.0493),(Item=(Type=Bone,Name=\"spine_03\"),Weight=0.1224),(Item=(Type=Bone,Name=\"clavicle_r\"), ...(truncated)"
    Weight=1
   )
   > &RigUnit_ParentConstraint(
    AdvancedSettings=(RotationOrderForFilter=XZY)
    Child=(Type=Bone,Name="dyn_splineStrapFr_05")
    Parents="((Item=(Type=Bone,Name=\"dyn_collar_f_l\"),Weight=0.0002),(Item=(Type=Bone,Name=\"dyn_collar_f_r\"),Weight=0.0012),(Item=(Type=Bone,Name=\"spine_02\"),Weight=0.0692),(Item=(Type=Bone,Name=\"clavicle_r\"),Weight=0.0854),(Item=(Type=Bone,Name=\"deform_pec_r\"),Weight= ...(truncated)"
    Weight=1
   )
   > &RigUnit_ParentConstraint(
    AdvancedSettings=(RotationOrderForFilter=XZY)
    Child=(Type=Bone,Name="dyn_splineStrapFr_06")
    Parents="((Item=(Type=Bone,Name=\"dyn_collar_f_l\"),Weight=0.0001),(Item=(Type=Bone,Name=\"clavicle_l\"),Weight=0.0078),(Item=(Type=Bone,Name=\"deform_pec_l\"),Weight=0.0109),(Item=(Type=Bone,Name=\"spine_01\"),Weight=0.0286),(Item=(Type=Bone,Name=\"spine_05\"),Weight=0.0871 ...(truncated)"
    Weight=1
   )
   > &RigUnit_ParentConstraint(
    AdvancedSettings=(RotationOrderForFilter=XZY)
    Child=(Type=Bone,Name="dyn_splineStrapFr_07")
    Parents="((Item=(Type=Bone,Name=\"pelvis\"),Weight=0.0043),(Item=(Type=Bone,Name=\"spine_05\"),Weight=0.028),(Item=(Type=Bone,Name=\"clavicle_l\"),Weight=0.0522),(Item=(Type=Bone,Name=\"deform_pec_l\"),Weight=0.0528),(Item=(Type=Bone,Name=\"spine_01\"),Weight=0.0578),(Item=( ...(truncated)"
    Weight=1
   )
   > &RigUnit_ParentConstraint(
    AdvancedSettings=(RotationOrderForFilter=XZY)
    Child=(Type=Bone,Name="dyn_splineStrapFr_08")
    Parents="((Item=(Type=Bone,Name=\"spine_05\"),Weight=0.0018),(Item=(Type=Bone,Name=\"pelvis\"),Weight=0.0211),(Item=(Type=Bone,Name=\"deform_pec_l\"),Weight=0.0581),(Item=(Type=Bone,Name=\"clavicle_l\"),Weight=0.0717),(Item=(Type=Bone,Name=\"spine_01\"),Weight=0.0794),(Item= ...(truncated)"
    Weight=1
   )
   > &RigUnit_ParentConstraint(
    AdvancedSettings=(RotationOrderForFilter=XZY)
    Child=(Type=Bone,Name="dyn_splineStrapFr_09")
    Parents="((Item=(Type=Bone,Name=\"deform_pec_l\"),Weight=0.0128),(Item=(Type=Bone,Name=\"clavicle_l\"),Weight=0.0443),(Item=(Type=Bone,Name=\"pelvis\"),Weight=0.0705),(Item=(Type=Bone,Name=\"spine_04\"),Weight=0.0893),(Item=(Type=Bone,Name=\"spine_01\"),Weight=0.1435),(Item= ...(truncated)"
    Weight=1
   )
   > &RigUnit_ParentConstraint(
    AdvancedSettings=(RotationOrderForFilter=XZY)
    Child=(Type=Bone,Name="dyn_splineStrapFr_10")
    Parents="((Item=(Type=Bone,Name=\"deform_pec_l\"),Weight=0.0015),(Item=(Type=Bone,Name=\"clavicle_l\"),Weight=0.0351),(Item=(Type=Bone,Name=\"spine_04\"),Weight=0.0641),(Item=(Type=Bone,Name=\"pelvis\"),Weight=0.1083),(Item=(Type=Bone,Name=\"spine_01\"),Weight=0.1749),(Item= ...(truncated)"
    Weight=1
   )
   > &RigUnit_ParentConstraint(
    AdvancedSettings=(RotationOrderForFilter=XZY)
    Child=(Type=Bone,Name="dyn_splineStrapFr_11")
    Parents="((Item=(Type=Bone,Name=\"clavicle_l\"),Weight=0.0362),(Item=(Type=Bone,Name=\"spine_04\"),Weight=0.0642),(Item=(Type=Bone,Name=\"pelvis\"),Weight=0.12),(Item=(Type=Bone,Name=\"spine_01\"),Weight=0.1743),(Item=(Type=Bone,Name=\"spine_03\"),Weight=0.2799),(Item=(Type= ...(truncated)"
    Weight=1
   )
   RigUnit_BeginExecution = &RigUnit_BeginExecution()
   RigUnit_ItemArray = &RigUnit_ItemArray(Items="((Type=Bone,Name=\"dyn_splineStrap_01\"),(Type=Bone,Name=\"dyn_splineStrap_02\"),(Type=Bone,Name=\"dyn_splineStrap_03\"),(Type=Bone,Name=\"dyn_splineStrap_04\"),(Type=Bone,Name=\"dyn_splineStrap_05\"),(Type=Bone,Name=\"dyn_splineStrap_06\"),(Type=Bone,Name=\"dyn_spline ...(truncated)")
   RigVMVariableNode(ExecuteContext=@RigUnit_BeginExecution.ExecutePin, Value=@RigUnit_ItemArray.Items) {variable=rear_bones}
   n1 = &RigUnit_ItemArray(Items="((Type=Bone,Name=\"dyn_splineStrapFr_01\"),(Type=Bone,Name=\"dyn_splineStrapFr_02\"),(Type=Bone,Name=\"dyn_splineStrapFr_03\"),(Type=Bone,Name=\"dyn_splineStrapFr_04\"),(Type=Bone,Name=\"dyn_splineStrapFr_05\"),(Type=Bone,Name=\"dyn_splineStrapFr_06\"),(Type=Bone,Name ...(truncated)")
   > RigVMVariableNode(Value=@n1.Items) {variable=front_bones}
   n2 = &RigUnit_ItemArray(Items="((Type=Control,Name=\"dyn_splineStrapFr_01_ctrl\"),(Type=Control,Name=\"dyn_splineStrapFr_02_ctrl\"),(Type=Control,Name=\"dyn_splineStrapFr_03_ctrl\"),(Type=Control,Name=\"dyn_splineStrapFr_04_ctrl\"),(Type=Control,Name=\"dyn_splineStrapFr_05_ctrl\"),(Type=Control,N ...(truncated)")
   > RigVMVariableNode(Value=@n2.Items) {variable=front_ctrls}
   n3 = &RigUnit_ItemArray(Items="((Type=Control,Name=\"dyn_splineStrap_01_ctrl\"),(Type=Control,Name=\"dyn_splineStrap_02_ctrl\"),(Type=Control,Name=\"dyn_splineStrap_03_ctrl\"),(Type=Control,Name=\"dyn_splineStrap_04_ctrl\"),(Type=Control,Name=\"dyn_splineStrap_05_ctrl\"),(Type=Control,Name=\"dyn_s ...(truncated)")
   > n4 = RigVMVariableNode(Value=@n3.Items) {variable=rear_ctrls}
   RigVMFunction_ControlFlowBranch = &RigVMFunction_ControlFlowBranch(ExecuteContext=@n4.ExecuteContext, Condition=RigVMVariableNode() {variable=is rban enabled}.Value)
   Pin default_1 = RigVMFunctionReferenceNode(ExecuteContext=@RigVMFunction_ControlFlowBranch.False)
   Pin default = RigVMFunctionReferenceNode(ExecuteContext=@RigVMFunction_ControlFlowBranch.True)
   > Solve Strap Spline IK = RigVMFunctionReferenceNode()
  ]}
]
hierarchy [
 root {loc=<0 0 0>, rot=<0 0 0 1>}
 pelvis {parent=root, loc=<0 -0.7071 93.1191>, rot=<0.0127 0.707 -0.0127 -0.707>}
 spine_01 {parent=pelvis, loc=<6.7494 -0.6388 0>, rot=<0 0 0.034 -0.9994>}
 spine_02 {parent=spine_01, loc=<7.252 -0.035 0>, rot=<0 0 0.0238 -0.9997>}
 spine_03 {parent=spine_02, loc=<6.0106 0 0>, rot=<0 0 0.0149 -0.9999>}
 spine_04 {parent=spine_03, loc=<8.1023 0 0>, rot=<0 0 -0.0132 -0.9999>}
 spine_05 {parent=spine_04, loc=<10.0142 0 0>, rot=<0 0 -0.0015 -1>}
 attach_backpack {parent=spine_05, loc=<-0.123 -13.473 0>, rot=<-0.0093 -0.707 0.0093 -0.707>}
 neck_01 {parent=spine_05, loc=<11.3138 0.1529 0>, rot=<0 0 -0.2003 -0.9797>}
 neck_02 {parent=neck_01, loc=<4.5627 0 0>, rot=<0 0 0.0504 -0.9987>}
 head {parent=neck_02, loc=<5.3914 0 0>, rot=<0 0 0.115 -0.9934>}
 clavicle_l {parent=spine_05, loc=<4.3575 -0.1466 -2.3223>, rot=<0.0465 -0.7289 -0.009 -0.6829>}
 upperarm_l {parent=clavicle_l, loc=<15.5111 0 0>, rot=<0.0226 -0.3234 -0.0088 -0.9459>}
 lowerarm_l {parent=upperarm_l, loc=<26.6514 0 0>, rot=<0.0391 -0.0099 -0.2453 -0.9686>}
 hand_l {parent=lowerarm_l, loc=<22.3491 0 0>, rot=<-0.7081 -0.0507 0.0975 -0.6975>}
 thumb_01_l {parent=hand_l, loc=<1.3118 -0.3191 -2.459>, rot=<0.6546 -0.4042 0.0664 -0.6354>}
 thumb_02_l {parent=thumb_01_l, loc=<5.9336 -0.0448 0.0401>, rot=<-0.0037 0.0213 0.298 -0.9543>}
 thumb_03_l {parent=thumb_02_l, loc=<2.2223 0.3682 -0.0907>, rot=<0.0118 -0.013 0.0066 -0.9998>}
 index_metacarpal_l {parent=hand_l, loc=<1.6535 0.0183 -2.96>, rot=<0.0158 -0.0124 -0.0173 -0.9996>}
 index_01_l {parent=index_metacarpal_l, loc=<8.3008 0.0357 -0.0524>, rot=<-0.0044 0.0485 0.0924 -0.9945>}
 index_02_l {parent=index_01_l, loc=<3.3587 -0.3004 0.0452>, rot=<-0.002 0.0027 0.3225 -0.9466>}
 index_03_l {parent=index_02_l, loc=<2.5578 -0.123 -0.0252>, rot=<-0.002 0.0067 0.2558 -0.9667>}
 middle_metacarpal_l {parent=hand_l, loc=<2.1856 -0.1725 -0.5984>, rot=<0.006 0.0168 -0.0329 -0.9993>}
 middle_01_l {parent=middle_metacarpal_l, loc=<7.3275 -0.2447 0.0129>, rot=<-0.0208 0.0042 0.0858 -0.9961>}
 middle_02_l {parent=middle_01_l, loc=<4.0123 0.1401 0.1175>, rot=<-0.0218 0.0258 0.3512 -0.9357>}
 middle_03_l {parent=middle_02_l, loc=<2.6077 0.0534 0.1373>, rot=<0.0082 0.0188 0.1708 -0.9851>}
 ring_metacarpal_l {parent=hand_l, loc=<2.2074 -0.3625 1.3638>, rot=<-0.0284 0.0722 -0.0458 -0.9959>}
 ring_01_l {parent=ring_metacarpal_l, loc=<7.0295 -0.6197 -0.1446>, rot=<-0.0489 -0.0252 0.1892 -0.9804>}
 ring_02_l {parent=ring_01_l, loc=<3.6174 0.4647 -0.1027>, rot=<0 0 0.2806 -0.9598>}
 ring_03_l {parent=ring_02_l, loc=<2.4152 0.114 0.0199>, rot=<0 0 0.1189 -0.9929>}
 pinky_metacarpal_l {parent=hand_l, loc=<1.7616 -0.9753 3.3814>, rot=<-0.0397 0.0814 -0.069 -0.9935>}
 pinky_01_l {parent=pinky_metacarpal_l, loc=<7.1709 -0.4831 -0.1094>, rot=<-0.0654 -0.0035 0.2242 -0.9723>}
 pinky_02_l {parent=pinky_01_l, loc=<2.8274 0.0893 -0.1043>, rot=<0.0037 0.0011 0.2854 -0.9584>}
 pinky_03_l {parent=pinky_02_l, loc=<2.2046 0.0307 -0.0078>, rot=<-0.0002 0.0098 0.1445 -0.9895>}
 deform_wrist_dn_base_l {parent=hand_l, loc=<0 -2.2342 0>, rot=<0 0 0 -1>}
 deform_wrist_dn_l {parent=deform_wrist_dn_base_l, loc=<0 0 0>, rot=<0 0 0 1>}
 deform_wrist_up_base_l {parent=hand_l, loc=<0 2.2682 0>, rot=<0 0 0 -1>}
 deform_wrist_up_l {parent=deform_wrist_up_base_l, loc=<0 0 0>, rot=<0 0 0 1>}
 lowerarm_twist_02_l {parent=lowerarm_l, loc=<9.6199 0 0>, rot=<0 0 0 -1>}
 lowerarm_twist_01_l {parent=lowerarm_l, loc=<17.8549 0 0>, rot=<0 0 0 -1>}
 deform_elbow_in_base_l {parent=lowerarm_l, loc=<0.2112 5.259 0.8274>, rot=<0 0 0 -1>}
 deform_elbow_in_l {parent=deform_elbow_in_base_l, loc=<0 0 0>, rot=<0 0 0 1>}
 deform_elbow_out_base_l {parent=lowerarm_l, loc=<0 -4 0>, rot=<0 0 0 -1>}
 deform_elbow_out_l {parent=deform_elbow_out_base_l, loc=<0 0 0>, rot=<0 0 0 1>}
 upperarm_twist_01_l {parent=upperarm_l, loc=<0.0041 0 0>, rot=<0 0 0 -1>}
 deform_upperarm_bk_base_l {parent=upperarm_twist_01_l, loc=<1.0813 -7.2193 -1.9453>, rot=<0 0 0 1>}
 deform_upperarm_bk_l {parent=deform_upperarm_bk_base_l, loc=<0 0 0>, rot=<0 0 0 1>}
 deform_upperarm_fr_base_l {parent=upperarm_twist_01_l, loc=<-0.3691 7.0506 -2.765>, rot=<0 0 0 1>}
 deform_upperarm_fr_l {parent=deform_upperarm_fr_base_l, loc=<0 0 0>, rot=<0 0 0 1>}
 deform_upperarm_up_base_l {parent=upperarm_twist_01_l, loc=<-3.2216 -1.1391 3.3311>, rot=<0 0 0 1>}
 deform_upperarm_up_l {parent=deform_upperarm_up_base_l, loc=<0 0 0>, rot=<0 0 0 1>}
 upperarm_twist_02_l {parent=upperarm_l, loc=<13.7941 0 0>, rot=<0 0 0 -1>}
 attach_cape {parent=spine_05, loc=<9.017 -11.714 0>, rot=<0.0616 -0.7044 -0.0616 -0.7044>}
 clavicle_r {parent=spine_05, loc=<4.3575 -0.1466 2.3223>, rot=<0.7289 0.0465 -0.6829 0.009>}
 upperarm_r {parent=clavicle_r, loc=<-15.5111 0 0>, rot=<0.0226 -0.3234 -0.0088 -0.9459>}
 lowerarm_r {parent=upperarm_r, loc=<-26.6514 0 0>, rot=<0.0391 -0.0099 -0.2453 -0.9686>}
 hand_r {parent=lowerarm_r, loc=<-22.3491 0 0>, rot=<-0.7081 -0.0507 0.0975 -0.6975>}
 thumb_01_r {parent=hand_r, loc=<-1.3118 0.3191 2.459>, rot=<0.6546 -0.4042 0.0664 -0.6354>}
 thumb_02_r {parent=thumb_01_r, loc=<-5.9336 0.0448 -0.0401>, rot=<-0.0037 0.0213 0.298 -0.9543>}
 thumb_03_r {parent=thumb_02_r, loc=<-2.2223 -0.3682 0.0907>, rot=<0.0118 -0.013 0.0066 -0.9998>}
 index_metacarpal_r {parent=hand_r, loc=<-1.6535 -0.0183 2.96>, rot=<0.0158 -0.0124 -0.0173 -0.9996>}
 index_01_r {parent=index_metacarpal_r, loc=<-8.3008 -0.0357 0.0524>, rot=<-0.0044 0.0485 0.0924 -0.9945>}
 index_02_r {parent=index_01_r, loc=<-3.3587 0.3004 -0.0452>, rot=<-0.002 0.0027 0.3225 -0.9466>}
 index_03_r {parent=index_02_r, loc=<-2.5578 0.123 0.0252>, rot=<-0.002 0.0067 0.2558 -0.9667>}
 middle_metacarpal_r {parent=hand_r, loc=<-2.1856 0.1725 0.5984>, rot=<0.006 0.0168 -0.0329 -0.9993>}
 middle_01_r {parent=middle_metacarpal_r, loc=<-7.3275 0.2447 -0.0129>, rot=<-0.0208 0.0042 0.0858 -0.9961>}
 middle_02_r {parent=middle_01_r, loc=<-4.0123 -0.1401 -0.1175>, rot=<-0.0218 0.0258 0.3512 -0.9357>}
 middle_03_r {parent=middle_02_r, loc=<-2.6077 -0.0534 -0.1373>, rot=<0.0082 0.0188 0.1708 -0.9851>}
 ring_metacarpal_r {parent=hand_r, loc=<-2.2074 0.3625 -1.3638>, rot=<-0.0284 0.0722 -0.0458 -0.9959>}
 ring_01_r {parent=ring_metacarpal_r, loc=<-7.0295 0.6197 0.1446>, rot=<-0.0489 -0.0252 0.1892 -0.9804>}
 ring_02_r {parent=ring_01_r, loc=<-3.6174 -0.4647 0.1027>, rot=<0 0 0.2806 -0.9598>}
 ring_03_r {parent=ring_02_r, loc=<-2.4152 -0.114 -0.0199>, rot=<0 0 0.1189 -0.9929>}
 pinky_metacarpal_r {parent=hand_r, loc=<-1.7616 0.9753 -3.3814>, rot=<-0.0397 0.0814 -0.069 -0.9935>}
 pinky_01_r {parent=pinky_metacarpal_r, loc=<-7.1709 0.4831 0.1094>, rot=<-0.0654 -0.0035 0.2242 -0.9723>}
 pinky_02_r {parent=pinky_01_r, loc=<-2.8274 -0.0893 0.1043>, rot=<0.0037 0.0011 0.2854 -0.9584>}
 pinky_03_r {parent=pinky_02_r, loc=<-2.2046 -0.0307 0.0078>, rot=<-0.0002 0.0098 0.1445 -0.9895>}
 deform_wrist_dn_base_r {parent=hand_r, loc=<-0.0004 2.2345 -0.0001>, rot=<0 0 0 -1>}
 deform_wrist_dn_r {parent=deform_wrist_dn_base_r, loc=<0 0 0>, rot=<0 0 0 1>}
 deform_wrist_up_base_r {parent=hand_r, loc=<0.0003 -2.2684 0.0001>, rot=<0 0 0 -1>}
 deform_wrist_up_r {parent=deform_wrist_up_base_r, loc=<0 0 0>, rot=<0 0 0 1>}
 lowerarm_twist_02_r {parent=lowerarm_r, loc=<-9.6199 0 0>, rot=<0 0 0 -1>}
 lowerarm_twist_01_r {parent=lowerarm_r, loc=<-17.8549 0 0>, rot=<0 0 0 -1>}
 deform_elbow_in_base_r {parent=lowerarm_r, loc=<-0.2111 -5.2591 -0.8275>, rot=<0 0 0 -1>}
 deform_elbow_in_r {parent=deform_elbow_in_base_r, loc=<-0.0001 0 0.0002>, rot=<0 0 0 1>}
 deform_elbow_out_base_r {parent=lowerarm_r, loc=<0 4 -0.0001>, rot=<0 0 0 -1>}
 deform_elbow_out_r {parent=deform_elbow_out_base_r, loc=<0 0 0>, rot=<0 0 0 1>}
 upperarm_twist_01_r {parent=upperarm_r, loc=<-0.0041 0 0>, rot=<0 0 0 -1>}
 deform_upperarm_bk_base_r {parent=upperarm_twist_01_r, loc=<-1.0814 7.2192 1.9455>, rot=<0 0 0 1>}
 deform_upperarm_bk_r {parent=deform_upperarm_bk_base_r, loc=<0 0 0>, rot=<0 0 0 1>}
 deform_upperarm_fr_base_r {parent=upperarm_twist_01_r, loc=<0.3689 -7.0506 2.7652>, rot=<0 0 0 1>}
 deform_upperarm_fr_r {parent=deform_upperarm_fr_base_r, loc=<0 0 0>, rot=<0 0 0 1>}
 deform_upperarm_up_base_r {parent=upperarm_twist_01_r, loc=<3.2214 1.1391 -3.3308>, rot=<0 0 0 1>}
 deform_upperarm_up_r {parent=deform_upperarm_up_base_r, loc=<0 0 0>, rot=<0 0 0 -1>}
 upperarm_twist_02_r {parent=upperarm_r, loc=<-13.7941 0 0>, rot=<0 0 0 -1>}
 dyn_splineStrap_MSTR {parent=clavicle_r, loc=<-11.9951 0.4227 -9.6787>, rot=<-0.2221 -0.0156 -0.637 -0.738>}
 dyn_splineStrap_01 {parent=dyn_splineStrap_MSTR, loc=<0 0 0>, rot=<0 0 0 1>}
 dyn_splineStrap_02 {parent=dyn_splineStrap_01, loc=<6.072 0 0>, rot=<-0.087 0.318 0.0392 -0.9433>}
 dyn_splineStrap_03 {parent=dyn_splineStrap_02, loc=<6.3737 0 0>, rot=<-0.0238 0.1931 0.0435 -0.9799>}
 dyn_splineStrap_04 {parent=dyn_splineStrap_03, loc=<6.2894 0 0>, rot=<-0.0556 0.0676 -0.02 -0.996>}
 dyn_splineStrap_05 {parent=dyn_splineStrap_04, loc=<4.5577 0 0>, rot=<-0.0149 0.0603 0.0045 -0.9981>}
 dyn_splineStrap_06 {parent=dyn_splineStrap_05, loc=<5.6332 0 0>, rot=<-0.0176 0.0674 0.0116 -0.9975>}
 dyn_splineStrap_07 {parent=dyn_splineStrap_06, loc=<5.3457 0 0>, rot=<-0.0115 0.0302 -0.0075 -0.9994>}
 dyn_splineStrap_08 {parent=dyn_splineStrap_07, loc=<5.0665 0 0>, rot=<-0.0085 0.0914 -0.0059 -0.9958>}
 dyn_splineStrap_09 {parent=dyn_splineStrap_08, loc=<5.3393 0 0>, rot=<0.0832 0.2051 0.0578 -0.9735>}
 dyn_splineStrap_10 {parent=dyn_splineStrap_09, loc=<4.6513 0 0>, rot=<0.097 0.2612 -0.0029 -0.9604>}
 dyn_splineStrap_11 {parent=dyn_splineStrap_10, loc=<4.5238 0 0>, rot=<0 0 0 1>}
 dyn_splineStrapFr_01 {parent=dyn_splineStrap_MSTR, loc=<0 0 0>, rot=<0.0985 -0.9247 0.0137 0.3674>}
 dyn_splineStrapFr_02 {parent=dyn_splineStrapFr_01, loc=<6.6834 0 0>, rot=<0.005 -0.1964 0.0538 -0.979>}
 dyn_splineStrapFr_03 {parent=dyn_splineStrapFr_02, loc=<5.5909 0 0>, rot=<0.0279 -0.122 0.0487 -0.9909>}
 dyn_splineStrapFr_04 {parent=dyn_splineStrapFr_03, loc=<6.5431 0 0>, rot=<0.0699 -0.0007 0.0319 -0.997>}
 dyn_splineStrapFr_05 {parent=dyn_splineStrapFr_04, loc=<6.2566 0 0>, rot=<0.0079 -0.1759 0.0094 -0.9843>}
 dyn_splineStrapFr_06 {parent=dyn_splineStrapFr_05, loc=<5.6953 0 0>, rot=<-0.0042 -0.0873 0.0427 -0.9953>}
 dyn_splineStrapFr_07 {parent=dyn_splineStrapFr_06, loc=<6.4843 0 0>, rot=<-0.0384 -0.1346 -0.0066 -0.9901>}
 dyn_splineStrapFr_08 {parent=dyn_splineStrapFr_07, loc=<5.5953 0 0>, rot=<-0.0612 -0.1568 0.0647 -0.9836>}
 dyn_splineStrapFr_09 {parent=dyn_splineStrapFr_08, loc=<5.0256 0 0>, rot=<-0.0204 -0.149 0.0432 -0.9877>}
 dyn_splineStrapFr_10 {parent=dyn_splineStrapFr_09, loc=<4.2025 0 0>, rot=<-0.0914 -0.2185 0.0539 -0.97>}
 dyn_splineStrapFr_11 {parent=dyn_splineStrapFr_10, loc=<4.3567 0 0>, rot=<0.1305 0 0 -0.9914>}
 dyn_splineStrapFr_03_null {parent=dyn_splineStrapFr_03, type=Null, loc=<11.5727 1.1068 2.93>, rot=<0.1888 0.2499 -0.092 0.9452>, scale=<0.9991 1 1>}
 deform_collar_l {parent=spine_05, loc=<11.9654 0.4309 -9.507>, rot=<0 0 -0.2003 -0.9797>}
 dyn_collar_l {parent=deform_collar_l, loc=<0 0 0>, rot=<0 0 0 1>}
 deform_collar_r {parent=spine_05, loc=<11.965 0.4308 9.507>, rot=<0 0 0.9797 -0.2003>}
 dyn_collar_r {parent=deform_collar_r, loc=<0 0 0>, rot=<0 0 0 1>}
 deform_collar_b {parent=spine_05, loc=<15.2886 -5.6882 0>, rot=<0 0 -0.2003 -0.9797>}
 dyn_collar_b {parent=deform_collar_b, loc=<0 0 0>, rot=<0 0 0 1>}
 deform_collar_f_l {parent=spine_05, loc=<6.7139 8.1548 -5.2701>, rot=<-0.3434 -0.0444 -0.2545 -0.9029>}
 dyn_collar_f_l {parent=deform_collar_f_l, loc=<0 0 0>, rot=<0 0 0 1>}
 deform_collar_f_r {parent=spine_05, loc=<6.7141 8.1548 5.2701>, rot=<-0.0444 0.3434 0.9029 -0.2545>}
 dyn_collar_f_r {parent=deform_collar_f_r, loc=<0 0 0>, rot=<0 0 0 1>}
 dyn_strap_c_05 {parent=spine_05, loc=<-3.9738 -11.7306 1.1466>, rot=<-0.2702 0.6564 0.6707 0.2151>}
 dyn_strap_c_01 {parent=spine_05, loc=<1.003 12.1572 6.6994>, rot=<0.8401 -0.2929 0.1843 0.4176>}
 dyn_strap_c_04 {parent=spine_05, loc=<5.5847 -10.8983 9.2415>, rot=<-0.2702 0.6564 0.6707 0.2151>}
 dyn_strap_c_03 {parent=spine_05, loc=<13.3415 -1.333 14.8746>, rot=<0.4847 -0.8235 -0.2652 -0.1287>}
 dyn_strap_c_02 {parent=spine_05, loc=<7.4544 9.2914 11.4839>, rot=<0.8479 -0.5091 0.0862 0.1205>}
 deform_pec_aim_base_l {parent=spine_04, loc=<7.8728 12.6388 -2.4123>, rot=<0.0102 -0.4602 0.3979 -0.7936>}
 deform_pec_aim_l {parent=deform_pec_aim_base_l, loc=<0 0 0>, rot=<0 0 0 1>}
 deform_pec_base_l {parent=deform_pec_aim_l, loc=<10.766 0 0>, rot=<0 0 0 1>}
 deform_pec_l {parent=deform_pec_base_l, loc=<0.055 0 0>, rot=<0 0 0 1>}
 deform_pec_aim_base_r {parent=spine_04, loc=<7.8723 12.6388 2.4123>, rot=<-0.4602 -0.0102 0.7936 0.3979>}
 deform_pec_aim_r {parent=deform_pec_aim_base_r, loc=<0 0 0>, rot=<0 0 0 1>}
 deform_pec_base_r {parent=deform_pec_aim_r, loc=<-10.7662 -0.0006 -0.0007>, rot=<0 0 0 1>}
 deform_pec_r {parent=deform_pec_base_r, loc=<0 0 0>, rot=<0 0 0 1>}
 dyn_skirtTopA_l_MSTR {parent=spine_02, loc=<-2.2546 4.936 13.426>, rot=<0.233 -0.5051 0.8012 0.2206>}
 dyn_skirtTopA_l_01 {parent=dyn_skirtTopA_l_MSTR, loc=<0 0 0>, rot=<0 0 0 1>}
 dyn_skirtTopB_l_MSTR {parent=spine_02, loc=<-1.3921 -8.5262 -4.6265>, rot=<-0.1252 -0.0323 0.1343 -0.9825>}
 dyn_skirtTopB_l_01 {parent=dyn_skirtTopB_l_MSTR, loc=<0 0 0>, rot=<0 0 0 1>}
 dyn_skirtTopC_l_MSTR {parent=spine_02, loc=<-1.0091 -2.768 -12.4915>, rot=<-0.5546 -0.1614 0.1672 -0.799>}
 dyn_skirtTopC_l_01 {parent=dyn_skirtTopC_l_MSTR, loc=<0 0 0>, rot=<0 0 0 1>}
 dyn_bottleB_MSTR {parent=spine_01, loc=<5.0349 -5.3616 -12.344>, rot=<-0.2029 0.6511 0.7201 0.1282>}
 dyn_bottleB_01 {parent=dyn_bottleB_MSTR, loc=<0 0 0>, rot=<0 0 0 1>}
 dyn_bottleB_02 {parent=dyn_bottleB_01, loc=<4.9886 0 0>, rot=<0 0 0 1>}
 dyn_bottleA_MSTR {parent=spine_01, loc=<4.2565 -0.1746 -14.4124>, rot=<-0.1878 0.5623 0.7996 0.0961>}
 dyn_bottleA_01 {parent=dyn_bottleA_MSTR, loc=<0 0 0>, rot=<0 0 0 1>}
 dyn_bottleA_02 {parent=dyn_bottleA_01, loc=<4.4928 0 0>, rot=<0 0 0 1>}
 thigh_l {parent=pelvis, loc=<-2.5693 -0.0109 -7.7501>, rot=<0.0012 0.0572 0.0155 -0.9982>}
 calf_l {parent=thigh_l, loc=<-39.0484 0 0>, rot=<0.0006 -0.0161 -0.0391 -0.9991>}
 foot_l {parent=calf_l, loc=<-42.561 0.1077 0.4292>, rot=<0.0418 -0.0455 0.0343 -0.9975>}
 ball_l {parent=foot_l, loc=<-6.1599 13.8264 0.6806>, rot=<0.0001 -0.0001 -0.7063 -0.7079>}
 calf_twist_02_l {parent=calf_l, loc=<-18.8064 -0.0967 0.0397>, rot=<0.0237 0.0095 -0.0035 -0.9997>}
 calf_twist_01_l {parent=calf_l, loc=<-35.813 -0.0967 0.0397>, rot=<0.0237 0.0095 -0.0035 -0.9997>}
 deform_knee_in_base_l {parent=calf_l, loc=<1.298 -5.2957 -0.0496>, rot=<0 0 0 -1>}
 deform_knee_in_l {parent=deform_knee_in_base_l, loc=<0 0 0>, rot=<0 0 0 1>}
 deform_knee_out_base_l {parent=calf_l, loc=<0 4 0>, rot=<0 0 0 -1>}
 deform_knee_out_l {parent=deform_knee_out_base_l, loc=<0 0 0>, rot=<0 0 0 1>}
 thigh_twist_01_l {parent=thigh_l, loc=<-21.6303 0 0>, rot=<0.0013 0 0 -1>}
 deform_groin_base_l {parent=thigh_l, loc=<-2.1908 8.1142 -1.7442>, rot=<0 0 0 -1>}
 deform_groin_l {parent=deform_groin_base_l, loc=<0 0 0>, rot=<0 0 0 1>}
 thigh_r {parent=pelvis, loc=<-2.5693 -0.0109 7.7501>, rot=<0.0572 -0.0012 0.9982 0.0155>}
 calf_r {parent=thigh_r, loc=<39.0484 0 0>, rot=<0.0006 -0.0161 -0.0391 -0.9991>}
 foot_r {parent=calf_r, loc=<42.561 -0.1077 -0.4292>, rot=<0.0418 -0.0455 0.0343 -0.9975>}
 ball_r {parent=foot_r, loc=<6.1599 -13.8264 -0.6806>, rot=<0.0001 -0.0001 -0.7063 -0.7079>}
 calf_twist_02_r {parent=calf_r, loc=<18.8064 0.0967 -0.0397>, rot=<0.0237 0.0095 -0.0035 -0.9997>}
 calf_twist_01_r {parent=calf_r, loc=<35.813 0.0967 -0.0397>, rot=<0.0237 0.0095 -0.0035 -0.9997>}
 deform_knee_in_base_r {parent=calf_r, loc=<-1.298 5.2957 0.0496>, rot=<0 0 0 -1>}
 deform_knee_in_r {parent=deform_knee_in_base_r, loc=<0 0 0>, rot=<0 0 0 1>}
 deform_knee_out_base_r {parent=calf_r, loc=<0 -4 0>, rot=<0 0 0 -1>}
 deform_knee_out_r {parent=deform_knee_out_base_r, loc=<0 0 0>, rot=<0 0 0 1>}
 thigh_twist_01_r {parent=thigh_r, loc=<21.6303 0 0>, rot=<0.0013 0 0 -1>}
 deform_groin_base_r {parent=thigh_r, loc=<2.1908 -8.1142 1.7442>, rot=<0 0 0 -1>}
 deform_groin_r {parent=deform_groin_base_r, loc=<0 0 0>, rot=<0 0 0 1>}
 deform_glute_base_l {parent=pelvis, loc=<-8.8064 -8.8222 -8.4287>, rot=<0.0275 0.0501 -0.4472 -0.8926>}
 deform_glute_l {parent=deform_glute_base_l, loc=<0 0 0>, rot=<0 0 0 1>}
 deform_glute_base_r {parent=pelvis, loc=<-8.8062 -8.8222 8.4287>, rot=<0.0501 -0.0275 0.8926 -0.4472>}
 deform_glute_r {parent=deform_glute_base_r, loc=<0 0 0>, rot=<0 0 0 1>}
 dyn_skirtA_l_MSTR {parent=pelvis, loc=<5.1456 8.2609 -5.1258>, rot=<-0.1449 0.1169 0.9763 0.1105>}
 dyn_skirtA_l_01 {parent=dyn_skirtA_l_MSTR, loc=<0 0 0>, rot=<0 0 0 1>}
 dyn_skirtA_l_02 {parent=dyn_skirtA_l_01, loc=<7.4441 0 0>, rot=<0 0 -0.0087 -1>}
 dyn_skirtA_l_03 {parent=dyn_skirtA_l_02, loc=<8.3991 0 0>, rot=<0 0 -0.0269 -0.9996>}
 dyn_skirtA_l_04 {parent=dyn_skirtA_l_03, loc=<14.1276 0.4542 0>, rot=<0 0 -0.0466 -0.9989>}
 dyn_skirtA_l_05 {parent=dyn_skirtA_l_04, loc=<11.0826 0 0>, rot=<0 0 0.0175 -0.9998>}
 dyn_skirtB_l_MSTR {parent=pelvis, loc=<7.5896 -1.4735 -13.6831>, rot=<0.0996 -0.7604 -0.6343 -0.0976>}
 dyn_skirtB_l_01 {parent=dyn_skirtB_l_MSTR, loc=<0 0 0>, rot=<0 0 0 1>}
 dyn_skirtB_l_02 {parent=dyn_skirtB_l_01, loc=<8.2649 0 0>, rot=<-0.0285 0.0186 0.0168 -0.9993>}
 dyn_skirtB_l_03 {parent=dyn_skirtB_l_02, loc=<9.5201 0 0>, rot=<0 0 0 1>}
 dyn_skirtB_l_04 {parent=dyn_skirtB_l_03, loc=<11.4527 0 0>, rot=<0 0 0 1>}
 dyn_ropeA_l_MSTR {parent=pelvis, loc=<8.846 8.7916 3.5661>, rot=<0.2664 -0.252 0.8465 0.3859>}
 dyn_ropeA_l_01 {parent=dyn_ropeA_l_MSTR, loc=<0 0 0>, rot=<0 0 0 1>}
 dyn_ropeA_l_02 {parent=dyn_ropeA_l_01, loc=<5.5567 0 0>, rot=<-0.0045 -0.0164 -0.2919 -0.9563>}
 dyn_ropeA_l_03 {parent=dyn_ropeA_l_02, loc=<6.8628 0 0>, rot=<0 0 0 1>}
 dyn_ropeB_l_MSTR {parent=pelvis, loc=<6.7632 8.6289 5.6481>, rot=<-0.0176 -0.0024 0.9744 0.2243>}
 dyn_ropeB_l_01 {parent=dyn_ropeB_l_MSTR, loc=<0 0 0>, rot=<0 0 0 1>}
 dyn_ropeB_l_02 {parent=dyn_ropeB_l_01, loc=<6.7386 0 0>, rot=<-0.0097 0.0157 -0.154 -0.9879>}
 dyn_ropeB_l_03 {parent=dyn_ropeB_l_02, loc=<7.0916 0 0>, rot=<0 0 0 1>}
 dyn_skirtB_r_MSTR {parent=pelvis, loc=<7.4323 0.9 13.3576>, rot=<0.0672 -0.6891 0.7153 0.0944>}
 dyn_skirtB_r_01 {parent=dyn_skirtB_r_MSTR, loc=<0 0 0>, rot=<0 0 0 1>}
 dyn_skirtB_r_02 {parent=dyn_skirtB_r_01, loc=<6.257 -0.0125 0.076>, rot=<0.005 -0.0189 -0.0032 -0.9998>}
 dyn_skirtB_r_03 {parent=dyn_skirtB_r_02, loc=<8.2197 -0.3611 0.1697>, rot=<0 0 0 1>}
 dyn_skirtA_r_MSTR {parent=pelvis, loc=<5.3241 7.5916 6.1206>, rot=<0.0574 -0.0924 0.9894 0.096>}
 dyn_skirtA_r_01 {parent=dyn_skirtA_r_MSTR, loc=<0 0 0>, rot=<0 0 0 1>}
 dyn_skirtA_r_02 {parent=dyn_skirtA_r_01, loc=<5.5952 0 0>, rot=<0 0 0 1>}
 dyn_skirtC_r_MSTR {parent=pelvis, loc=<7.4671 -9.0098 8.909>, rot=<0.1322 -0.9912 0 0>}
 dyn_skirtC_r_01 {parent=dyn_skirtC_r_MSTR, loc=<0 0 0>, rot=<0 0 0 1>}
 dyn_skirtC_r_02 {parent=dyn_skirtC_r_01, loc=<8.9346 0 0>, rot=<0 0 -0.0552 -0.9985>}
 dyn_skirtC_r_03 {parent=dyn_skirtC_r_02, loc=<10.7076 0 0>, rot=<0 0 0 1>}
 dyn_skirtC_l_MSTR {parent=pelvis, loc=<7.4646 -8.507 -9.1278>, rot=<0.1354 -0.9908 0 0>}
 dyn_skirtC_l_01 {parent=dyn_skirtC_l_MSTR, loc=<0 0 0>, rot=<0 0 0 1>}
 dyn_skirtC_l_02 {parent=dyn_skirtC_l_01, loc=<8.9343 0 0>, rot=<0 0 -0.0412 -0.9992>}
 dyn_skirtC_l_03 {parent=dyn_skirtC_l_02, loc=<10.7077 0 0>, rot=<0 0 0 1>}
 dyn_skirtC_l_04 {parent=dyn_skirtC_l_03, loc=<13.7494 0 0>, rot=<0 0 0 1>}
 dyn_splineStrap_11_null {type=Null, loc=<13.6387 -1.9356 112.8722>, rot=<0.4072 0.2923 -0.5827 -0.6397>}
 dyn_splineStrap_11_ctrl {controlType=EULER_TRANSFORM, parent=dyn_splineStrap_11_null, type=Control, loc=<0 0 0>, rot=<0 0 0 1>}
 dyn_splineStrap_10_null {type=Null, loc=<12.96 -6.3849 113.3277>, rot=<-0.4072 -0.2923 0.5827 0.6397>}
 dyn_splineStrap_10_ctrl {controlType=EULER_TRANSFORM, parent=dyn_splineStrap_10_null, type=Control, loc=<0 0 0>, rot=<0 0 0 1>}
 dyn_splineStrap_09_null {type=Null, loc=<10.4818 -9.9119 115.0752>, rot=<0.4804 0.0583 -0.4798 -0.7318>}
 dyn_splineStrap_09_ctrl {controlType=EULER_TRANSFORM, parent=dyn_splineStrap_09_null, type=Control, loc=<0 0 0>, rot=<0 0 0 1>}
 dyn_splineStrap_08_null {type=Null, loc=<7.2647 -12.3071 118.5997>, rot=<-0.5086 0.1611 0.4157 0.7366>}
 dyn_splineStrap_08_ctrl {controlType=EULER_TRANSFORM, parent=dyn_splineStrap_08_null, type=Control, loc=<0 0 0>, rot=<0 0 0 1>}
 dyn_splineStrap_07_null {type=Null, loc=<4.0396 -13.7187 122.2432>, rot=<0.5516 -0.2211 -0.3644 -0.7169>}
 dyn_splineStrap_07_ctrl {controlType=EULER_TRANSFORM, parent=dyn_splineStrap_07_null, type=Control, loc=<0 0 0>, rot=<0 0 0 1>}
 dyn_splineStrap_06_null {type=Null, loc=<0.5521 -14.8882 126.1222>, rot=<-0.5722 0.2344 0.3448 0.7063>}
 dyn_splineStrap_06_ctrl {controlType=EULER_TRANSFORM, parent=dyn_splineStrap_06_null, type=Control, loc=<0 0 0>, rot=<0 0 0 1>}
 dyn_splineStrap_05_null {type=Null, loc=<-3.0488 -15.3849 130.4256>, rot=<0.6037 -0.2819 -0.3177 -0.6746>}
 dyn_splineStrap_05_ctrl {controlType=EULER_TRANSFORM, parent=dyn_splineStrap_05_null, type=Control, loc=<0 0 0>, rot=<0 0 0 1>}
 dyn_splineStrap_04_null {type=Null, loc=<-5.9173 -15.24 133.9645>, rot=<-0.6305 0.3201 0.2879 0.6459>}
 dyn_splineStrap_04_ctrl {controlType=EULER_TRANSFORM, parent=dyn_splineStrap_04_null, type=Control, loc=<0 0 0>, rot=<0 0 0 1>}
 dyn_splineStrap_03_null {type=Null, loc=<-10.0252 -14.198 138.6116>, rot=<0.6897 -0.3339 -0.2489 -0.5923>}
 dyn_splineStrap_03_ctrl {controlType=EULER_TRANSFORM, parent=dyn_splineStrap_03_null, type=Control, loc=<0 0 0>, rot=<0 0 0 1>}
 dyn_splineStrap_02_null {type=Null, loc=<-13.3695 -10.8034 142.8444>, rot=<-0.7235 0.4656 0.1444 0.4887>}
 dyn_splineStrap_02_ctrl {controlType=EULER_TRANSFORM, parent=dyn_splineStrap_02_null, type=Control, loc=<0 0 0>, rot=<0 0 0 1>}
 dyn_splineStrap_01_null {type=Null, loc=<-14.9024 -5.1225 144.3429>, rot=<0.7527 -0.6104 0.0342 -0.2443>}
 dyn_splineStrap_01_ctrl {controlType=EULER_TRANSFORM, parent=dyn_splineStrap_01_null, type=Control, loc=<0 0 0>, rot=<0 0 0 1>}
 dyn_splineStrapFr_01_ctrl_null {type=Null, loc=<-14.9024 -5.1225 144.3429>, rot=<-0.2758 0.0053 0.6267 0.7288>}
 dyn_splineStrapFr_01_ctrl {
  controlType=EULER_TRANSFORM
  parent=dyn_splineStrapFr_01_ctrl_null
  type=Control
  loc=<58.8582 43.7306 -125.3258>
  rot=<0.2758 -0.0053 -0.6267 0.7288>
 }
 dyn_splineStrapFr_02_ctrl_null {type=Null, loc=<-13.4694 0.9634 141.9816>, rot=<0.397 -0.1303 -0.5203 -0.7448>}
 dyn_splineStrapFr_02_ctrl {
  controlType=EULER_TRANSFORM
  parent=dyn_splineStrapFr_02_ctrl_null
  type=Control
  loc=<-6.1294 -0.6903 -2.5733>
  rot=<-0.005 0.1964 -0.0538 -0.979>
 }
 dyn_splineStrapFr_03_ctrl_null {type=Null, loc=<-11.0951 4.7181 138.587>, rot=<-0.4839 0.1861 0.4345 0.7364>}
 dyn_splineStrapFr_03_ctrl {
  controlType=EULER_TRANSFORM
  parent=dyn_splineStrapFr_03_ctrl_null
  type=Control
  loc=<-5.3981 -0.5013 -1.3665>
  rot=<-0.0279 0.122 -0.0487 -0.9909>
 }
 dyn_splineStrapFr_04_ctrl_null {type=Null, loc=<-7.4764 7.7271 134.0412>, rot=<0.5402 -0.1403 -0.4225 -0.7142>}
 dyn_splineStrapFr_04_ctrl {
  controlType=EULER_TRANSFORM
  parent=dyn_splineStrapFr_04_ctrl_null
  type=Control
  loc=<-6.5298 -0.4154 -0.0379>
  rot=<-0.0699 0.0007 -0.0319 -0.997>
 }
 dyn_splineStrapFr_05_ctrl_null {type=Null, loc=<-3.6993 10.5541 129.9319>, rot=<-0.613 0.2553 0.3152 0.6781>}
 dyn_splineStrapFr_05_ctrl {
  controlType=EULER_TRANSFORM
  parent=dyn_splineStrapFr_05_ctrl_null
  type=Control
  loc=<-5.8683 -0.0989 -2.1674>
  rot=<-0.0079 0.1759 -0.0094 -0.9843>
 }
 dyn_splineStrapFr_06_ctrl_null {type=Null, loc=<0.122 11.2059 125.7596>, rot=<0.6456 -0.2884 -0.2301 -0.6686>}
 dyn_splineStrapFr_06_ctrl {
  controlType=EULER_TRANSFORM
  parent=dyn_splineStrapFr_06_ctrl_null
  type=Control
  loc=<-5.5876 -0.4888 -0.9878>
  rot=<0.0042 0.0873 -0.0427 -0.9953>
 }
 dyn_splineStrapFr_07_ctrl_null {type=Null, loc=<4.8408 10.7863 121.3321>, rot=<-0.6426 0.3887 0.1343 0.6465>}
 dyn_splineStrapFr_07_ctrl {
  controlType=EULER_TRANSFORM
  parent=dyn_splineStrapFr_07_ctrl_null
  type=Control
  loc=<82.2315 89.913 3.8567>
  rot=<0.6426 -0.3887 -0.1343 0.6465>
 }
 dyn_splineStrapFr_08_ctrl_null {type=Null, loc=<8.5434 8.9624 117.5544>, rot=<0.6387 -0.4503 0.0343 -0.623>}
 dyn_splineStrapFr_08_ctrl {
  controlType=EULER_TRANSFORM
  parent=dyn_splineStrapFr_08_ctrl_null
  type=Control
  loc=<-5.2734 -0.8197 -1.6814>
  rot=<0.0612 0.1568 -0.0647 -0.9836>
 }
 dyn_splineStrapFr_09_ctrl_null {type=Null, loc=<11.5189 5.8566 114.9549>, rot=<-0.6325 0.5093 -0.1652 0.5597>}
 dyn_splineStrapFr_09_ctrl {
  controlType=EULER_TRANSFORM
  parent=dyn_splineStrapFr_09_ctrl_null
  type=Control
  loc=<-4.7836 -0.4597 -1.4708>
  rot=<0.0204 0.149 -0.0432 -0.9877>
 }
 dyn_splineStrapFr_10_ctrl_null {type=Null, loc=<13.3118 2.3719 113.4371>, rot=<0.5537 -0.5672 0.3752 -0.4806>}
 dyn_splineStrapFr_10_ctrl {
  controlType=EULER_TRANSFORM
  parent=dyn_splineStrapFr_10_ctrl_null
  type=Control
  loc=<-3.7767 -0.607 -1.7404>
  rot=<0.0914 0.2185 -0.0539 -0.97>
 }
 dyn_splineStrapFr_11_ctrl_null {type=Null, loc=<13.6387 -1.9356 112.8722>, rot=<-0.6117 0.6113 -0.2979 0.4042>}
 dyn_splineStrapFr_11_ctrl {controlType=EULER_TRANSFORM, parent=dyn_splineStrapFr_11_ctrl_null, type=Control, loc=<-4.3567 0 0>, rot=<-0.1305 0 0 -0.9914>}
]
variables [
 front_bones {
  type="TArray<FRigElementKey>"
  default="((Type=Bone,Name=\"dyn_splineStrapFr_01\"),(Type=Bone,Name=\"dyn_splineStrapFr_02\"),(Type=Bone,Name=\"dyn_splineStrapFr_03\"),(Type=Bone,Name=\"dyn_splineStrapFr_04\"),(Type=Bone,Name=\"dyn_splineStrapFr_05\"),(Type=Bone,Name=\"dyn_splineStrapFr_06\"),(Type=Bone,Name ...(truncated)"
 }
 front_ctrls {
  type="TArray<FRigElementKey>"
  default="((Type=Control,Name=\"dyn_splineStrapFr_01_ctrl\"),(Type=Control,Name=\"dyn_splineStrapFr_02_ctrl\"),(Type=Control,Name=\"dyn_splineStrapFr_03_ctrl\"),(Type=Control,Name=\"dyn_splineStrapFr_04_ctrl\"),(Type=Control,Name=\"dyn_splineStrapFr_05_ctrl\"),(Type=Control,N ...(truncated)"
 }
 has_hang_glider {type=bool, default=False}
 is rban enabled {type=bool, default=False}
 is targeting {type=bool, default=False}
 isPistolADS {type=bool, default=False}
 is_floating {type=bool, default=False}
 is_not_playing_melee {type=bool, default=False}
 rear_bones {
  type="TArray<FRigElementKey>"
  default="((Type=Bone,Name=\"dyn_splineStrap_01\"),(Type=Bone,Name=\"dyn_splineStrap_02\"),(Type=Bone,Name=\"dyn_splineStrap_03\"),(Type=Bone,Name=\"dyn_splineStrap_04\"),(Type=Bone,Name=\"dyn_splineStrap_05\"),(Type=Bone,Name=\"dyn_splineStrap_06\"),(Type=Bone,Name=\"dyn_spline ...(truncated)"
 }
 rear_ctrls {
  type="TArray<FRigElementKey>"
  default="((Type=Control,Name=\"dyn_splineStrap_01_ctrl\"),(Type=Control,Name=\"dyn_splineStrap_02_ctrl\"),(Type=Control,Name=\"dyn_splineStrap_03_ctrl\"),(Type=Control,Name=\"dyn_splineStrap_04_ctrl\"),(Type=Control,Name=\"dyn_splineStrap_05_ctrl\"),(Type=Control,Name=\"dyn_s ...(truncated)"
 }
]
