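# Secondary-motion AnimBlueprint for the CubicVice Dual male body cosmetic, copying the attached parent's pose.
# Two Control Rigs (main, strap) feed AnimDynamics on skirt, ropes and bottles, then a pelvis-space RigidBody pass.
# A front-end bottle RigidBody replaces bottle AnimDynamics outside emotes, and a collar bone tilts during the custom idle montage.
# The event graph derives gating flags from front-end, emote, melee, pistol-ADS, skydiving and glider state.
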AnimBlueprint /BRCosmetics/Characters/Player/Male/Medium/Bodies/M_MED_CubicVice_Dual/Meshes/M_MED_CubicVice_Dual_AnimBP
target_skeleton=&/M_MED_CubicVice_Dual_Skeleton
graphs [
 AnimGraph {nodes [
   n1 = RigidBody!(
    ComponentPose=Local To Component(LocalPose=Use cached pose 'postSimCache'())
    bAlphaBoolEnabled:bool=Get is_front_end_not_emoting() -> (is_front_end_not_emoting:bool)
    ExternalForce
   ) -> (Pose) {
    alphaInputType=Bool
    bClampLinearTranslationLimitToRefPose=true
    bOverrideWorldGravity=true
    overridePhysicsAsset=&/M_MED_CubicVice_Dual_Bottles_FE_Physics
    simulationSpace=BaseBoneSpace
    alphaBoolBlend {blendInTime=0.8, blendOption=Linear, blendOutTime=0.8}
    baseBoneRef {boneName=root}
    componentAppliedLinearAccClamp=<2000 2000 2000>
    componentLinearAccScale=<0.6 0.6 0.6>
    componentLinearVelScale=<0.6 0.6 0.6>
    overrideWorldGravity {z=-980}
   }
   n2 = "Transform (Modify) Bone - Bone: deform_collar_f_l"(ComponentPose=@n1, Alpha=Get isCustomEmotePlaying() -> (isCustomEmotePlaying:double), Translation="0,0,0", Rotation="-20,0,0", Scale="1,1,1") {
    rotationMode=BMM_Additive
    rotationSpace=BCS_BoneSpace
    alphaBoolBlend {blendOption=Linear}
    alphaScaleBiasClamp {bInterpResult=true}
    boneToModify {boneName=deform_collar_f_l}
    rotation {pitch=-20}
   }
   Output Pose(Result=Component To Local(ComponentPose=@n2)) {layerGroup=None}
   n3 = Get bIsTargeting()
   n4 = Control Rig!(
    Source=Copy Pose From Mesh() {bUseAttachedParent=true}
    Alpha
    is_front_end_not_emoting:bool=Get is_front_end_not_emoting()
    is_targeting:bool=@n3
    is_front_end:bool=Get bIsInFrontEnd()
   ) -> (Pose) {
    alphaBoolBlend {blendOption=Linear}
    controlRigAssetReference {blueprintRigClass=&/M_MED_CubicVice_Dual_CtrlRig.M_MED_CubicVice_Dual_CtrlRig_C}
   }
   Save cached pose 'CRCache'(Pose=@n4) {cacheName=CRCache}
   n5 = Control Rig!(
    Source=Use cached pose 'CRCache'()
    Alpha
    has_hang_glider:bool=Get has_hand_glider_parachute_opened() -> (has_hand_glider_parachute_opened:bool)
    is_floating:bool=Get is_skydiving_not_parachuting() -> (is_skydiving_not_parachuting:bool)
    is_not_playing_melee:bool=Get is_not_playing_melee() -> (is_not_playing_melee:bool)
    isPistolADS:bool=Get isPistolADS() -> (isPistolADS:bool)
    is rban enabled:bool=Get bRigidBodyEnabled()
    is targeting:bool=@n3
   ) -> (Pose) {
    alphaBoolBlend {blendOption=Linear}
    controlRigAssetReference {blueprintRigClass=&/M_MED_CubicVice_Dual_Strap_CtrlRig.M_MED_CubicVice_Dual_Strap_CtrlRig_C}
   }
   Save cached pose 'preSimCache'(Pose=@n5) {cacheName=preSimCache}
   n6 = Anim Dynamics - Bone: dyn_skirtTopB_l_01(ComponentPose=Local To Component(LocalPose=Use cached pose 'preSimCache'()), Alpha=1, ExternalForce="0,0,0") {
    angularDampingOverride=1
    angularSpringConstant=8000
    bAngularSpring=true
    bOverrideAngularDamping=true
    bUseGravityOverride=true
    gravityScale=0.35
    simulationSpace=BoneRelative
    alphaBoolBlend {blendOption=Linear}
    boundBone {boneName=dyn_skirtTopB_l_01}
    componentAppliedLinearAccClamp=<3000 3000 3000>
    componentLinearAccScale=<1 1 1>
    componentLinearVelScale=<1 1 1>
    gravityOverride {z=-980}
    physicsBodyDefinitions [{
      boundBone {boneName=dyn_skirtTopB_l_01}
      boxExtents=<20 30 30>
      constraintSetup {angularLimitsMax=<15 15 15>, angularLimitsMin=<-15 -15 -15>, angularTarget=<1 0 0>}
      localJointOffset=<4 0 0>
     }]
    relativeSpaceBone {boneName=root}
   }
   n7 = Anim Dynamics - Bone: dyn_skirtTopC_l_01(ComponentPose=@n6, Alpha=1, ExternalForce="0,0,0") {
    angularDampingOverride=1
    angularSpringConstant=8000
    bAngularSpring=true
    bOverrideAngularDamping=true
    bUseGravityOverride=true
    gravityScale=0.35
    simulationSpace=BoneRelative
    alphaBoolBlend {blendOption=Linear}
    boundBone {boneName=dyn_skirtTopC_l_01}
    componentAppliedLinearAccClamp=<3000 3000 3000>
    componentLinearAccScale=<1 1 1>
    componentLinearVelScale=<1 1 1>
    gravityOverride {z=-980}
    physicsBodyDefinitions [{
      boundBone {boneName=dyn_skirtTopC_l_01}
      boxExtents=<20 30 30>
      constraintSetup {angularLimitsMax=<15 15 15>, angularLimitsMin=<-15 -15 -15>, angularTarget=<1 0 0>}
      localJointOffset=<4 0 0>
     }]
    relativeSpaceBone {boneName=root}
   }
   n8 = Anim Dynamics - Bone: dyn_skirtTopA_l_01(ComponentPose=@n7, Alpha=1, ExternalForce="0,0,0") {
    angularDampingOverride=1
    angularSpringConstant=8000
    bAngularSpring=true
    bOverrideAngularDamping=true
    bUseGravityOverride=true
    gravityScale=0.35
    simulationSpace=BoneRelative
    alphaBoolBlend {blendOption=Linear}
    boundBone {boneName=dyn_skirtTopA_l_01}
    componentAppliedLinearAccClamp=<3000 3000 3000>
    componentLinearAccScale=<1 1 1>
    componentLinearVelScale=<1 1 1>
    gravityOverride {z=-980}
    physicsBodyDefinitions [{
      boundBone {boneName=dyn_skirtTopA_l_01}
      boxExtents=<20 30 30>
      constraintSetup {angularLimitsMax=<15 15 10>, angularLimitsMin=<-15 -15 -25>, angularTarget=<1 0 0>}
      localJointOffset=<4 0 0>
     }]
    relativeSpaceBone {boneName=root}
   }
   n9 = Anim Dynamics - Bone: dyn_skirtTopB_l_01(ComponentPose=@n8, Alpha=1, ExternalForce="0,0,0") {
    angularDampingOverride=1
    angularSpringConstant=20000
    bAngularSpring=true
    bOverrideAngularDamping=true
    bUseGravityOverride=true
    gravityScale=0.35
    simulationSpace=BoneRelative
    alphaBoolBlend {blendOption=Linear}
    boundBone {boneName=dyn_skirtTopB_l_01}
    componentAppliedLinearAccClamp=<3000 3000 3000>
    componentLinearAccScale=<1 1 1>
    componentLinearVelScale=<1 1 1>
    gravityOverride {z=-980}
    physicsBodyDefinitions [{
      boundBone {boneName=dyn_skirtTopB_l_01}
      boxExtents=<20 30 30>
      constraintSetup {angularLimitsMax=<15 15 15>, angularLimitsMin=<-15 -15 -15>, angularTarget=<1 0 0>}
      localJointOffset=<4 0 0>
     }]
    relativeSpaceBone {boneName=root}
   }
   n10 = Anim Dynamics - Bone: dyn_skirtTopC_l_01(ComponentPose=@n9, Alpha=1, ExternalForce="0,0,0") {
    angularDampingOverride=1
    angularSpringConstant=20000
    bAngularSpring=true
    bOverrideAngularDamping=true
    bUseGravityOverride=true
    gravityScale=0.35
    simulationSpace=BoneRelative
    alphaBoolBlend {blendOption=Linear}
    boundBone {boneName=dyn_skirtTopC_l_01}
    componentAppliedLinearAccClamp=<3000 3000 3000>
    componentLinearAccScale=<1 1 1>
    componentLinearVelScale=<1 1 1>
    gravityOverride {z=-980}
    physicsBodyDefinitions [{
      boundBone {boneName=dyn_skirtTopC_l_01}
      boxExtents=<20 30 30>
      constraintSetup {angularLimitsMax=<15 15 15>, angularLimitsMin=<-15 -15 -15>, angularTarget=<1 0 0>}
      localJointOffset=<4 0 0>
     }]
    relativeSpaceBone {boneName=root}
   }
   n11 = Anim Dynamics - Bone: dyn_ropeA_l_01!(ComponentPose=@n10, bAlphaBoolEnabled:bool=True, ExternalForce="0,0,0") -> (Pose) {
    alphaInputType=Bool
    angularDampingOverride=1
    angularSpringConstant=10000
    bAngularSpring=true
    bOverrideAngularDamping=true
    bUseGravityOverride=true
    gravityScale=0.35
    simulationSpace=RootRelative
    alphaBoolBlend {blendInTime=0.3, blendOption=Linear, blendOutTime=0.3}
    boundBone {boneName=dyn_ropeA_l_01}
    componentLinearAccScale=<1 1 1>
    componentLinearVelScale=<1 1 1>
    gravityOverride {z=-980}
    physicsBodyDefinitions [{
      boundBone {boneName=dyn_ropeA_l_01}
      boxExtents=<20 30 30>
      constraintSetup {angularLimitsMax=<15 15 10>, angularLimitsMin=<-15 -15 -15>, angularTarget=<1 0 0>}
      localJointOffset=<4 0 0>
     }]
   }
   n12 = Anim Dynamics - Bone: dyn_ropeA_l_02!(ComponentPose=@n11, bAlphaBoolEnabled:bool=True, ExternalForce="0,0,0") -> (Pose) {
    alphaInputType=Bool
    angularDampingOverride=1
    angularSpringConstant=10000
    bAngularSpring=true
    bOverrideAngularDamping=true
    bUseGravityOverride=true
    gravityScale=0.35
    simulationSpace=RootRelative
    alphaBoolBlend {blendInTime=0.3, blendOption=Linear, blendOutTime=0.3}
    boundBone {boneName=dyn_ropeA_l_02}
    componentLinearAccScale=<1 1 1>
    componentLinearVelScale=<1 1 1>
    gravityOverride {z=-980}
    physicsBodyDefinitions [{
      boundBone {boneName=dyn_ropeA_l_02}
      boxExtents=<20 30 30>
      constraintSetup {angularLimitsMax=<15 15 10>, angularLimitsMin=<-15 -15 -15>, angularTarget=<1 0 0>}
      localJointOffset=<4 0 0>
     }]
   }
   n13 = Anim Dynamics - Bone: dyn_ropeA_l_03!(ComponentPose=@n12, bAlphaBoolEnabled:bool=True, ExternalForce="0,0,0") -> (Pose) {
    alphaInputType=Bool
    angularDampingOverride=1
    angularSpringConstant=10000
    bAngularSpring=true
    bOverrideAngularDamping=true
    bUseGravityOverride=true
    gravityScale=0.35
    simulationSpace=RootRelative
    alphaBoolBlend {blendInTime=0.3, blendOption=Linear, blendOutTime=0.3}
    boundBone {boneName=dyn_ropeA_l_03}
    componentLinearAccScale=<1 1 1>
    componentLinearVelScale=<1 1 1>
    gravityOverride {z=-980}
    physicsBodyDefinitions [{
      boundBone {boneName=dyn_ropeA_l_03}
      boxExtents=<20 30 30>
      constraintSetup {angularLimitsMax=<15 15 10>, angularLimitsMin=<-15 -15 -15>, angularTarget=<1 0 0>}
      localJointOffset=<4 0 0>
     }]
   }
   n14 = Anim Dynamics - Bone: dyn_ropeB_l_01!(ComponentPose=@n13, bAlphaBoolEnabled:bool=True, ExternalForce="0,0,0") -> (Pose) {
    alphaInputType=Bool
    angularDampingOverride=1
    angularSpringConstant=8000
    bAngularSpring=true
    bOverrideAngularDamping=true
    bUseGravityOverride=true
    gravityScale=0.35
    simulationSpace=RootRelative
    alphaBoolBlend {blendInTime=0.3, blendOption=Linear, blendOutTime=0.3}
    boundBone {boneName=dyn_ropeB_l_01}
    componentLinearAccScale=<1 1 1>
    componentLinearVelScale=<1 1 1>
    gravityOverride {z=-980}
    physicsBodyDefinitions [{
      boundBone {boneName=dyn_ropeB_l_01}
      boxExtents=<20 30 30>
      constraintSetup {angularLimitsMax=<15 15 10>, angularLimitsMin=<-15 -15 -15>, angularTarget=<1 0 0>}
      localJointOffset=<4 0 0>
     }]
   }
   n15 = Anim Dynamics - Bone: dyn_ropeB_l_02!(ComponentPose=@n14, bAlphaBoolEnabled:bool=True, ExternalForce="0,0,0") -> (Pose) {
    alphaInputType=Bool
    angularDampingOverride=1
    angularSpringConstant=8000
    bAngularSpring=true
    bOverrideAngularDamping=true
    bUseGravityOverride=true
    gravityScale=0.35
    simulationSpace=RootRelative
    alphaBoolBlend {blendInTime=0.3, blendOption=Linear, blendOutTime=0.3}
    boundBone {boneName=dyn_ropeB_l_02}
    componentLinearAccScale=<1 1 1>
    componentLinearVelScale=<1 1 1>
    gravityOverride {z=-980}
    physicsBodyDefinitions [{
      boundBone {boneName=dyn_ropeB_l_02}
      boxExtents=<20 30 30>
      constraintSetup {angularLimitsMax=<15 15 10>, angularLimitsMin=<-15 -15 -15>, angularTarget=<1 0 0>}
      localJointOffset=<4 0 0>
     }]
   }
   n16 = Anim Dynamics - Bone: dyn_ropeB_l_03!(ComponentPose=@n15, bAlphaBoolEnabled:bool=True, ExternalForce="0,0,0") -> (Pose) {
    alphaInputType=Bool
    angularDampingOverride=1
    angularSpringConstant=8000
    bAngularSpring=true
    bOverrideAngularDamping=true
    bUseGravityOverride=true
    gravityScale=0.35
    simulationSpace=RootRelative
    alphaBoolBlend {blendInTime=0.3, blendOption=Linear, blendOutTime=0.3}
    boundBone {boneName=dyn_ropeB_l_03}
    componentLinearAccScale=<1 1 1>
    componentLinearVelScale=<1 1 1>
    gravityOverride {z=-980}
    physicsBodyDefinitions [{
      boundBone {boneName=dyn_ropeB_l_03}
      boxExtents=<20 30 30>
      constraintSetup {angularLimitsMax=<15 15 10>, angularLimitsMin=<-15 -15 -15>, angularTarget=<1 0 0>}
      localJointOffset=<4 0 0>
     }]
   }
   n17 = Get is_front_end_emoting() -> (is_front_end_emoting:bool)
   n18 = Anim Dynamics - Bone: dyn_bottleB_01!(ComponentPose=@n16, bAlphaBoolEnabled:bool=@n17, ExternalForce="0,0,0") -> (Pose) {
    alphaInputType=Bool
    angularDampingOverride=1
    angularSpringConstant=5000
    bAngularSpring=true
    bOverrideAngularDamping=true
    bUseGravityOverride=true
    gravityScale=0.35
    simulationSpace=RootRelative
    alphaBoolBlend {blendInTime=0.5, blendOption=Linear, blendOutTime=0.5}
    boundBone {boneName=dyn_bottleB_01}
    componentAppliedLinearAccClamp=<6000 6000 6000>
    componentLinearAccScale=<0.6 0.5 0.5>
    componentLinearVelScale=<0.6 0.5 0.5>
    gravityOverride {z=-980}
    physicsBodyDefinitions [{
      boundBone {boneName=dyn_bottleB_01}
      boxExtents=<20 30 30>
      constraintSetup {angularLimitsMax=<15 15 0>, angularLimitsMin=<-15 -15 -30>, angularTarget=<1 0 0>}
      localJointOffset=<4 0 0>
     }]
   }
   n19 = Anim Dynamics - Bone: dyn_bottleB_02!(ComponentPose=@n18, bAlphaBoolEnabled:bool=@n17, ExternalForce="0,0,0") -> (Pose) {
    alphaInputType=Bool
    angularDampingOverride=1
    angularSpringConstant=8000
    bAngularSpring=true
    bOverrideAngularDamping=true
    bUseGravityOverride=true
    gravityScale=0.35
    simulationSpace=BoneRelative
    alphaBoolBlend {blendInTime=0.5, blendOption=Linear, blendOutTime=0.5}
    boundBone {boneName=dyn_bottleB_02}
    componentAppliedLinearAccClamp=<6000 6000 6000>
    componentLinearAccScale=<0.6 1.5 1.5>
    componentLinearVelScale=<0.6 1.5 1.5>
    gravityOverride {z=-980}
    physicsBodyDefinitions [{
      boundBone {boneName=dyn_bottleB_02}
      boxExtents=<20 30 30>
      constraintSetup {angularLimitsMax=<8 8 0>, angularLimitsMin=<-8 -8 -20>, angularTarget=<1 0 0>}
      localJointOffset=<4 0 0>
     }]
    relativeSpaceBone {boneName=root}
   }
   n20 = Anim Dynamics - Bone: dyn_bottleA_01!(ComponentPose=@n19, bAlphaBoolEnabled:bool=@n17, ExternalForce="0,0,0") -> (Pose) {
    alphaInputType=Bool
    angularDampingOverride=1
    angularSpringConstant=20000
    bAngularSpring=true
    bOverrideAngularDamping=true
    bUseGravityOverride=true
    gravityScale=0.35
    simulationSpace=RootRelative
    alphaBoolBlend {blendInTime=0.5, blendOption=Linear, blendOutTime=0.5}
    boundBone {boneName=dyn_bottleA_01}
    componentAppliedLinearAccClamp=<6000 6000 6000>
    componentLinearAccScale=<0.8 0.8 0.8>
    componentLinearVelScale=<0.8 0.8 0.8>
    gravityOverride {z=-980}
    physicsBodyDefinitions [{
      boundBone {boneName=dyn_bottleA_01}
      boxExtents=<20 30 30>
      constraintSetup {angularLimitsMax=<15 15 0>, angularLimitsMin=<-15 -15 -20>, angularTarget=<1 0 0>}
      localJointOffset=<4 0 0>
     }]
   }
   n21 = Anim Dynamics - Bone: dyn_bottleA_02!(ComponentPose=@n20, bAlphaBoolEnabled:bool=@n17, ExternalForce="0,0,0") -> (Pose) {
    alphaInputType=Bool
    angularDampingOverride=1
    angularSpringConstant=6000
    bAngularSpring=true
    bOverrideAngularDamping=true
    bUseGravityOverride=true
    gravityScale=0.35
    simulationSpace=RootRelative
    alphaBoolBlend {blendInTime=0.5, blendOption=Linear, blendOutTime=0.5}
    boundBone {boneName=dyn_bottleA_02}
    componentAppliedLinearAccClamp=<6000 6000 6000>
    componentLinearAccScale=<0.8 0.8 0.8>
    componentLinearVelScale=<0.8 0.8 0.8>
    gravityOverride {z=-980}
    physicsBodyDefinitions [{
      boundBone {boneName=dyn_bottleA_02}
      boxExtents=<20 30 30>
      constraintSetup {angularLimitsMax=<15 15 0>, angularLimitsMin=<-15 -15 -20>, angularTarget=<1 0 0>}
      localJointOffset=<4 0 0>
     }]
   }
   Save cached pose 'rbanCache'(Pose=Component To Local(ComponentPose=@n21)) {cacheName=rbanCache}
   n22 = RigidBody!(ComponentPose=Local To Component(LocalPose=Use cached pose 'rbanCache'()), Alpha:float, OverrideWorldGravity:Vector="0,0,0", ExternalForce) -> (Pose) {
    bClampLinearTranslationLimitToRefPose=true
    bOverrideWorldGravity=true
    overridePhysicsAsset=&/M_MED_CubicVice_Dual_Physics
    simulationSpace=BaseBoneSpace
    alphaBoolBlend {blendOption=Linear}
    baseBoneRef {boneName=pelvis}
    componentAppliedLinearAccClamp=<6000 6000 600>
    componentLinearAccScale=<0.3 0.5 0.5>
    componentLinearVelScale=<0.3 0.5 0.5>
   }
   Save cached pose 'postSimCache'(Pose=Component To Local(ComponentPose=@n22)) {cacheName=postSimCache}
  ]}
 EventGraph {nodes [
   Event BlueprintUpdateAnimation()
   > n1 = Sequence!(execute) -> (then_0, then_1, then_2:Exec, then_3:Exec, then_4:Exec)
   > n2 = Gravity Override State Function DEPRECATED(CharacterPartAnimInstance=Self-Reference(), GameTimeInSeconds=GetGameTimeInSeconds())
   > Set Gravity Override(Gravity Override=@n2)
   n3 = Get bIsKneeSliding()
   > Set SlidingAlpha(SlidingAlpha:float="To Float (Boolean)"(InBool=NOT Boolean(A=@n3))) -> (Output_Get:float)
   n4 = Cast To FortPlayerAnimInstance(execute=@n1.then_3, Object=GetAnimInstance(self=GetOwnerSkeletalMeshComponent()))
   > Set isPistol(isPistol=false)
   > Set isPistolADS(isPistolADS:bool=false) -> (Output_Get:bool)
   > n5 = Switch on EFortWeaponCoreAnimation!(execute, Selection:EFortWeaponCoreAnimation Enum=Get WeaponCoreAnim(self=@n4)) -> (
    Melee:Exec
    Pistol:Exec
    Shotgun:Exec
    PaperBlueprint:Exec
    Rifle:Exec
    MeleeOneHand:Exec
    MachinePistol:Exec
    RocketLauncher:Exec
    GrenadeLauncher:Exec
    GoingCommando:Exec
    AssaultRifle:Exec
    TacticalShotgun:Exec
    SniperRifle:Exec
    TrapPlacement:Exec
    ShoulderLauncher:Exec
    AbilityDecoTool:Exec
    Crossbow:Exec
    C4:Exec
    RemoteControl:Exec
    DualWield:Exec
    AR_BullPup:Exec
    AR_ForwardGrip:Exec
    MedPackPaddles:Exec
    SMG_P90:Exec
    AR_DrumGun:Exec
    Consumable_Small:Exec
    Consumable_Large:Exec
    Balloon:Exec
    MountedTurret:Exec
    CreativeTool:Exec
    ExplosiveBow:Exec
    AshtonIndigo:Exec
    AshtonChicago:Exec
    MeleeDualWield:Exec
    Unarmed:Exec
   )
   n6 = Set isPistol(execute=[@n5.MachinePistol, @n5.Pistol], isPistol=true)
   > n7 = Branch(Condition=AND Boolean(A=@n6, B=Get bIsTargeting()))
   > Set isPistolADS(isPistolADS=true)
   Set isPistolADS(execute=@n7.else, isPistolADS=false)
   n8 = GetAnimInstance(self=GetOwnerSkeletalMeshComponent())
   Set isCustomEmotePlaying(
    execute=[@n1.then_4, @n18.Is Valid]
    isCustomEmotePlaying:double=SelectFloat(
     A=1
     B=0
     bPickA=Montage_IsPlaying(self=@n8, Montage=/BRCosmetics/Animation/Game/MainPlayer/Menu/BR/CustomSkins/_s33/CubicViceM/CubicViceDual_Male_Idle_Montage)
    )
   ) -> (Output_Get:double)
   n9 = Get bIsPlayingEmote()
   n10 = Get bIsPlayingMeleeAnim()
   n11 = Branch(execute=@n1.then_2, Condition=AND Boolean(A=NOT Boolean(A=@n9), B=NOT Boolean(A=@n10)))
   > Set is_not_playing_melee(is_not_playing_melee:bool=true) -> (Output_Get:bool)
   Set is_not_playing_melee(execute=@n11.else, is_not_playing_melee=false)
   n12 = Get bIsParachuteOpened()
   Set is_skydiving_not_parachuting(execute=@n1.then_1, is_skydiving_not_parachuting:bool=AND Boolean(A=Get bIsSkyDiving(), B=NOT Boolean(A=@n12))) -> (Output_Get:bool)
   n13 = Get bIsPlayingEmote()
   n14 = Get bIsInFrontEnd()
   > Set is_front_end_not_emoting(is_front_end_not_emoting:bool=AND Boolean(A=@n14, B=NOT Boolean(A=@n13))) -> (Output_Get:bool)
   > Set is_not_front_end(is_not_front_end:bool=NOT Boolean(A=@n14)) -> (Output_Get:bool)
   n15 = OR Boolean(A=AND Boolean(A=Get bIsInFrontEnd(), B=Get bIsPlayingEmote()), B=Get is_not_front_end() -> (is_not_front_end:bool))
   > Set is_front_end_emoting(is_front_end_emoting:bool=@n15) -> (Output_Get:bool)
   Event BlueprintInitializeAnimation()
   > Set Skel_Mesh(Skel_Mesh=GetOwnerSkeletalMeshComponent())
   > Set fort_player(fort_player:Fort Player Anim Instance Object Reference=GetOwnerFortPlayerAnimInstance!(self:Custom Character Part Anim Instance Object Reference) -> (ReturnValue:Fort Player Anim Instance Object Reference)) -> (Output_Get:Fort Player Anim Instance Object Reference)
   > Set has_hand_glider(has_hand_glider:bool=false) -> (Output_Get:bool)
   n16 = Get fort_player() -> (fort_player:Fort Player Anim Instance Object Reference)
   > Switch on EGliderType(Selection=Break Fort Anim Input Player Glider Anim Asset!(FortAnimInput_PlayerGliderAnimAsset:"Fort Anim Input Player Glider Anim Asset Structure (by ref)"=Get CurrentGliderAnimSet!(self:Fort Player Anim Instance Object Reference=@n16) -> (CurrentGliderAnimSet:Fort Anim Input Player Glider Anim Asset Structure)) -> (
     Default_Base_BS:Blend Space Object Reference
     Default_BodyAdditive_MaleMedium_BS:Blend Space Object Reference
     Default_BodyAdditive_MaleLarge_BS:Blend Space Object Reference
     Default_BodyAdditive_FemaleSmall_BS:Blend Space Object Reference
     Default_BodyAdditive_FemaleMedium_BS:Blend Space Object Reference
     Default_BodyAdditive_FemaleLarge_BS:Blend Space Object Reference
     Default_TurnAdditive_MaleMedium_BS:Blend Space Object Reference
     Default_TurnAdditive_MaleLarge_BS:Blend Space Object Reference
     Default_TurnAdditive_FemaleSmall_BS:Blend Space Object Reference
     Default_TurnAdditive_FemaleMedium_BS:Blend Space Object Reference
     Default_TurnAdditive_FemaleLarge_BS:Blend Space Object Reference
     Into_Base_BS:Blend Space Object Reference
     Into_Base_MaleMedium_BS:Blend Space Object Reference
     Into_BodyAdditive_MaleMedium_BS:Blend Space Object Reference
     Into_BodyAdditive_MaleLarge_BS:Blend Space Object Reference
     Into_BodyAdditive_FemaleSmall_BS:Blend Space Object Reference
     Into_BodyAdditive_FemaleMedium_BS:Blend Space Object Reference
     Into_BodyAdditive_FemaleLarge_BS:Blend Space Object Reference
     Into_TurnAdditive_MaleMedium_BS:Blend Space Object Reference
     Into_TurnAdditive_MaleLarge_BS:Blend Space Object Reference
     Into_TurnAdditive_FemaleSmall_BS:Blend Space Object Reference
     Into_TurnAdditive_FemaleMedium_BS:Blend Space Object Reference
     Into_TurnAdditive_FemaleLarge_BS:Blend Space Object Reference
     Lean_MaleMedium_BS:Blend Space Object Reference
     Lean_MaleLarge_BS:Blend Space Object Reference
     Lean_FemaleSmall_BS:Blend Space Object Reference
     Lean_FemaleMedium_BS:Blend Space Object Reference
     Lean_FemaleLarge_BS:Blend Space Object Reference
     LeanAdditive_Center_MaleMedium_Pose:Anim Sequence Object Reference
     LeanAdditive_Center_MaleLarge_Pose:Anim Sequence Object Reference
     LeanAdditive_Center_FemaleSmall_Pose:Anim Sequence Object Reference
     LeanAdditive_Center_FemaleMedium_Pose:Anim Sequence Object Reference
     LeanAdditive_Center_FemaleLarge_Pose:Anim Sequence Object Reference
     LeanAdditive_Into_BS:Blend Space Object Reference
     LeanAdditive_ForwardInto_Anim:Anim Sequence Object Reference
     LeanAdditive_ForwardInto_FromDeploy_Anim:Anim Sequence Object Reference
     LeanAdditive_BackInto_Anim:Anim Sequence Object Reference
     LeanAdditive_LeftInto_Anim:Anim Sequence Object Reference
     LeanAdditive_RightInto_Anim:Anim Sequence Object Reference
     LeanAdditive_ForwardOut_Anim:Anim Sequence Object Reference
     LeanAdditive_BackOut_Anim:Anim Sequence Object Reference
     LeanAdditive_LeftOut_Anim:Anim Sequence Object Reference
     LeanAdditive_RightOut_Anim:Anim Sequence Object Reference
     ToGlide_BS:Blend Space Object Reference
     ToGlide_Lean_BS:Blend Space Object Reference
     ToDive_BS:Blend Space Object Reference
     ToDive_Lean_BS:Blend Space Object Reference
     Dive_WeaponR_Additive_BS:Blend Space Object Reference
     Glide_WeaponR_Additive_Anim:Anim Sequence Object Reference
     GenericAdditive_Male_BS:Blend Space Object Reference
     GenericAdditive_Female_BS:Blend Space Object Reference
     RootModPitchMin:float=~
     RootModPitchMax:float=~
     RootModYOffsetMin:float=~
     RootModYOffsetMax:float=~
     PelvisModVertical:Pelvis Mod Vertical Input Spring Structure
     PelvisModLateral:Pelvis Mod Lateral Input Spring Structure
     PelvisModBankAngle:Pelvis Mod Bank Angle Input Spring Structure
     ThighModLegAngle:Thigh Mod Leg Angle Input Structure
     ThighModLegPitch:Thigh Mod Leg Pitch Spring Structure
     ThighModLegBank:Thigh Mod Leg Bank Spring Structure
     ThighModLeftLegDrag:Thigh Mod Left Leg Drag Structure
     PlayerGliderType:EGliderType Enum
     bEnableSpringMods:bool=~
     bAllowPlayerDeployRootMod:bool=~
     bUseSurfStyle:bool=~
     bUseLeanPlayRateAdjust:bool=~
     bUseForwardAs_Y_BS:bool=~
     bUseDeployCurveForIntoPlayRateAdjust:bool=~
     LeanForwardValue:float=~
     IdleValue:float=~
     LeanBackValue:float=~
     IdleInterpToSpeed:float=~
     LeanForwardInterpToSpeed:float=~
     LeanBackInterpToSpeed:float=~
    ) {is=title K2Node_BreakStruct}.PlayerGliderType)
   > n17 = Set has_hand_glider(has_hand_glider=true)
   > Set has_hand_glider_parachute_opened(has_hand_glider_parachute_opened:bool=AND Boolean(A=@n17, B=Get bIsParachuteOpened())) -> (Output_Get:bool)
   n18 = Is Valid(InputObject=@n8)
   FInterpTo(Current="To Float (Boolean)"(InBool=@n15), Target=1, DeltaTime=0, InterpSpeed=10)
  ]}
]
variables [
 ar_firingPosition {type=integer}
 fort_player {type=FortPlayerAnimInstance, default=null}
 Gravity Override {type=Vector}
 has_hand_glider {type=boolean}
 has_hand_glider_parachute_opened {type=boolean}
 is_front_end_emoting {type=boolean}
 is_front_end_not_emoting {type=boolean}
 is_not_front_end {type=boolean}
 is_not_playing_melee {type=boolean}
 is_skydiving_not_parachuting {type=boolean}
 isCustomEmotePlaying {type=number}
 isPistol {type=boolean}
 isPistolADS {type=boolean}
 Skel_Mesh {type=SkeletalMeshComponent}
 SlidingAlpha {type=number, default=1}
]
