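# Physics asset of collision shapes for the cape on the CubicVice Dual medium male player body, named as its cape colliders.
# Capsules follow the pelvis, spine, neck, head, clavicles, upper arms, legs and feet, plus four dynamic skirt bones and one bottle bone.
# A very large root sphere centered far below the origin forms a floor-like surface near the feet.
# It has no constraints, and the lower-arm bodies carry no shapes.
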PhysicsAsset /BRCosmetics/Characters/Player/Male/Medium/Bodies/M_MED_CubicVice_Dual/Meshes/M_MED_CubicVice_Dual_CapeColliders
bodies {
 calf_l [{shapeType=Capsule, length=36.5369, radius=10.0145, center=<-21.8561 0 0>, rotation=<rot 90 0 0>}]
 calf_r [{shapeType=Capsule, length=36.5369, radius=10.0145, center=<21.8561 0 0>, rotation=<rot 90 180 0>}]
 clavicle_l [{shapeType=Capsule, length=9, radius=11, center=<9.0559 0 0>, rotation=<rot 90 0 0>}]
 clavicle_r [{shapeType=Capsule, length=9, radius=11, center=<-9.0559 0 0>, rotation=<rot 90 0 180>}]
 dyn_bottleB_02 [{
   shapeType=Capsule
   length=3.2236
   radius=8.2461
   shapeName=dyn_bottleB_02_capsule
   center=<4.7429 -2.3123 -0.2495>
   rotation=<rot 0 -87.376 -89.0478>
  }]
 dyn_skirtC_l_01 [{shapeType=Capsule, length=12, radius=15.5, shapeName=dyn_skirtC_l_01_capsule, center=<11.0653 12.3003 0>, rotation=<rot 90 0 0>}]
 dyn_skirtC_l_03 [{shapeType=Capsule, length=12, radius=15.5, shapeName=dyn_skirtC_l_03_capsule, center=<11.0653 12.3003 0>, rotation=<rot 90 0 0>}]
 dyn_skirtC_r_01 [{shapeType=Capsule, length=12, radius=15.5, shapeName=dyn_skirtC_r_01_capsule, center=<11.0653 12.3003 0>, rotation=<rot 90 0 0>}]
 dyn_skirtC_r_03 [{shapeType=Capsule, length=12, radius=15.5, shapeName=dyn_skirtC_r_03_capsule, center=<11.0653 12.3003 0>, rotation=<rot 90 0 0>}]
 foot_l [{shapeType=Capsule, length=15.89, radius=7.2017, center=<-3.9675 8.0076 0>, rotation=<rot 0 15.0296 90>}]
 foot_r [{shapeType=Capsule, length=15.89, radius=7.2017, center=<4.2686 -7.7665 0>, rotation=<rot 0 15.0296 90>}]
 head [{shapeType=Capsule, length=10.6737, radius=7.3797, center=<4.68 1.9501 0>, rotation=<rot 90 0 0>}]
 lowerarm_l []
 lowerarm_r []
 neck_01 [{shapeType=Capsule, length=6.1454, radius=8.8546, center=<0 1.2878 0>, rotation=<rot 90 0 0>}]
 neck_02 [{shapeType=Capsule, length=5.2432, radius=9.7568, center=<0 2.2107 0>, rotation=<rot 90 0 0>}]
 pelvis [{shapeType=Capsule, length=5, radius=10}]
 root [{shapeType=Sphere, radius=212.4896, shapeName=root_sphere, center=<0 0 -197.5378>}]
 spine_01 [{shapeType=Capsule, length=5, radius=11.5}]
 spine_02 [{shapeType=Capsule, length=6.5, radius=10}]
 spine_03 [{shapeType=Capsule, length=6.5, radius=10}]
 spine_04 [{shapeType=Capsule, length=7.5, radius=11.5}]
 spine_05 [{shapeType=Capsule, length=7, radius=13}]
 thigh_l [{shapeType=Capsule, length=39.098, radius=9.5411, center=<-17.6896 0 0>, rotation=<rot 90 0 0>}]
 thigh_r [{shapeType=Capsule, length=39.098, radius=9.5411, center=<17.6896 0 0>, rotation=<rot 90 0 180>}]
 upperarm_l [{shapeType=Capsule, length=25, radius=7.5, center=<10.7425 0 0>, rotation=<rot 90 0 0>}]
 upperarm_r [{shapeType=Capsule, length=25, radius=7.5, center=<-10.7425 0 0>, rotation=<rot 90 0 180>}]
}
constraints []
