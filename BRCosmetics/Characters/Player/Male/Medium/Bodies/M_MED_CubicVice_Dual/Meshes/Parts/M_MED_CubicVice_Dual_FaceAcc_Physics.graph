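# Physics asset for the face accessory mesh of the male medium CubicVice_Dual cosmetic body, covering two right-side front hair strands.
# Capsule bodies cover the head, upper neck and two bone chains, A with five segments and B with four, each starting at a small MSTR root body.
# The head-to-neck and root-to-head joints swing and twist freely, while each segment links to the previous one with narrow, asymmetric swing limits and a zero twist limit.
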PhysicsAsset /BRCosmetics/Characters/Player/Male/Medium/Bodies/M_MED_CubicVice_Dual/Meshes/Parts/M_MED_CubicVice_Dual_FaceAcc_Physics
bodies {
 dyn_hairFrontA_r_01 [{shapeType=Capsule, length=2, radius=2, shapeName=dyn_hairFrontA_r_01_capsule, center=<2 0 0>, rotation=<rot 90 0 0>}]
 dyn_hairFrontA_r_02 [{shapeType=Capsule, length=2, radius=2, shapeName=dyn_hairFrontA_r_02_capsule, center=<2 0 0>, rotation=<rot 90 0 0>}]
 dyn_hairFrontA_r_03 [{shapeType=Capsule, length=2, radius=2, shapeName=dyn_hairFrontA_r_03_capsule, center=<2 0 0>, rotation=<rot 90 0 0>}]
 dyn_hairFrontA_r_04 [{shapeType=Capsule, length=2, radius=2, shapeName=dyn_hairFrontA_r_04_capsule, center=<2 0 0>, rotation=<rot 90 0 0>}]
 dyn_hairFrontA_r_05 [{shapeType=Capsule, length=2, radius=2, shapeName=dyn_hairFrontA_r_05_capsule, center=<2 0 0>, rotation=<rot 90 0 0>}]
 dyn_hairFrontA_r_MSTR [{shapeType=Capsule, length=0.5, radius=0.5, shapeName=dyn_hairFrontA_r_MSTR_capsule, center=<2 0 0>, rotation=<rot 90 0 0>}]
 dyn_hairFrontB_r_01 [{shapeType=Capsule, length=2, radius=2, shapeName=dyn_hairFrontB_r_01_capsule, center=<2 0 0>, rotation=<rot 90 0 0>}]
 dyn_hairFrontB_r_02 [{shapeType=Capsule, length=2, radius=2, shapeName=dyn_hairFrontB_r_02_capsule, center=<2 0 0>, rotation=<rot 90 0 0>}]
 dyn_hairFrontB_r_03 [{shapeType=Capsule, length=2, radius=2, shapeName=dyn_hairFrontB_r_03_capsule, center=<2 0 0>, rotation=<rot 90 0 0>}]
 dyn_hairFrontB_r_04 [{shapeType=Capsule, length=2, radius=2, shapeName=dyn_hairFrontB_r_04_capsule, center=<2 0 0>, rotation=<rot 90 0 0>}]
 dyn_hairFrontB_r_MSTR [{shapeType=Capsule, length=0.5, radius=0.5, shapeName=dyn_hairFrontB_r_MSTR_capsule, center=<2 0 0>, rotation=<rot 90 0 0>}]
 head [{shapeType=Capsule, length=4.8522, radius=7.1149, center=<4.9984 1.1938 0>, rotation=<rot 90 0 0>}]
 neck_02 [{shapeType=Capsule, length=5, radius=5, rotation=<rot 90 0 0>}]
}
constraints [
 {
  bone1Name=head
  bone2Name=neck_02
  swing1LimitDegrees=45
  swing1Motion=Free
  swing2LimitDegrees=45
  swing2Motion=Free
  twistLimitDegrees=45
  twistMotion=Free
 }
 {
  bone1Name=dyn_hairFrontA_r_MSTR
  bone2Name=head
  swing1LimitDegrees=45
  swing1Motion=Free
  swing2LimitDegrees=45
  swing2Motion=Free
  twistLimitDegrees=45
  twistMotion=Free
 }
 {bone1Name=dyn_hairFrontA_r_01, bone2Name=dyn_hairFrontA_r_MSTR, swing1LimitDegrees=13, swing2LimitDegrees=5, twistLimitDegrees=0}
 {bone1Name=dyn_hairFrontA_r_02, bone2Name=dyn_hairFrontA_r_01, swing1LimitDegrees=13, swing2LimitDegrees=5, twistLimitDegrees=0}
 {bone1Name=dyn_hairFrontA_r_03, bone2Name=dyn_hairFrontA_r_02, swing1LimitDegrees=13, swing2LimitDegrees=5, twistLimitDegrees=0}
 {bone1Name=dyn_hairFrontA_r_04, bone2Name=dyn_hairFrontA_r_03, swing1LimitDegrees=13, swing2LimitDegrees=5, twistLimitDegrees=0}
 {bone1Name=dyn_hairFrontA_r_05, bone2Name=dyn_hairFrontA_r_04, swing1LimitDegrees=13, swing2LimitDegrees=5, twistLimitDegrees=0}
 {
  bone1Name=dyn_hairFrontB_r_MSTR
  bone2Name=head
  swing1LimitDegrees=45
  swing1Motion=Free
  swing2LimitDegrees=45
  swing2Motion=Free
  twistLimitDegrees=45
  twistMotion=Free
 }
 {bone1Name=dyn_hairFrontB_r_01, bone2Name=dyn_hairFrontB_r_MSTR, swing1LimitDegrees=13, swing2LimitDegrees=5, twistLimitDegrees=0}
 {bone1Name=dyn_hairFrontB_r_02, bone2Name=dyn_hairFrontB_r_01, swing1LimitDegrees=13, swing2LimitDegrees=5, twistLimitDegrees=0}
 {bone1Name=dyn_hairFrontB_r_03, bone2Name=dyn_hairFrontB_r_02, swing1LimitDegrees=13, swing2LimitDegrees=5, twistLimitDegrees=0}
 {bone1Name=dyn_hairFrontB_r_04, bone2Name=dyn_hairFrontB_r_03, swing1LimitDegrees=13, swing2LimitDegrees=5, twistLimitDegrees=0}
]
